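# Physics asset for the female medium MutedMaroon cosmetic body, covering its skirt cloth and back hip chain.
# Capsule bodies form a four-link front-left skirt chain, a two-link left-side chain and a two-link hip chain, each tied to the pelvis through a base bone.
# Pelvis, left thigh and three back-skirt capsules add body volume.
# The front chain links, back-skirt joints and last hip-chain joint are fully angle-locked.
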PhysicsAsset /BRCosmetics/Characters/Player/Female/Medium/Bodies/F_MED_MutedMaroon/Meshes/F_MED_MutedMaroon_Physics
bodies {
 Mstr_Skirt_Bck_l [{shapeType=Capsule, length=10, radius=2, shapeName=Mstr_Skirt_Bck_l_capsule, center=<0 0 -2>, rotation=<rot 0 0 90>}]
 Mstr_Skirt_Bck_r [{shapeType=Capsule, length=10, radius=2, shapeName=Mstr_Skirt_Bck_r_capsule, center=<0 0 -2>, rotation=<rot 0 0 90>}]
 Mstr_Skirt_Cloth_Frnt [{shapeType=Capsule, length=0.505, radius=0.505, shapeName=Mstr_Skirt_Cloth_Frnt_capsule, rotation=<rot 0 -90 0>}]
 Mstr_Skirt_Cloth_Side [{shapeType=Capsule, length=0.505, radius=0.505, shapeName=Mstr_Skirt_Cloth_Side_capsule, rotation=<rot 0 -90 0>}]
 Mstr_Skirt_Mid_Bck [{shapeType=Capsule, length=10, radius=2, shapeName=Mstr_Skirt_Mid_Bck_capsule, center=<0 0 -3>, rotation=<rot 0 0 90>}]
 base_hipchain_bk1 [{shapeType=Capsule, length=0.505, radius=0.505, shapeName=base_hipchain_bk1_capsule, rotation=<rot 0 -90 0>}]
 dyn_Skirt_Cloth_Frnt_l_1 [{shapeType=Capsule, length=2, radius=3, shapeName=dyn_Skirt_Cloth_Frnt_l_1_capsule, center=<6 0 0>, rotation=<rot 0 90 90>}]
 dyn_Skirt_Cloth_Frnt_l_2 [{shapeType=Capsule, length=5, radius=3.5, shapeName=dyn_Skirt_Cloth_Frnt_l_2_capsule, center=<7 0 -1>, rotation=<rot 0 90 90>}]
 dyn_Skirt_Cloth_Frnt_l_3 [{shapeType=Capsule, length=5, radius=4, shapeName=dyn_Skirt_Cloth_Frnt_l_3_capsule, center=<7 0 0>, rotation=<rot 81.1283 90 90>}]
 dyn_Skirt_Cloth_Frnt_l_4 [{shapeType=Capsule, length=8, radius=4, shapeName=dyn_Skirt_Cloth_Frnt_l_4_capsule, center=<9 0 0>, rotation=<rot 81.1283 90 90>}]
 dyn_Skirt_Cloth_Side_l_1 [{shapeType=Capsule, length=2, radius=2.5, shapeName=dyn_Skirt_Cloth_Side_l_1_capsule, center=<5 0 0>, rotation=<rot 81.1283 90 90>}]
 dyn_Skirt_Cloth_Side_l_2 [{shapeType=Capsule, length=10, radius=3.5, shapeName=dyn_Skirt_Cloth_Side_l_2_capsule, center=<10 0 0>, rotation=<rot 81.1283 90 90>}]
 dyn_hipchain_bk_01 [
  {
   shapeType=Capsule
   length=22.875
   radius=3.6793
   shapeName=dyn_hipchain_bk_01_capsule
   center=<0.7209 0.0421 2.2892>
   rotation=<rot 0 0 -88.9475>
  }
  {shapeType=Capsule, length=17.5, radius=1.375, shapeName=dyn_hipchain_bk_01_capsule, center=<5.3847 0 0.5133>, rotation=<rot -0.0005 0 90>}
 ]
 dyn_hipchain_bk_02 [{
   shapeType=Capsule
   length=14.625
   radius=3.45
   shapeName=dyn_hipchain_bk_02_capsule
   center=<1.9755 0.293 3.3492>
   rotation=<rot -0.0005 0 -85>
  }]
 pelvis [
  {shapeType=Capsule, length=18.125, radius=12, center=<-0.9657 0.1084 0>}
  {shapeType=Capsule, length=24.875, radius=6.875, shapeName=pelvis_capsule, center=<5.8395 4.1534 0>}
 ]
 thigh_l [{shapeType=Capsule, length=40, radius=7, shapeName=thigh_l_capsule, center=<-22 -1 2>, rotation=<rot 95 0 0>}]
}
constraints [
 {bone1Name=base_hipchain_bk1, bone2Name=pelvis, swing1LimitDegrees=15, swing2LimitDegrees=2, twistLimitDegrees=0}
 {bone1Name=dyn_hipchain_bk_01, bone2Name=base_hipchain_bk1, swing1LimitDegrees=0, swing2LimitDegrees=60, twistLimitDegrees=0}
 {
  bone1Name=dyn_hipchain_bk_02
  bone2Name=dyn_hipchain_bk_01
  swing1LimitDegrees=2
  swing1Motion=Locked
  swing2LimitDegrees=50
  swing2Motion=Locked
  twistLimitDegrees=0
  twistMotion=Locked
 }
 {
  bone1Name=Mstr_Skirt_Bck_r
  bone2Name=pelvis
  swing1LimitDegrees=45
  swing1Motion=Locked
  swing2LimitDegrees=45
  swing2Motion=Locked
  twistLimitDegrees=45
  twistMotion=Locked
 }
 {
  bone1Name=Mstr_Skirt_Bck_l
  bone2Name=pelvis
  swing1LimitDegrees=45
  swing1Motion=Locked
  swing2LimitDegrees=45
  swing2Motion=Locked
  twistLimitDegrees=45
  twistMotion=Locked
 }
 {
  bone1Name=Mstr_Skirt_Mid_Bck
  bone2Name=pelvis
  swing1LimitDegrees=45
  swing1Motion=Locked
  swing2LimitDegrees=45
  swing2Motion=Locked
  twistLimitDegrees=45
  twistMotion=Locked
 }
 {
  bone1Name=thigh_l
  bone2Name=pelvis
  swing1LimitDegrees=45
  swing1Motion=Free
  swing2LimitDegrees=45
  swing2Motion=Free
  twistLimitDegrees=45
  twistMotion=Free
 }
 {
  bone1Name=Mstr_Skirt_Cloth_Frnt
  bone2Name=pelvis
  swing1LimitDegrees=45
  swing1Motion=Free
  swing2LimitDegrees=45
  swing2Motion=Free
  twistLimitDegrees=45
  twistMotion=Free
 }
 {
  bone1Name=dyn_Skirt_Cloth_Frnt_l_1
  bone2Name=Mstr_Skirt_Cloth_Frnt
  swing1LimitDegrees=1
  swing1Motion=Locked
  swing2LimitDegrees=1
  swing2Motion=Locked
  twistLimitDegrees=0
  twistMotion=Locked
 }
 {
  bone1Name=dyn_Skirt_Cloth_Frnt_l_2
  bone2Name=dyn_Skirt_Cloth_Frnt_l_1
  swing1LimitDegrees=1
  swing1Motion=Locked
  swing2LimitDegrees=1
  swing2Motion=Locked
  twistLimitDegrees=0
  twistMotion=Locked
 }
 {
  bone1Name=dyn_Skirt_Cloth_Frnt_l_3
  bone2Name=dyn_Skirt_Cloth_Frnt_l_2
  swing1LimitDegrees=50
  swing1Motion=Locked
  swing2LimitDegrees=10
  swing2Motion=Locked
  twistLimitDegrees=0
  twistMotion=Locked
 }
 {
  bone1Name=dyn_Skirt_Cloth_Frnt_l_4
  bone2Name=dyn_Skirt_Cloth_Frnt_l_3
  swing1LimitDegrees=50
  swing1Motion=Locked
  swing2LimitDegrees=10
  swing2Motion=Locked
  twistLimitDegrees=0
  twistMotion=Locked
 }
 {
  bone1Name=Mstr_Skirt_Cloth_Side
  bone2Name=pelvis
  swing1LimitDegrees=45
  swing1Motion=Free
  swing2LimitDegrees=45
  swing2Motion=Free
  twistLimitDegrees=45
  twistMotion=Free
 }
 {bone1Name=dyn_Skirt_Cloth_Side_l_1, bone2Name=Mstr_Skirt_Cloth_Side, swing1LimitDegrees=10, swing2LimitDegrees=10, twistLimitDegrees=0}
 {bone1Name=dyn_Skirt_Cloth_Side_l_2, bone2Name=dyn_Skirt_Cloth_Side_l_1, swing1LimitDegrees=10, swing2LimitDegrees=10, twistLimitDegrees=0}
]
