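# Physics asset for the MutedMaroon medium female player body, named as a cape collider.
# Capsule bodies cover the pelvis, spine, clavicles, upper arms, thighs and calves, plus five four-bone dynamic skirt chains across the back and sides.
# Constraints chain each skirt strand from the pelvis and follow the skeleton for the body bones.
# Swing and twist are free on every joint except the two thigh-to-pelvis constraints.
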PhysicsAsset /BRCosmetics/Characters/Player/Female/Medium/Bodies/F_MED_MutedMaroon/Meshes/F_MED_MutedMaroon_CapeCollider
bodies {
 calf_l [{shapeType=Capsule, length=40, radius=12, shapeName=calf_l_capsule, center=<-18 0 0>, rotation=<rot 90 0 0>}]
 calf_r [{shapeType=Capsule, length=40, radius=12, shapeName=calf_r_capsule, center=<18 2 0>, rotation=<rot 90 0 0>}]
 clavicle_l [{shapeType=Capsule, length=5, radius=5, shapeName=clavicle_l_capsule, center=<5 0 0>, rotation=<rot 90 0 0>}]
 clavicle_r [{shapeType=Capsule, length=5, radius=5, shapeName=clavicle_r_capsule, center=<-5 0 0>, rotation=<rot 90 0 0>}]
 dyn_Skirt_Bck_l_1 [{shapeType=Capsule, length=12, radius=8, shapeName=dyn_Skirt_Bck_l_1_capsule, rotation=<rot 0 0 90>}]
 dyn_Skirt_Bck_l_2 [{shapeType=Capsule, length=12, radius=8, shapeName=dyn_Skirt_Bck_l_2_capsule, rotation=<rot 0 0 90>}]
 dyn_Skirt_Bck_l_3 [{shapeType=Capsule, length=12, radius=8, shapeName=dyn_Skirt_Bck_l_3_capsule, rotation=<rot 0 0 90>}]
 dyn_Skirt_Bck_l_4 [{shapeType=Capsule, length=12, radius=8, shapeName=dyn_Skirt_Bck_l_4_capsule, center=<2 0 0>, rotation=<rot 0 0 90>}]
 dyn_Skirt_Bck_r_1 [{shapeType=Capsule, length=12, radius=8, shapeName=dyn_Skirt_Bck_r_1_capsule, rotation=<rot 0 0 90>}]
 dyn_Skirt_Bck_r_2 [{shapeType=Capsule, length=12, radius=8, shapeName=dyn_Skirt_Bck_r_2_capsule, rotation=<rot 0 0 90>}]
 dyn_Skirt_Bck_r_3 [{shapeType=Capsule, length=12, radius=8, shapeName=dyn_Skirt_Bck_r_3_capsule, rotation=<rot 0 0 90>}]
 dyn_Skirt_Bck_r_4 [{shapeType=Capsule, length=12, radius=8, shapeName=dyn_Skirt_Bck_r_4_capsule, center=<2 0 0>, rotation=<rot 0 0 90>}]
 dyn_Skirt_Mid_Bck_1 [{shapeType=Capsule, length=12, radius=8, shapeName=dyn_Skirt_Mid_Bck_1_capsule, rotation=<rot 0 0 90>}]
 dyn_Skirt_Mid_Bck_2 [{shapeType=Capsule, length=12, radius=8, shapeName=dyn_Skirt_Mid_Bck_2_capsule, rotation=<rot 0 0 90>}]
 dyn_Skirt_Mid_Bck_3 [{shapeType=Capsule, length=12, radius=8, shapeName=dyn_Skirt_Mid_Bck_3_capsule, rotation=<rot 0 0 90>}]
 dyn_Skirt_Mid_Bck_4 [{shapeType=Capsule, length=12, radius=8, shapeName=dyn_Skirt_Mid_Bck_4_capsule, center=<2 0 0>, rotation=<rot 0 0 90>}]
 dyn_Skirt_Side_l_1 [{shapeType=Capsule, length=12, radius=8, shapeName=dyn_Skirt_Side_l_1_capsule, rotation=<rot 0 0 90>}]
 dyn_Skirt_Side_l_2 [{shapeType=Capsule, length=12, radius=8, shapeName=dyn_Skirt_Side_l_2_capsule, rotation=<rot 0 0 90>}]
 dyn_Skirt_Side_l_3 [{shapeType=Capsule, length=12, radius=8, shapeName=dyn_Skirt_Side_l_3_capsule, rotation=<rot 0 0 90>}]
 dyn_Skirt_Side_l_4 [{shapeType=Capsule, length=12, radius=8, shapeName=dyn_Skirt_Side_l_4_capsule, center=<2 0 0>, rotation=<rot 0 0 90>}]
 dyn_Skirt_Side_r_1 [{shapeType=Capsule, length=12, radius=8, shapeName=dyn_Skirt_Side_r_1_capsule, rotation=<rot 0 0 90>}]
 dyn_Skirt_Side_r_2 [{shapeType=Capsule, length=12, radius=8, shapeName=dyn_Skirt_Side_r_2_capsule, rotation=<rot 0 0 90>}]
 dyn_Skirt_Side_r_3 [{shapeType=Capsule, length=12, radius=8, shapeName=dyn_Skirt_Side_r_3_capsule, rotation=<rot 0 0 90>}]
 dyn_Skirt_Side_r_4 [{shapeType=Capsule, length=12, radius=8, shapeName=dyn_Skirt_Side_r_4_capsule, center=<2 0 0>, rotation=<rot 0 0 90>}]
 pelvis [{shapeType=Capsule, length=15, radius=12}]
 spine_02 [{shapeType=Capsule, length=10, radius=8, shapeName=spine_02_capsule}]
 spine_03 [{shapeType=Capsule, length=0, radius=8, shapeName=spine_03_capsule}]
 spine_04 [{shapeType=Capsule, length=10.8289, radius=8, shapeName=spine_04_capsule}]
 spine_05 [{shapeType=Capsule, length=10.4948, radius=9, shapeName=spine_05_capsule}]
 thigh_l [{shapeType=Capsule, length=45, radius=9, center=<-20 0 0>, rotation=<rot 90 0 0>}]
 thigh_r [{shapeType=Capsule, length=45, radius=9, center=<20 0 0>, rotation=<rot 90 0 0>}]
 upperarm_l [{shapeType=Capsule, length=20, radius=4, shapeName=upperarm_l_capsule, center=<10 0 -1>, rotation=<rot 90 0 0>}]
 upperarm_r [{shapeType=Capsule, length=20, radius=4, shapeName=upperarm_r_capsule, center=<-10 0 1>, rotation=<rot 90 0 0>}]
}
constraints [
 {bone1Name=thigh_l, bone2Name=pelvis, swing1LimitDegrees=45, swing2LimitDegrees=45, twistLimitDegrees=45}
 {bone1Name=thigh_r, bone2Name=pelvis, swing1LimitDegrees=45, swing2LimitDegrees=45, twistLimitDegrees=45}
 {
  bone1Name=dyn_Skirt_Bck_l_1
  bone2Name=pelvis
  swing1LimitDegrees=45
  swing1Motion=Free
  swing2LimitDegrees=45
  swing2Motion=Free
  twistLimitDegrees=45
  twistMotion=Free
 }
 {
  bone1Name=dyn_Skirt_Bck_l_2
  bone2Name=dyn_Skirt_Bck_l_1
  swing1LimitDegrees=45
  swing1Motion=Free
  swing2LimitDegrees=45
  swing2Motion=Free
  twistLimitDegrees=45
  twistMotion=Free
 }
 {
  bone1Name=dyn_Skirt_Bck_l_3
  bone2Name=dyn_Skirt_Bck_l_2
  swing1LimitDegrees=45
  swing1Motion=Free
  swing2LimitDegrees=45
  swing2Motion=Free
  twistLimitDegrees=45
  twistMotion=Free
 }
 {
  bone1Name=dyn_Skirt_Bck_l_4
  bone2Name=dyn_Skirt_Bck_l_3
  swing1LimitDegrees=45
  swing1Motion=Free
  swing2LimitDegrees=45
  swing2Motion=Free
  twistLimitDegrees=45
  twistMotion=Free
 }
 {
  bone1Name=dyn_Skirt_Bck_r_1
  bone2Name=pelvis
  swing1LimitDegrees=45
  swing1Motion=Free
  swing2LimitDegrees=45
  swing2Motion=Free
  twistLimitDegrees=45
  twistMotion=Free
 }
 {
  bone1Name=dyn_Skirt_Bck_r_2
  bone2Name=dyn_Skirt_Bck_r_1
  swing1LimitDegrees=45
  swing1Motion=Free
  swing2LimitDegrees=45
  swing2Motion=Free
  twistLimitDegrees=45
  twistMotion=Free
 }
 {
  bone1Name=dyn_Skirt_Bck_r_3
  bone2Name=dyn_Skirt_Bck_r_2
  swing1LimitDegrees=45
  swing1Motion=Free
  swing2LimitDegrees=45
  swing2Motion=Free
  twistLimitDegrees=45
  twistMotion=Free
 }
 {
  bone1Name=dyn_Skirt_Bck_r_4
  bone2Name=dyn_Skirt_Bck_r_3
  swing1LimitDegrees=45
  swing1Motion=Free
  swing2LimitDegrees=45
  swing2Motion=Free
  twistLimitDegrees=45
  twistMotion=Free
 }
 {
  bone1Name=dyn_Skirt_Mid_Bck_1
  bone2Name=pelvis
  swing1LimitDegrees=45
  swing1Motion=Free
  swing2LimitDegrees=45
  swing2Motion=Free
  twistLimitDegrees=45
  twistMotion=Free
 }
 {
  bone1Name=dyn_Skirt_Mid_Bck_2
  bone2Name=dyn_Skirt_Mid_Bck_1
  swing1LimitDegrees=45
  swing1Motion=Free
  swing2LimitDegrees=45
  swing2Motion=Free
  twistLimitDegrees=45
  twistMotion=Free
 }
 {
  bone1Name=dyn_Skirt_Mid_Bck_3
  bone2Name=dyn_Skirt_Mid_Bck_2
  swing1LimitDegrees=45
  swing1Motion=Free
  swing2LimitDegrees=45
  swing2Motion=Free
  twistLimitDegrees=45
  twistMotion=Free
 }
 {
  bone1Name=dyn_Skirt_Mid_Bck_4
  bone2Name=dyn_Skirt_Mid_Bck_3
  swing1LimitDegrees=45
  swing1Motion=Free
  swing2LimitDegrees=45
  swing2Motion=Free
  twistLimitDegrees=45
  twistMotion=Free
 }
 {
  bone1Name=dyn_Skirt_Side_l_1
  bone2Name=pelvis
  swing1LimitDegrees=45
  swing1Motion=Free
  swing2LimitDegrees=45
  swing2Motion=Free
  twistLimitDegrees=45
  twistMotion=Free
 }
 {
  bone1Name=dyn_Skirt_Side_l_2
  bone2Name=dyn_Skirt_Side_l_1
  swing1LimitDegrees=45
  swing1Motion=Free
  swing2LimitDegrees=45
  swing2Motion=Free
  twistLimitDegrees=45
  twistMotion=Free
 }
 {
  bone1Name=dyn_Skirt_Side_l_3
  bone2Name=dyn_Skirt_Side_l_2
  swing1LimitDegrees=45
  swing1Motion=Free
  swing2LimitDegrees=45
  swing2Motion=Free
  twistLimitDegrees=45
  twistMotion=Free
 }
 {
  bone1Name=dyn_Skirt_Side_l_4
  bone2Name=dyn_Skirt_Side_l_3
  swing1LimitDegrees=45
  swing1Motion=Free
  swing2LimitDegrees=45
  swing2Motion=Free
  twistLimitDegrees=45
  twistMotion=Free
 }
 {
  bone1Name=dyn_Skirt_Side_r_1
  bone2Name=pelvis
  swing1LimitDegrees=45
  swing1Motion=Free
  swing2LimitDegrees=45
  swing2Motion=Free
  twistLimitDegrees=45
  twistMotion=Free
 }
 {
  bone1Name=dyn_Skirt_Side_r_2
  bone2Name=dyn_Skirt_Side_r_1
  swing1LimitDegrees=45
  swing1Motion=Free
  swing2LimitDegrees=45
  swing2Motion=Free
  twistLimitDegrees=45
  twistMotion=Free
 }
 {
  bone1Name=dyn_Skirt_Side_r_3
  bone2Name=dyn_Skirt_Side_r_2
  swing1LimitDegrees=45
  swing1Motion=Free
  swing2LimitDegrees=45
  swing2Motion=Free
  twistLimitDegrees=45
  twistMotion=Free
 }
 {
  bone1Name=dyn_Skirt_Side_r_4
  bone2Name=dyn_Skirt_Side_r_3
  swing1LimitDegrees=45
  swing1Motion=Free
  swing2LimitDegrees=45
  swing2Motion=Free
  twistLimitDegrees=45
  twistMotion=Free
 }
 {
  bone1Name=calf_l
  bone2Name=thigh_l
  swing1LimitDegrees=45
  swing1Motion=Free
  swing2LimitDegrees=45
  swing2Motion=Free
  twistLimitDegrees=45
  twistMotion=Free
 }
 {
  bone1Name=calf_r
  bone2Name=thigh_r
  swing1LimitDegrees=45
  swing1Motion=Free
  swing2LimitDegrees=45
  swing2Motion=Free
  twistLimitDegrees=45
  twistMotion=Free
 }
 {
  bone1Name=spine_02
  bone2Name=pelvis
  swing1LimitDegrees=45
  swing1Motion=Free
  swing2LimitDegrees=45
  swing2Motion=Free
  twistLimitDegrees=45
  twistMotion=Free
 }
 {
  bone1Name=spine_03
  bone2Name=spine_02
  swing1LimitDegrees=45
  swing1Motion=Free
  swing2LimitDegrees=45
  swing2Motion=Free
  twistLimitDegrees=45
  twistMotion=Free
 }
 {
  bone1Name=spine_04
  bone2Name=spine_03
  swing1LimitDegrees=45
  swing1Motion=Free
  swing2LimitDegrees=45
  swing2Motion=Free
  twistLimitDegrees=45
  twistMotion=Free
 }
 {
  bone1Name=spine_05
  bone2Name=spine_04
  swing1LimitDegrees=45
  swing1Motion=Free
  swing2LimitDegrees=45
  swing2Motion=Free
  twistLimitDegrees=45
  twistMotion=Free
 }
 {
  bone1Name=clavicle_l
  bone2Name=spine_05
  swing1LimitDegrees=45
  swing1Motion=Free
  swing2LimitDegrees=45
  swing2Motion=Free
  twistLimitDegrees=45
  twistMotion=Free
 }
 {
  bone1Name=upperarm_l
  bone2Name=clavicle_l
  swing1LimitDegrees=45
  swing1Motion=Free
  swing2LimitDegrees=45
  swing2Motion=Free
  twistLimitDegrees=45
  twistMotion=Free
 }
 {
  bone1Name=clavicle_r
  bone2Name=spine_05
  swing1LimitDegrees=45
  swing1Motion=Free
  swing2LimitDegrees=45
  swing2Motion=Free
  twistLimitDegrees=45
  twistMotion=Free
 }
 {
  bone1Name=upperarm_r
  bone2Name=clavicle_r
  swing1LimitDegrees=45
  swing1Motion=Free
  swing2LimitDegrees=45
  swing2Motion=Free
  twistLimitDegrees=45
  twistMotion=Free
 }
]
