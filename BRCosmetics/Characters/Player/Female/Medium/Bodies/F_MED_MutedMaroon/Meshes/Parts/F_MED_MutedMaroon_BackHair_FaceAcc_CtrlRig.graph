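# Control Rig for the F_MED_MutedMaroon character's back hair and face accessory part, defining corrective offsets for back hair bones.
# A DeltaRotation function splits head rotation against head_null by swing/twist and remaps one Euler axis to a weight for two hair-correction chains.
# A third chain, gated by AR_UP_ADJ, weights hair-base offsets with a right upper-arm spherical pose reader.
# All three chains appear unwired from the Begin Execution sequence.
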ControlRigBlueprint /BRCosmetics/Characters/Player/Female/Medium/Bodies/F_MED_MutedMaroon/Meshes/Parts/F_MED_MutedMaroon_BackHair_FaceAcc_CtrlRig
graphs [
 New Function_ContainedGraph {nodes [
   n1 = RigVMFunctionEntryNode()
   RigVMFunction_MathBoolToInteger = &RigVMFunction_MathBoolToInteger(Value=@n1.onlyTwist)
   RigVMFunction_MathQuaternionSwingTwist = &RigVMFunction_MathQuaternionSwingTwist(
    TwistAxis=@n1.twistAxis
    Input=&RigUnit_GetRelativeTransformForItem(bChildInitial=False, bParentInitial=False, Child=@n1.child, Parent=@n1.parent).RelativeTransform.Rotation
   )
   n2 = RigVMDispatchNode(
    Values=((),())
    Index=@RigVMFunction_MathBoolToInteger.Result
    Values.0=@RigVMFunction_MathQuaternionSwingTwist.Swing
    Values.1=@RigVMFunction_MathQuaternionSwingTwist.Twist
   ) {template="DISPATCH_RigVMDispatch_SelectInt32(in Index,in Values,out Result)"}
   RigVMFunction_MathQuaternionToEuler = &RigVMFunction_MathQuaternionToEuler(Value=@n2.Result, RotationOrder=@n1.Rotation Order)
   > DISPATCH_RigVMDispatch_Print = RigVMDispatchNode(ScreenColor=(R=1,G=1,B=1,A=1), ScreenDuration=1.5, Enabled=@n1.debugRotation, Value=@RigVMFunction_MathQuaternionToEuler.Result) {template="DISPATCH_RigVMDispatch_Print(in Prefix,in Value,in Enabled,in ScreenDuration,in ScreenColor)"}
   Cast = RigVMDispatchNode(Value=@n1.triggerAxis) {template="DISPATCH_RigVMDispatch_CastEnumToInt(in Value,out Result)"}
   DISPATCH_RigVMDispatch_SelectInt32 = RigVMDispatchNode(
    Values=(0,0,0)
    Index=@Cast.Result
    Values.0=@RigVMFunction_MathQuaternionToEuler.Result.X
    Values.1=@RigVMFunction_MathQuaternionToEuler.Result.Y
    Values.2=@RigVMFunction_MathQuaternionToEuler.Result.Z
   ) {template="DISPATCH_RigVMDispatch_SelectInt32(in Index,in Values,out Result)"}
   Remap = &RigVMFunction_MathDoubleRemap(
    bClamp=False
    Value=@DISPATCH_RigVMDispatch_SelectInt32.Result
    TargetMinimum=@n1.outputMinMax.X
    TargetMaximum=@n1.outputMinMax.Y
    SourceMinimum=@n1.rotationMinMax.X
    SourceMaximum=@n1.rotationMinMax.Y
   )
   Absolute = &RigVMFunction_MathDoubleAbs(Value=@Remap.Result)
   Multiply = &RigVMFunction_MathDoubleMul(
    A=&RigVMFunction_MathDoubleClamp(Value=RigVMDispatchNode(True=@Absolute.Result, Condition=@n1.absoluteValue, False=@Remap.Result) {template="DISPATCH_RigVMDispatch_If(in Condition,in True,in False,out Result)"}.Result).Result
    B=@n1.premult
   )
   > RigVMFunctionReturnNode(outputQuat=@n2.Result, outputWeight=@Multiply.Result)
  ]}
 RigVMFunctionLibrary {nodes [DeltaRotation v4 = RigVMCollapseNode(
    ExecuteContext=()
    Rotation Order=ZYX
    child=(Type=Bone,Name="None")
    outputMinMax=(X=0,Y=1)
    parent=(Type=Bone,Name="None")
    premult=1
    rotationMinMax=(X=0,Y=0)
    twistAxis=(X=1,Y=0,Z=0)
   )]}
 RigVMFunction_Sequence_1_ContainedGraph {nodes [
   n1 = RigVMFunctionEntryNode()
   RigVMFunction_Sequence = &RigVMFunction_Sequence(ExecuteContext=@n1.ExecuteContext)
   n2 = &RigVMFunction_Sequence(ExecuteContext=@RigVMFunction_Sequence.B)
   n3 = &RigVMFunction_Sequence(ExecuteContext=@n2.B)
   RigVMFunctionReturnNode(A=@RigVMFunction_Sequence.A, B=@n2.A, C=@n3.A, D=@n3.B)
  ]}
 RigVMModel {nodes [
   &RigUnit_ModifyTransforms(
    ExecutePin=()
    ItemToModify=((Item=(Type=Bone,Name="head"),Transform=(Rotation=(X=0,Y=0,Z=0,W=1),Translation=(X=0,Y=0,Z=0),Scale3D=(X=1,Y=1,Z=1))))
    Weight=1
    WeightMinimum=0
   )
   > DeltaRotation v4_1 = RigVMFunctionReferenceNode(
    Rotation Order=ZYX
    child=(Type=Bone,Name="head")
    outputMinMax=(X=0,Y=1)
    parent=(Type=Null,Name="head_null")
    premult=1
    rotationMinMax=(X=0,Y=30)
    triggerAxis=Z
    twistAxis=(X=1,Y=0,Z=0)
   )
   AlphaInterp = &RigVMFunction_AlphaInterp(
    InterpSpeedDecreasing=10
    InterpSpeedIncreasing=10
    Scale=1
    bClampResult=False
    bInterpResult=False
    Value=@"DeltaRotation v4_1".outputWeight
   )
   > &RigUnit_ModifyTransforms(
    ItemToModify="((Item=(Type=Bone,Name=\"dyn_hair_back_01_l\"),Transform=(Rotation=(X=0,Y=0,Z=-0.0436,W=0.999),Translation=(X=0,Y=0,Z=0),Scale3D=(X=1,Y=1,Z=1))),(Item=(Type=Bone,Name=\"dyn_hair_back_01_m\"),Transform=(Rotation=(X=0,Y=0,Z=-0.0436,W=0.999),Translation=(X=0,Y=0, ...(truncated)"
    WeightMinimum=0
    Weight=@AlphaInterp.Result
   )
   > &RigUnit_ModifyTransforms(
    ItemToModify="((Item=(Type=Bone,Name=\"dyn_hair_back_02_l\"),Transform=(Rotation=(X=0,Y=0,Z=0.0872,W=0.9962),Translation=(X=0,Y=0,Z=0),Scale3D=(X=1,Y=1,Z=1))),(Item=(Type=Bone,Name=\"dyn_hair_back_02_m\"),Transform=(Rotation=(X=0,Y=0,Z=0.0872,W=0.9962),Translation=(X=0,Y=0, ...(truncated)"
    WeightMinimum=0
    Weight=@AlphaInterp.Result
   )
   &RigUnit_ModifyTransforms(
    ExecutePin=()
    ItemToModify=((Item=(Type=Bone,Name="head"),Transform=(Rotation=(X=0,Y=0,Z=0,W=1),Translation=(X=0,Y=0,Z=0),Scale3D=(X=1,Y=1,Z=1))))
    Weight=1
    WeightMinimum=0
   )
   > DeltaRotation v4 = RigVMFunctionReferenceNode(
    Rotation Order=ZYX
    child=(Type=Bone,Name="head")
    outputMinMax=(X=0,Y=1)
    parent=(Type=Null,Name="head_null")
    premult=1
    rotationMinMax=(X=0,Y=-30)
    triggerAxis=Y
    twistAxis=(X=1,Y=0,Z=0)
   )
   n1 = &RigVMFunction_AlphaInterp(InterpSpeedDecreasing=10, InterpSpeedIncreasing=10, Scale=1, bClampResult=False, bInterpResult=False, Value=@"DeltaRotation v4".outputWeight)
   > &RigUnit_ModifyTransforms(
    ItemToModify="((Item=(Type=Bone,Name=\"dyn_hair_back_01_r\"),Transform=(Rotation=(X=0,Y=0.1736,Z=0,W=0.9848),Translation=(X=0,Y=0,Z=0),Scale3D=(X=1,Y=1,Z=1))),(Item=(Type=Bone,Name=\"dyn_hair_back_01_m\"),Transform=(Rotation=(X=0,Y=0.0872,Z=0,W=0.9962),Translation=(X=0,Y=0, ...(truncated)"
    WeightMinimum=0
    Weight=@n1.Result
   )
   RigUnit_BeginExecution = &RigUnit_BeginExecution()
   RigVMFunction_Sequence = RigVMAggregateNode(ExecuteContext=@RigUnit_BeginExecution.ExecutePin)
   RigVMFunction_ControlFlowBranch = &RigVMFunction_ControlFlowBranch(ExecuteContext=(), Condition=RigVMVariableNode() {variable=AR_UP_ADJ}.Value)
   n2 = &RigUnit_SphericalPoseReader(
    ActiveRegionScaleFactors=(PositiveWidth=1.2,NegativeWidth=1.2,PositiveHeight=1.2,NegativeHeight=1.2)
    ActiveRegionSize=0.1
    Debug=(bDrawDebug=true,bDraw2D=False,bDrawLocalAxes=False,DebugScale=25,DebugSegments=20,DebugThickness=0.25)
    DriverAxis=(X=-1,Y=0,Z=0)
    DriverItem=(Type=Bone,Name="upperarm_r")
    FalloffRegionScaleFactors=(PositiveWidth=1,NegativeWidth=1,PositiveHeight=1,NegativeHeight=1)
    FalloffSize=0.3
    FlipHeightScaling=False
    FlipWidthScaling=False
    OptionalParentItem=(Type=Bone,Name="spine_05")
    RotationOffset=(X=-46.0107,Y=13.9098,Z=0)
    ExecutePin=@RigVMFunction_ControlFlowBranch.True
   )
   > &RigUnit_ModifyTransforms(
    ItemToModify="((Item=(Type=Bone,Name=\"base_hair_back_r\"),Transform=(Rotation=(X=0,Y=0.1736,Z=0,W=0.9848),Translation=(X=0,Y=0,Z=0),Scale3D=(X=1,Y=1,Z=1))),(Item=(Type=Bone,Name=\"base_hair_back_m\"),Transform=(Rotation=(X=0,Y=0.1736,Z=0,W=0.9848),Translation=(X=0,Y=0,Z=0) ...(truncated)"
    WeightMinimum=0
    Weight=@n2.OutputParam
   )
  ]}
]
hierarchy [
 root {loc=<0 0 0>, rot=<0 0 0 1>}
 pelvis {parent=root, loc=<0 -0.5301 98.0571>, rot=<0.0127 0.707 -0.0127 -0.707>}
 spine_01 {parent=pelvis, loc=<6.1208 -0.9328 0>, rot=<0 0 0.034 -0.9994>}
 spine_02 {parent=spine_01, loc=<6.5997 -0.2095 0>, rot=<0 0 0.0238 -0.9997>}
 spine_03 {parent=spine_02, loc=<6.4914 -0.1237 0>, rot=<0 0 0.0149 -0.9999>}
 spine_04 {parent=spine_03, loc=<6.9598 0.0794 0>, rot=<0 0 -0.0132 -0.9999>}
 spine_05 {parent=spine_04, loc=<6.9987 0 0>, rot=<0 0 -0.0015 -1>}
 neck_01 {parent=spine_05, loc=<11.3138 0.1529 0>, rot=<0 0 -0.2003 -0.9797>}
 neck_02 {parent=neck_01, loc=<4.5627 0 0>, rot=<0 0 0.0504 -0.9987>}
 head {parent=neck_02, loc=<5.3914 0 0>, rot=<0 0 0.115 -0.9934>}
 Mstr_Back_Hair {parent=head, loc=<5.65 1.2676 0.2981>, rot=<0.5521 -0.4418 0.4418 -0.5521>}
 base_headwing_l {parent=Mstr_Back_Hair, loc=<-1.9648 6.4829 5.1567>, rot=<-0.7062 -0.0047 -0.4373 -0.5568>}
 dyn_headwing_l_01 {parent=base_headwing_l, loc=<0 0 0>, rot=<0 0 0 1>}
 dyn_headwing_l_02 {parent=dyn_headwing_l_01, loc=<6.9999 0 0>, rot=<0 0.0678 0 -0.9977>}
 dyn_headwing_l_03 {parent=dyn_headwing_l_02, loc=<7 0 0>, rot=<0 0 0 1>}
 base_headwing_r {parent=Mstr_Back_Hair, loc=<-1.9648 -5.8867 5.1567>, rot=<-0.4107 0.464 0.0887 -0.7799>}
 dyn_headwing_r_01 {parent=base_headwing_r, loc=<0 0 0>, rot=<-0.0003 -0.0224 -0.0115 -0.9997>}
 dyn_headwing_r_02 {parent=dyn_headwing_r_01, loc=<7 0 0>, rot=<-0.0178 -0.1192 0.0717 -0.9901>}
 dyn_headwing_r_03 {parent=dyn_headwing_r_02, loc=<7 0 0>, rot=<0 0 0 1>}
 MSTR_Top_hair {parent=Mstr_Back_Hair, loc=<-1.7028 3.1681 10.2985>, rot=<0.0499 0.4151 0.4857 -0.7676>}
 dyn_Top_Hair_1 {parent=MSTR_Top_hair, loc=<0 0 0>, rot=<0 0 0 1>}
 dyn_Top_Hair_2 {parent=dyn_Top_Hair_1, loc=<3.5 0 0>, rot=<0 -0.2219 0 -0.9751>}
 dyn_Top_Hair_3 {parent=dyn_Top_Hair_2, loc=<3.5 0 0>, rot=<0 0 0 1>}
 Mstr_FrntHair_A {parent=Mstr_Back_Hair, loc=<-2.7535 3.8794 8.659>, rot=<-0.0625 0.4313 0.8907 -0.1295>}
 dyn_FrntHair_A_1 {parent=Mstr_FrntHair_A, loc=<0 0 0>, rot=<0 -0.2362 0 -0.9717>}
 dyn_FrntHair_A_2 {parent=dyn_FrntHair_A_1, loc=<5 0 0>, rot=<0 -0.2526 0 -0.9676>}
 dyn_FrntHair_A_3 {parent=dyn_FrntHair_A_2, loc=<5 0 0>, rot=<0 -0.4146 0 -0.91>}
 dyn_FrntHair_A_4 {parent=dyn_FrntHair_A_3, loc=<5 0 0>, rot=<0 -0.1184 0 -0.993>}
 dyn_FrntHair_A_5 {parent=dyn_FrntHair_A_4, loc=<4.4136 0 0>, rot=<0 0 0 1>}
 Mstr_FrntHair_B {parent=Mstr_Back_Hair, loc=<-2.6611 4.6645 6.9498>, rot=<-0.0066 0.399 0.9105 0.1083>}
 dyn_FrntHair_B_1 {parent=Mstr_FrntHair_B, loc=<0 0 0>, rot=<0 -0.2362 0 -0.9717>}
 dyn_FrntHair_B_2 {parent=dyn_FrntHair_B_1, loc=<5 0 0>, rot=<0 -0.3474 0 -0.9377>}
 dyn_FrntHair_B_3 {parent=dyn_FrntHair_B_2, loc=<5 0 0>, rot=<0 -0.3113 0 -0.9503>}
 Mstr_FrntHair_C {parent=Mstr_Back_Hair, loc=<-2.1588 5.1097 6.7446>, rot=<-0.2305 0.1149 0.8608 0.4389>}
 dyn_FrntHair_C_1 {parent=Mstr_FrntHair_C, loc=<0 0 0>, rot=<0 -0.2362 0 -0.9717>}
 dyn_FrntHair_C_2 {parent=dyn_FrntHair_C_1, loc=<5 0 0>, rot=<0 -0.2411 0 -0.9705>}
 dyn_FrntHair_A_B_1 {parent=Mstr_Back_Hair, loc=<0.1571 -8.6118 1.3803>, rot=<-0.6476 0.0736 0.3165 -0.6892>}
 dyn_FrntHair_A_B_2 {parent=dyn_FrntHair_A_B_1, loc=<4.4136 0 0>, rot=<0 0 -0.1317 -0.9913>}
 base_hair_back_l {parent=Mstr_Back_Hair, loc=<3.1798 6.1524 2.7286>, rot=<-0.547 0.3936 0.5653 0.4757>}
 dyn_hair_back_01_l {parent=base_hair_back_l, loc=<0 0 0>, rot=<0 0 0 1>}
 dyn_hair_back_02_l {parent=dyn_hair_back_01_l, loc=<8.9243 0.1654 0>, rot=<0 0 0.1022 -0.9948>}
 dyn_hair_back_03_l {parent=dyn_hair_back_02_l, loc=<6.4897 0.0983 0>, rot=<0 0 0.126 -0.992>}
 dyn_hair_back_04_l {parent=dyn_hair_back_03_l, loc=<7.6353 -0.1135 0>, rot=<0 0 -0.1976 -0.9803>}
 dyn_hair_back_05_l {parent=dyn_hair_back_04_l, loc=<7.0615 0.0597 0>, rot=<0 0 -0.1412 -0.99>}
 dyn_hair_back_06_l {parent=dyn_hair_back_05_l, loc=<7.3651 -0.0072 0>, rot=<0 0 -0.025 -0.9997>}
 dyn_Back_Hair_A_A {parent=dyn_hair_back_01_l, loc=<8.636 -2.2026 2.6252>, rot=<0.2181 -0.7855 -0.396 0.4225>}
 base_hair_back_m {parent=Mstr_Back_Hair, loc=<6.7016 0.2985 1.6228>, rot=<0.523 -0.4759 -0.4759 -0.523>}
 dyn_hair_back_01_m {parent=base_hair_back_m, loc=<0 0 0>, rot=<0 0 0 1>}
 dyn_hair_back_02_m {parent=dyn_hair_back_01_m, loc=<8.0943 -0.0648 0>, rot=<0 0 0.0109 -0.9999>}
 dyn_hair_back_03_m {parent=dyn_hair_back_02_m, loc=<6.0888 -0.0022 0>, rot=<0 0 0.0056 -1>}
 dyn_hair_back_04_m {parent=dyn_hair_back_03_m, loc=<6.5783 0 0>, rot=<0 0 -0.0259 -0.9997>}
 dyn_hair_back_05_m {parent=dyn_hair_back_04_m, loc=<6.9646 0 0>, rot=<0 0 -0.1294 -0.9916>}
 dyn_hair_back_06_m {parent=dyn_hair_back_05_m, loc=<7.3535 0.0097 0>, rot=<0 0 -0.0235 -0.9997>}
 base_hair_back_r {parent=Mstr_Back_Hair, loc=<3.1798 -5.5562 2.7286>, rot=<0.5422 -0.5076 -0.4656 -0.4812>}
 dyn_hair_back_01_r {parent=base_hair_back_r, loc=<0 0 0>, rot=<0 0 0 1>}
 dyn_hair_back_02_r {parent=dyn_hair_back_01_r, loc=<8.9243 0.1654 0>, rot=<0 0 0.1022 -0.9948>}
 dyn_hair_back_03_r {parent=dyn_hair_back_02_r, loc=<6.4894 0.0982 0>, rot=<0 0 0.126 -0.992>}
 dyn_hair_back_04_r {parent=dyn_hair_back_03_r, loc=<7.6161 -0.0997 0>, rot=<0 0 -0.2001 -0.9798>}
 dyn_hair_back_05_r {parent=dyn_hair_back_04_r, loc=<7.0615 0.0597 0>, rot=<0 0 -0.1412 -0.99>}
 dyn_hair_back_06_r {parent=dyn_hair_back_05_r, loc=<7.3651 -0.0072 0>, rot=<0 0 -0.025 -0.9997>}
 dyn_Back_Hair_C_A {parent=dyn_hair_back_01_r, loc=<9.1571 -2.6029 -4.9278>, rot=<0.7089 -0.1268 -0.6886 0.0857>}
 Mstr_earring_l {parent=head, loc=<0.0241 -0.0112 -6.0776>, rot=<0.0039 -0.9374 0.3482 -0.0014>}
 dyn_earring_l {parent=Mstr_earring_l, loc=<0 0 0>, rot=<0 0 0 1>}
 Mstr_earring_r {parent=head, loc=<0.115 -0.0241 6.0468>, rot=<0.0015 -0.3532 0.9355 -0.0039>}
 dyn_earring_r {parent=Mstr_earring_r, loc=<0 0 0>, rot=<0 0 0 1>}
 head_null {parent=neck_02, type=Null, loc=<5.3914 0 0>, rot=<0 0 0.115 -0.9934>}
 clavicle_l {parent=spine_05, loc=<5.4831 -0.1053 -3.2946>, rot=<0.0465 -0.7289 -0.009 -0.6829>}
 upperarm_l {parent=clavicle_l, loc=<9.8658 -1.2136 -0.4105>, rot=<0.0164 -0.3245 -0.0269 -0.9454>}
 clavicle_r {parent=spine_05, loc=<5.4831 -0.1053 3.2946>, rot=<0.7289 0.0465 -0.6829 0.009>}
 upperarm_r {parent=clavicle_r, loc=<-9.8658 1.2136 0.4105>, rot=<0.0164 -0.3245 -0.0269 -0.9454>}
 attach {parent=root, loc=<0 0 0>, rot=<0 0 0 1>}
 ik_hand_root {parent=root, loc=<0 0 0>, rot=<0 0 0 1>}
 ik_hand_gun {parent=ik_hand_root, loc=<-48.1743 3.3545 104.4381>, rot=<0.6276 0.1309 0.4183 -0.6434>}
 ik_hand_r {parent=ik_hand_gun, loc=<0 0 0>, rot=<0 0 0 1>}
 ik_hand_l {parent=ik_hand_gun, loc=<59.3271 67.6997 34.3524>, rot=<-0.6158 -0.7027 -0.3565 0>}
 ik_foot_root {parent=root, loc=<0 0 0>, rot=<0 0 0 1>}
 ik_foot_l {parent=ik_foot_root, loc=<14.8262 -7.3545 8.296>, rot=<0.0357 0.7045 0.0284 -0.7082>}
 ik_foot_r {parent=ik_foot_root, loc=<-14.8262 -7.3545 8.296>, rot=<0.7082 0.0284 -0.7045 0.0357>}
 root_fp {parent=root, loc=<0 0 0>, rot=<0 0 0 1>}
 attach_fp {parent=root_fp, loc=<0 0 0>, rot=<0 0 0 1>}
 camera_root {parent=root_fp, loc=<0 0 0>, rot=<0 0 0 1>}
 camera {parent=camera_root, loc=<0 0 0>, rot=<0 0 0 1>}
 attach_emote_01 {parent=root, loc=<0 0 0>, rot=<0 0 0 1>}
 attach_emote_02 {parent=root, loc=<0 0 0>, rot=<0 0 0 1>}
 attach_emote_03 {parent=root, loc=<0 0 0>, rot=<0 0 0 1>}
 warp_point_01 {parent=root, loc=<0 0 0>, rot=<0 0 0 1>}
 warp_point_02 {parent=root, loc=<0 0 0>, rot=<0 0 0 1>}
 warp_point_03 {parent=root, loc=<0 0 0>, rot=<0 0 0 1>}
]
variables [
 AR_Down_ADJ {type=bool, default=False}
 AR_UP_ADJ {type=bool, default=False}
 CrouchingNot {type=bool, default=False}
]
