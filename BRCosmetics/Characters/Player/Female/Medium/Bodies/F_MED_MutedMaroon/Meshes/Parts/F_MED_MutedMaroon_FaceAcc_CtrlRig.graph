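# Control Rig for the F_MED_MutedMaroon character's face accessory mesh, correcting the dynamic back-hair bones against the upper body.
# Weighted parent constraints tie three back-hair chains to spine, clavicle and arm bones, with side-chain weights easing off as parachuting_alpha rises.
# A DeltaRotation function turns head rotation relative to head_null into a weight for offsetting hair bones.
# Spherical pose readers on each clavicle drive additional side-hair offsets.
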ControlRigBlueprint /BRCosmetics/Characters/Player/Female/Medium/Bodies/F_MED_MutedMaroon/Meshes/Parts/F_MED_MutedMaroon_FaceAcc_CtrlRig
graphs [
 New Function_ContainedGraph {nodes [
   n1 = RigVMFunctionEntryNode()
   RigVMFunction_MathBoolToInteger = &RigVMFunction_MathBoolToInteger(Value=@n1.onlyTwist)
   RigVMFunction_MathQuaternionSwingTwist = &RigVMFunction_MathQuaternionSwingTwist(
    TwistAxis=@n1.twistAxis
    Input=&RigUnit_GetRelativeTransformForItem(bChildInitial=False, bParentInitial=False, Child=@n1.child, Parent=@n1.parent).RelativeTransform.Rotation
   )
   n2 = RigVMDispatchNode(
    Values=((),())
    Index=@RigVMFunction_MathBoolToInteger.Result
    Values.0=@RigVMFunction_MathQuaternionSwingTwist.Swing
    Values.1=@RigVMFunction_MathQuaternionSwingTwist.Twist
   ) {template="DISPATCH_RigVMDispatch_SelectInt32(in Index,in Values,out Result)"}
   RigVMFunction_MathQuaternionToEuler = &RigVMFunction_MathQuaternionToEuler(Value=@n2.Result, RotationOrder=@n1.Rotation Order)
   > DISPATCH_RigVMDispatch_Print = RigVMDispatchNode(ScreenColor=(R=1,G=1,B=1,A=1), ScreenDuration=1.5, Enabled=@n1.debugRotation, Value=@RigVMFunction_MathQuaternionToEuler.Result) {template="DISPATCH_RigVMDispatch_Print(in Prefix,in Value,in Enabled,in ScreenDuration,in ScreenColor)"}
   Cast = RigVMDispatchNode(Value=@n1.triggerAxis) {template="DISPATCH_RigVMDispatch_CastEnumToInt(in Value,out Result)"}
   DISPATCH_RigVMDispatch_SelectInt32 = RigVMDispatchNode(
    Values=(0,0,0)
    Index=@Cast.Result
    Values.0=@RigVMFunction_MathQuaternionToEuler.Result.X
    Values.1=@RigVMFunction_MathQuaternionToEuler.Result.Y
    Values.2=@RigVMFunction_MathQuaternionToEuler.Result.Z
   ) {template="DISPATCH_RigVMDispatch_SelectInt32(in Index,in Values,out Result)"}
   Remap = &RigVMFunction_MathDoubleRemap(
    bClamp=False
    Value=@DISPATCH_RigVMDispatch_SelectInt32.Result
    TargetMinimum=@n1.outputMinMax.X
    TargetMaximum=@n1.outputMinMax.Y
    SourceMinimum=@n1.rotationMinMax.X
    SourceMaximum=@n1.rotationMinMax.Y
   )
   Absolute = &RigVMFunction_MathDoubleAbs(Value=@Remap.Result)
   Multiply = &RigVMFunction_MathDoubleMul(
    A=&RigVMFunction_MathDoubleClamp(Value=RigVMDispatchNode(True=@Absolute.Result, Condition=@n1.absoluteValue, False=@Remap.Result) {template="DISPATCH_RigVMDispatch_If(in Condition,in True,in False,out Result)"}.Result).Result
    B=@n1.premult
   )
   > RigVMFunctionReturnNode(outputQuat=@n2.Result, outputWeight=@Multiply.Result)
  ]}
 RigVMFunctionLibrary {nodes [DeltaRotation v4 = RigVMCollapseNode(
    ExecuteContext=()
    Rotation Order=ZYX
    child=(Type=Bone,Name="None")
    outputMinMax=(X=0,Y=1)
    parent=(Type=Bone,Name="None")
    premult=1
    rotationMinMax=(X=0,Y=0)
    twistAxis=(X=1,Y=0,Z=0)
   )]}
 RigVMModel {nodes [
   &RigUnit_ModifyTransforms(
    ExecutePin=()
    ItemToModify=((Item=(Type=Bone,Name="head"),Transform=(Rotation=(X=0,Y=0,Z=0,W=1),Translation=(X=0,Y=0,Z=0),Scale3D=(X=1,Y=1,Z=1))))
    Weight=1
    WeightMinimum=0
   )
   > DeltaRotation v4 = RigVMFunctionReferenceNode(
    Rotation Order=ZYX
    child=(Type=Bone,Name="head")
    outputMinMax=(X=0,Y=1)
    parent=(Type=Null,Name="head_null")
    premult=1
    rotationMinMax=(X=0,Y=-30)
    triggerAxis=Y
    twistAxis=(X=1,Y=0,Z=0)
   )
   > &RigUnit_ModifyTransforms(
    ItemToModify="((Item=(Type=Bone,Name=\"dyn_hair_back_01_r\"),Transform=(Rotation=(X=0,Y=0.4226,Z=0,W=0.9063),Translation=(X=0,Y=0,Z=0),Scale3D=(X=1,Y=1,Z=1))),(Item=(Type=Bone,Name=\"dyn_hair_back_01_m\"),Transform=(Rotation=(X=0,Y=0.3007,Z=0,W=0.9537),Translation=(X=0,Y=0, ...(truncated)"
    WeightMinimum=0
    Weight=@"DeltaRotation v4".outputWeight
   )
   RigUnit_BeginExecution = &RigUnit_BeginExecution()
   SequenceAggregate = RigVMAggregateNode(ExecuteContext=@RigUnit_BeginExecution.ExecutePin)
   &RigUnit_ParentConstraint(
    AdvancedSettings=(RotationOrderForFilter=XZY)
    Child=(Type=Bone,Name="dyn_hair_back_02_m")
    Parents="((Item=(Type=Bone,Name=\"upperarm_r\"),Weight=0.0002),(Item=(Type=Bone,Name=\"upperarm_l\"),Weight=0.0002),(Item=(Type=Bone,Name=\"deform_upperarm_up_l\"),Weight=0.0005),(Item=(Type=Bone,Name=\"deform_upperarm_up_r\"),Weight=0.0006),(Item=(Type=Bone,Name=\"spine_02 ...(truncated)"
    Weight=1
    ExecutePin=@SequenceAggregate.A
   )
   > &RigUnit_ParentConstraint(
    AdvancedSettings=(RotationOrderForFilter=XZY)
    Child=(Type=Bone,Name="dyn_hair_back_03_m")
    Parents="((Item=(Type=Bone,Name=\"spine_01\"),Weight=0.0005),(Item=(Type=Bone,Name=\"spine_02\"),Weight=0.0017),(Item=(Type=Bone,Name=\"spine_03\"),Weight=0.0089),(Item=(Type=Bone,Name=\"spine_04\"),Weight=0.0122),(Item=(Type=Bone,Name=\"clavicle_r\"),Weight=0.1847),(Item=(T ...(truncated)"
    Weight=1
   )
   > &RigUnit_ParentConstraint(
    AdvancedSettings=(RotationOrderForFilter=XZY)
    Child=(Type=Bone,Name="dyn_hair_back_04_m")
    Parents="((Item=(Type=Bone,Name=\"spine_01\"),Weight=0.0008),(Item=(Type=Bone,Name=\"spine_02\"),Weight=0.0019),(Item=(Type=Bone,Name=\"spine_03\"),Weight=0.0098),(Item=(Type=Bone,Name=\"spine_04\"),Weight=0.0129),(Item=(Type=Bone,Name=\"clavicle_r\"),Weight=0.1819),(Item=(T ...(truncated)"
    Weight=1
   )
   > &RigUnit_ParentConstraint(
    AdvancedSettings=(RotationOrderForFilter=XZY)
    Child=(Type=Bone,Name="dyn_hair_back_05_m")
    Parents="((Item=(Type=Bone,Name=\"upperarm_l\"),Weight=0.0004),(Item=(Type=Bone,Name=\"spine_01\"),Weight=0.0089),(Item=(Type=Bone,Name=\"spine_02\"),Weight=0.0139),(Item=(Type=Bone,Name=\"spine_03\"),Weight=0.0712),(Item=(Type=Bone,Name=\"clavicle_r\"),Weight=0.0845),(Item= ...(truncated)"
    Weight=1
   )
   n1 = RigVMVariableNode() {variable=parachuting_alpha}
   n2 = /Script/RigVM.RigVMFunction_MathDoubleLerp(A=1, B=0.5, T=@n1.Value) {is=struct RigVMUnitNode, template="Interpolate::Execute(in A,in B,in T,out Result)"}
   &RigUnit_ParentConstraint(
    AdvancedSettings=(RotationOrderForFilter=XZY)
    Child=(Type=Bone,Name="dyn_hair_back_02_l")
    Parents="((Item=(Type=Bone,Name=\"upperarm_twist_01_l\"),Weight=0.0002),(Item=(Type=Bone,Name=\"spine_02\"),Weight=0.0002),(Item=(Type=Bone,Name=\"spine_03\"),Weight=0.0014),(Item=(Type=Bone,Name=\"spine_04\"),Weight=0.0026),(Item=(Type=Bone,Name=\"upperarm_l\"),Weight=0.005 ...(truncated)"
    Weight=@n2.Result
    ExecutePin=@SequenceAggregate.B
   )
   n3 = &RigVMFunction_MathDoubleLerp(A=1, B=0.25, T=@n1.Value) {template="Interpolate::Execute(in A,in B,in T,out Result)"}
   > &RigUnit_ParentConstraint(
    AdvancedSettings=(RotationOrderForFilter=XZY)
    Child=(Type=Bone,Name="dyn_hair_back_03_l")
    Parents="((Item=(Type=Bone,Name=\"upperarm_twist_01_l\"),Weight=0.0006),(Item=(Type=Bone,Name=\"clavicle_r\"),Weight=0.0011),(Item=(Type=Bone,Name=\"spine_03\"),Weight=0.0012),(Item=(Type=Bone,Name=\"spine_04\"),Weight=0.0024),(Item=(Type=Bone,Name=\"upperarm_l\"),Weight=0.0 ...(truncated)"
    Weight=@n3.Result
   )
   n4 = &RigVMFunction_MathDoubleLerp(A=1, B=0, T=@n1.Value) {template="Interpolate::Execute(in A,in B,in T,out Result)"}
   > &RigUnit_ParentConstraint(
    AdvancedSettings=(RotationOrderForFilter=XZY)
    Child=(Type=Bone,Name="dyn_hair_back_04_l")
    Parents="((Item=(Type=Bone,Name=\"upperarm_twist_01_l\"),Weight=0.0005),(Item=(Type=Bone,Name=\"spine_02\"),Weight=0.0014),(Item=(Type=Bone,Name=\"spine_03\"),Weight=0.0062),(Item=(Type=Bone,Name=\"spine_04\"),Weight=0.0132),(Item=(Type=Bone,Name=\"upperarm_l\"),Weight=0.042 ...(truncated)"
    Weight=@n4.Result
   )
   > &RigUnit_ParentConstraint(
    AdvancedSettings=(RotationOrderForFilter=XZY)
    Child=(Type=Bone,Name="dyn_hair_back_05_l")
    Parents="((Item=(Type=Bone,Name=\"clavicle_r\"),Weight=0.0003),(Item=(Type=Bone,Name=\"pelvis\"),Weight=0.0004),(Item=(Type=Bone,Name=\"spine_01\"),Weight=0.0011),(Item=(Type=Bone,Name=\"spine_02\"),Weight=0.0022),(Item=(Type=Bone,Name=\"upperarm_twist_01_l\"),Weight=0.0148) ...(truncated)"
    Weight=@n4.Result
   )
   &RigUnit_ParentConstraint(
    AdvancedSettings=(RotationOrderForFilter=XZY)
    Child=(Type=Bone,Name="dyn_hair_back_02_r")
    Parents="((Item=(Type=Bone,Name=\"spine_02\"),Weight=0.0002),(Item=(Type=Bone,Name=\"spine_03\"),Weight=0.0011),(Item=(Type=Bone,Name=\"spine_04\"),Weight=0.002),(Item=(Type=Bone,Name=\"upperarm_r\"),Weight=0.0045),(Item=(Type=Bone,Name=\"clavicle_l\"),Weight=0.0123),(Item=( ...(truncated)"
    Weight=@n2.Result
    ExecutePin=@SequenceAggregate.C
   )
   > &RigUnit_ParentConstraint(
    AdvancedSettings=(RotationOrderForFilter=XZY)
    Child=(Type=Bone,Name="dyn_hair_back_03_r")
    Parents="((Item=(Type=Bone,Name=\"spine_03\"),Weight=0.0003),(Item=(Type=Bone,Name=\"upperarm_twist_01_r\"),Weight=0.0004),(Item=(Type=Bone,Name=\"spine_04\"),Weight=0.0009),(Item=(Type=Bone,Name=\"upperarm_r\"),Weight=0.0176),(Item=(Type=Bone,Name=\"deform_upperarm_up_r\"), ...(truncated)"
    Weight=@n3.Result
   )
   > &RigUnit_ParentConstraint(
    AdvancedSettings=(RotationOrderForFilter=XZY)
    Child=(Type=Bone,Name="dyn_hair_back_04_r")
    Parents="((Item=(Type=Bone,Name=\"spine_02\"),Weight=0.0012),(Item=(Type=Bone,Name=\"spine_03\"),Weight=0.0059),(Item=(Type=Bone,Name=\"spine_04\"),Weight=0.0122),(Item=(Type=Bone,Name=\"upperarm_r\"),Weight=0.0383),(Item=(Type=Bone,Name=\"deform_upperarm_up_r\"),Weight=0.10 ...(truncated)"
    Weight=@n4.Result
   )
   > &RigUnit_ParentConstraint(
    AdvancedSettings=(RotationOrderForFilter=XZY)
    Child=(Type=Bone,Name="dyn_hair_back_05_r")
    Parents="((Item=(Type=Bone,Name=\"spine_01\"),Weight=0.0019),(Item=(Type=Bone,Name=\"spine_02\"),Weight=0.0028),(Item=(Type=Bone,Name=\"upperarm_twist_01_r\"),Weight=0.0117),(Item=(Type=Bone,Name=\"deform_upperarm_up_r\"),Weight=0.032),(Item=(Type=Bone,Name=\"spine_03\"),Wei ...(truncated)"
    Weight=@n4.Result
   )
   n5 = &RigUnit_SphericalPoseReader(
    ActiveRegionScaleFactors=(PositiveWidth=1,NegativeWidth=2,PositiveHeight=2,NegativeHeight=2)
    ActiveRegionSize=0.1
    Debug=(bDrawDebug=true,bDraw2D=false,bDrawLocalAxes=false,DebugScale=25,DebugSegments=20,DebugThickness=0.25)
    DriverAxis=(X=1,Y=0,Z=0)
    DriverItem=(Type=Bone,Name="clavicle_l")
    ExecutePin=()
    FalloffRegionScaleFactors=(PositiveWidth=1,NegativeWidth=0,PositiveHeight=0,NegativeHeight=0)
    FalloffSize=0.2
    OptionalParentItem=(Type=Bone)
    RotationOffset=(X=0,Y=-50,Z=0)
   )
   > &RigUnit_ModifyTransforms(
    ItemToModify="((Item=(Type=Bone,Name=\"dyn_hair_back_02_l\"),Transform=(Rotation=(X=0,Y=0.2588,Z=0,W=0.9659),Translation=(X=0,Y=-2,Z=4),Scale3D=(X=1,Y=1,Z=1))),(Item=(Type=Bone,Name=\"dyn_hair_back_03_l\"),Transform=(Rotation=(X=0,Y=0.0872,Z=0,W=0.9962),Translation=(X=0,Y=0 ...(truncated)"
    Weight=@n5.OutputParam
   )
   > n6 = &RigUnit_SphericalPoseReader(
    ActiveRegionScaleFactors=(PositiveWidth=1,NegativeWidth=2,PositiveHeight=2,NegativeHeight=2)
    ActiveRegionSize=0.1
    Debug=(bDrawDebug=false,bDraw2D=false,bDrawLocalAxes=false,DebugScale=25,DebugSegments=20,DebugThickness=0.25)
    DriverAxis=(X=-1,Y=0,Z=0)
    DriverItem=(Type=Bone,Name="clavicle_r")
    FalloffRegionScaleFactors=(PositiveWidth=1,NegativeWidth=0,PositiveHeight=0,NegativeHeight=0)
    FalloffSize=0.2
    OptionalParentItem=(Type=Bone)
    RotationOffset=(X=0,Y=130,Z=0)
   )
   > &RigUnit_ModifyTransforms(
    ItemToModify="((Item=(Type=Bone,Name=\"dyn_hair_back_02_r\"),Transform=(Rotation=(X=0,Y=-0.2588,Z=0,W=0.9659),Translation=(X=0,Y=-2,Z=-4),Scale3D=(X=1,Y=1,Z=1))),(Item=(Type=Bone,Name=\"dyn_hair_back_03_r\"),Transform=(Rotation=(X=0,Y=-0.061,Z=0,W=0.9981),Translation=(X=0,Y ...(truncated)"
    Weight=@n6.OutputParam
   )
  ]}
 SequenceAggregate_3_ContainedGraph {nodes [
   n1 = RigVMFunctionEntryNode()
   SequenceAggregate = &RigVMFunction_Sequence(ExecuteContext=@n1.ExecuteContext)
   SequenceAggregate_1 = &RigVMFunction_Sequence(ExecuteContext=@SequenceAggregate.B)
   SequenceAggregate_2 = &RigVMFunction_Sequence(ExecuteContext=@SequenceAggregate_1.B)
   SequenceAggregate_3 = &RigVMFunction_Sequence(ExecuteContext=@SequenceAggregate_2.B)
   SequenceAggregate_4 = &RigVMFunction_Sequence(ExecuteContext=@SequenceAggregate_3.B)
   RigVMFunctionReturnNode(
    A=@SequenceAggregate.A
    B=@SequenceAggregate_1.A
    C=@SequenceAggregate_2.A
    D=@SequenceAggregate_3.A
    E=@SequenceAggregate_4.A
    F=@SequenceAggregate_4.B
   )
  ]}
]
hierarchy [
 root {loc=<0 0 0>, rot=<0 0 0 1>}
 pelvis {parent=root, loc=<0 -0.5301 98.0571>, rot=<0.0127 0.707 -0.0127 -0.707>}
 spine_01 {parent=pelvis, loc=<6.1208 -0.9328 0>, rot=<0 0 0.034 -0.9994>}
 spine_02 {parent=spine_01, loc=<6.5997 -0.2095 0>, rot=<0 0 0.0238 -0.9997>}
 spine_03 {parent=spine_02, loc=<6.4914 -0.1237 0>, rot=<0 0 0.0149 -0.9999>}
 spine_04 {parent=spine_03, loc=<6.9598 0.0794 0>, rot=<0 0 -0.0132 -0.9999>}
 spine_05 {parent=spine_04, loc=<6.9987 0 0>, rot=<0 0 -0.0015 -1>}
 neck_01 {parent=spine_05, loc=<11.3138 0.1529 0>, rot=<0 0 -0.2003 -0.9797>}
 neck_02 {parent=neck_01, loc=<4.5627 0 0>, rot=<0 0 0.0504 -0.9987>}
 head {parent=neck_02, loc=<5.3914 0 0>, rot=<0 0 0.115 -0.9934>}
 Mstr_Back_Hair {parent=head, loc=<5.65 1.2676 0.2981>, rot=<0.5521 -0.4418 0.4418 -0.5521>}
 base_headwing_l {parent=Mstr_Back_Hair, loc=<-1.9648 6.4829 5.1567>, rot=<-0.7062 -0.0047 -0.4373 -0.5568>}
 dyn_headwing_l_01 {parent=base_headwing_l, loc=<0 0 0>, rot=<0 0 0 1>}
 dyn_headwing_l_02 {parent=dyn_headwing_l_01, loc=<6.9999 0 0>, rot=<0 0.0678 0 -0.9977>}
 dyn_headwing_l_03 {parent=dyn_headwing_l_02, loc=<7 0 0>, rot=<0 0 0 1>}
 base_headwing_r {parent=Mstr_Back_Hair, loc=<-1.9648 -5.8867 5.1567>, rot=<-0.4107 0.464 0.0887 -0.7799>}
 dyn_headwing_r_01 {parent=base_headwing_r, loc=<0 0 0>, rot=<-0.0003 -0.0224 -0.0115 -0.9997>}
 dyn_headwing_r_02 {parent=dyn_headwing_r_01, loc=<7 0 0>, rot=<-0.0178 -0.1192 0.0717 -0.9901>}
 dyn_headwing_r_03 {parent=dyn_headwing_r_02, loc=<7 0 0>, rot=<0 0 0 1>}
 MSTR_Top_hair {parent=Mstr_Back_Hair, loc=<-1.7028 3.1681 10.2985>, rot=<0.0499 0.4151 0.4857 -0.7676>}
 dyn_Top_Hair_1 {parent=MSTR_Top_hair, loc=<0 0 0>, rot=<0 0 0 1>}
 dyn_Top_Hair_2 {parent=dyn_Top_Hair_1, loc=<3.5 0 0>, rot=<0 -0.2219 0 -0.9751>}
 dyn_Top_Hair_3 {parent=dyn_Top_Hair_2, loc=<3.5 0 0>, rot=<0 0 0 1>}
 Mstr_FrntHair_A {parent=Mstr_Back_Hair, loc=<-2.7535 3.8794 8.659>, rot=<-0.0625 0.4313 0.8907 -0.1295>}
 dyn_FrntHair_A_1 {parent=Mstr_FrntHair_A, loc=<0 0 0>, rot=<0 -0.2362 0 -0.9717>}
 dyn_FrntHair_A_2 {parent=dyn_FrntHair_A_1, loc=<5 0 0>, rot=<0 -0.2526 0 -0.9676>}
 dyn_FrntHair_A_3 {parent=dyn_FrntHair_A_2, loc=<5 0 0>, rot=<0 -0.4146 0 -0.91>}
 dyn_FrntHair_A_4 {parent=dyn_FrntHair_A_3, loc=<5 0 0>, rot=<0 -0.1184 0 -0.993>}
 dyn_FrntHair_A_5 {parent=dyn_FrntHair_A_4, loc=<4.4136 0 0>, rot=<0 0 0 1>}
 Mstr_FrntHair_B {parent=Mstr_Back_Hair, loc=<-2.6611 4.6645 6.9498>, rot=<-0.0066 0.399 0.9105 0.1083>}
 dyn_FrntHair_B_1 {parent=Mstr_FrntHair_B, loc=<0 0 0>, rot=<0 -0.2362 0 -0.9717>}
 dyn_FrntHair_B_2 {parent=dyn_FrntHair_B_1, loc=<5 0 0>, rot=<0 -0.3474 0 -0.9377>}
 dyn_FrntHair_B_3 {parent=dyn_FrntHair_B_2, loc=<5 0 0>, rot=<0 -0.3113 0 -0.9503>}
 Mstr_FrntHair_C {parent=Mstr_Back_Hair, loc=<-2.1588 5.1097 6.7446>, rot=<-0.2305 0.1149 0.8608 0.4389>}
 dyn_FrntHair_C_1 {parent=Mstr_FrntHair_C, loc=<0 0 0>, rot=<0 -0.2362 0 -0.9717>}
 dyn_FrntHair_C_2 {parent=dyn_FrntHair_C_1, loc=<5 0 0>, rot=<0 -0.2411 0 -0.9705>}
 dyn_FrntHair_A_B_1 {parent=Mstr_Back_Hair, loc=<0.1571 -8.6118 1.3803>, rot=<-0.6476 0.0736 0.3165 -0.6892>}
 dyn_FrntHair_A_B_2 {parent=dyn_FrntHair_A_B_1, loc=<4.4136 0 0>, rot=<0 0 -0.1317 -0.9913>}
 base_hair_back_l {parent=Mstr_Back_Hair, loc=<3.1798 6.1524 2.7286>, rot=<-0.547 0.3936 0.5653 0.4757>}
 dyn_hair_back_01_l {parent=base_hair_back_l, loc=<0 0 0>, rot=<0 0 0 1>}
 dyn_hair_back_02_l {parent=dyn_hair_back_01_l, loc=<8.9243 0.1654 0>, rot=<0 0 0.1022 -0.9948>}
 dyn_hair_back_03_l {parent=dyn_hair_back_02_l, loc=<6.4897 0.0983 0>, rot=<0 0 0.126 -0.992>}
 dyn_hair_back_04_l {parent=dyn_hair_back_03_l, loc=<7.6353 -0.1135 0>, rot=<0 0 -0.1976 -0.9803>}
 dyn_hair_back_05_l {parent=dyn_hair_back_04_l, loc=<7.0615 0.0597 0>, rot=<0 0 -0.1412 -0.99>}
 dyn_hair_back_06_l {parent=dyn_hair_back_05_l, loc=<7.3651 -0.0072 0>, rot=<0 0 -0.025 -0.9997>}
 dyn_Back_Hair_A_A {parent=dyn_hair_back_01_l, loc=<8.636 -2.2026 2.6252>, rot=<0.2181 -0.7855 -0.396 0.4225>}
 base_hair_back_m {parent=Mstr_Back_Hair, loc=<6.7016 0.2985 1.6228>, rot=<0.523 -0.4759 -0.4759 -0.523>}
 dyn_hair_back_01_m {parent=base_hair_back_m, loc=<0 0 0>, rot=<0 0 0 1>}
 dyn_hair_back_02_m {parent=dyn_hair_back_01_m, loc=<8.0943 -0.0648 0>, rot=<0 0 0.0109 -0.9999>}
 dyn_hair_back_03_m {parent=dyn_hair_back_02_m, loc=<6.0888 -0.0022 0>, rot=<0 0 0.0056 -1>}
 dyn_hair_back_04_m {parent=dyn_hair_back_03_m, loc=<6.5783 0 0>, rot=<0 0 -0.0259 -0.9997>}
 dyn_hair_back_05_m {parent=dyn_hair_back_04_m, loc=<6.9646 0 0>, rot=<0 0 -0.1294 -0.9916>}
 dyn_hair_back_06_m {parent=dyn_hair_back_05_m, loc=<7.3535 0.0097 0>, rot=<0 0 -0.0235 -0.9997>}
 base_hair_back_r {parent=Mstr_Back_Hair, loc=<3.1798 -5.5562 2.7286>, rot=<0.5422 -0.5076 -0.4656 -0.4812>}
 dyn_hair_back_01_r {parent=base_hair_back_r, loc=<0 0 0>, rot=<0 0 0 1>}
 dyn_hair_back_02_r {parent=dyn_hair_back_01_r, loc=<8.9243 0.1654 0>, rot=<0 0 0.1022 -0.9948>}
 dyn_hair_back_03_r {parent=dyn_hair_back_02_r, loc=<6.4894 0.0982 0>, rot=<0 0 0.126 -0.992>}
 dyn_hair_back_04_r {parent=dyn_hair_back_03_r, loc=<7.6161 -0.0997 0>, rot=<0 0 -0.2001 -0.9798>}
 dyn_hair_back_05_r {parent=dyn_hair_back_04_r, loc=<7.0615 0.0597 0>, rot=<0 0 -0.1412 -0.99>}
 dyn_hair_back_06_r {parent=dyn_hair_back_05_r, loc=<7.3651 -0.0072 0>, rot=<0 0 -0.025 -0.9997>}
 dyn_Back_Hair_C_A {parent=dyn_hair_back_01_r, loc=<9.1571 -2.6029 -4.9278>, rot=<0.7089 -0.1268 -0.6886 0.0857>}
 Mstr_earring_l {parent=head, loc=<0.0241 -0.0112 -6.0776>, rot=<0.0039 -0.9374 0.3482 -0.0014>}
 dyn_earring_l {parent=Mstr_earring_l, loc=<0 0 0>, rot=<0 0 0 1>}
 Mstr_earring_r {parent=head, loc=<0.115 -0.0241 6.0468>, rot=<0.0015 -0.3532 0.9355 -0.0039>}
 dyn_earring_r {parent=Mstr_earring_r, loc=<0 0 0>, rot=<0 0 0 1>}
 FACIAL_C_FacialRoot {parent=head, loc=<1.6807 0.7414 0>, rot=<0.5021 -0.4979 -0.5021 -0.4979>}
 FACIAL_C_Forehead {parent=FACIAL_C_FacialRoot, loc=<0 -3.9903 7.7126>, rot=<-0.0436 0 0 -0.999>}
 FACIAL_L_ForeheadIn {parent=FACIAL_C_FacialRoot, loc=<1.9396 -3.5959 7.517>, rot=<-0.0609 -0.0696 0.0043 -0.9957>}
 FACIAL_R_ForeheadIn {parent=FACIAL_C_FacialRoot, loc=<-1.9396 -3.5959 7.517>, rot=<-0.0609 0.0696 -0.0043 -0.9957>}
 FACIAL_L_ForeheadMid {parent=FACIAL_C_FacialRoot, loc=<3.7235 -3.7282 7.0097>, rot=<-0.0712 -0.1757 -0.0206 -0.9816>}
 FACIAL_R_ForeheadMid {parent=FACIAL_C_FacialRoot, loc=<-3.7235 -3.7282 7.0097>, rot=<-0.0712 0.1757 0.0206 -0.9816>}
 FACIAL_L_ForeheadOut {parent=FACIAL_C_FacialRoot, loc=<5.0326 -3.5235 5.9149>, rot=<-0.1215 -0.2557 -0.0792 -0.9558>}
 FACIAL_R_ForeheadOut {parent=FACIAL_C_FacialRoot, loc=<-5.0326 -3.5235 5.9149>, rot=<-0.1215 0.2557 0.0792 -0.9558>}
 FACIAL_L_EyesackUpper {parent=FACIAL_C_FacialRoot, loc=<3.4393 -2.7044 7.4345>, rot=<0.0487 -0.0843 -0.0646 -0.9932>}
 FACIAL_R_EyesackUpper {parent=FACIAL_C_FacialRoot, loc=<-3.4393 -2.7044 7.4345>, rot=<0.0487 0.0843 0.0646 -0.9932>}
 FACIAL_L_EyelidUpperFurrow {parent=FACIAL_C_FacialRoot, loc=<3.367 -2.5164 7.2946>, rot=<-0.0301 -0.1509 -0.0797 -0.9849>}
 FACIAL_R_EyelidUpperFurrow {parent=FACIAL_C_FacialRoot, loc=<-3.367 -2.5164 7.2946>, rot=<-0.0301 0.1509 0.0797 -0.9849>}
 FACIAL_L_EyelidUpperA {parent=FACIAL_C_FacialRoot, loc=<2.9845 -1.3704 4.9427>, rot=<-0.1478 0 0 -0.989>}
 FACIAL_R_EyelidUpperA {parent=FACIAL_C_FacialRoot, loc=<-2.9845 -1.3704 4.9427>, rot=<-0.1478 0 0 -0.989>}
 FACIAL_L_Eye {parent=FACIAL_C_FacialRoot, loc=<2.9845 -1.3704 4.9427>, rot=<0 0 0 1>}
 FACIAL_R_Eye {parent=FACIAL_C_FacialRoot, loc=<-2.9845 -1.3704 4.9427>, rot=<0 0 0 1>}
 FACIAL_L_EyelidLowerA {parent=FACIAL_C_FacialRoot, loc=<2.9845 -1.3704 4.9427>, rot=<0.284 0 0 -0.9588>}
 FACIAL_R_EyelidLowerA {parent=FACIAL_C_FacialRoot, loc=<-2.9845 -1.3704 4.9427>, rot=<0.284 0 0 -0.9588>}
 FACIAL_L_EyeCornerInner {parent=FACIAL_C_FacialRoot, loc=<1.7124 -1.0246 6.325>, rot=<0 0 0 -1>}
 FACIAL_R_EyeCornerInner {parent=FACIAL_C_FacialRoot, loc=<-1.7124 -1.0246 6.325>, rot=<0 0 0 -1>}
 FACIAL_L_EyeCornerOuter {parent=FACIAL_C_FacialRoot, loc=<4.946 -1.2111 5.3424>, rot=<0 -0.3827 0 -0.9239>}
 FACIAL_R_EyeCornerOuter {parent=FACIAL_C_FacialRoot, loc=<-4.946 -1.2111 5.3424>, rot=<0 0.3827 0 -0.9239>}
 FACIAL_L_EyesackLower {parent=FACIAL_C_FacialRoot, loc=<3.7533 0.7783 5.8435>, rot=<-0.0223 -0.1398 -0.0223 -0.9897>}
 FACIAL_R_EyesackLower {parent=FACIAL_C_FacialRoot, loc=<-3.7533 0.7783 5.8435>, rot=<-0.0223 0.1398 0.0223 -0.9897>}
 FACIAL_L_CheekInner {parent=FACIAL_C_FacialRoot, loc=<2.1661 0.6933 6.4268>, rot=<-0.1233 -0.088 -0.1972 -0.9686>}
 FACIAL_R_CheekInner {parent=FACIAL_C_FacialRoot, loc=<-2.1661 0.6933 6.4268>, rot=<-0.1233 0.088 0.1972 -0.9686>}
 FACIAL_L_CheekOuter {parent=FACIAL_C_FacialRoot, loc=<4.1266 0.9522 4.8125>, rot=<0.0551 -0.377 0.1003 -0.9191>}
 FACIAL_R_CheekOuter {parent=FACIAL_C_FacialRoot, loc=<-4.1266 0.9522 4.8125>, rot=<0.0551 0.377 -0.1003 -0.9191>}
 FACIAL_L_NasolabialBulge {parent=FACIAL_C_FacialRoot, loc=<3.2396 3.8442 4.9777>, rot=<0.0579 -0.3774 0.0238 -0.9239>}
 FACIAL_R_NasolabialBulge {parent=FACIAL_C_FacialRoot, loc=<-3.2396 3.8442 4.9777>, rot=<0.0579 0.3774 -0.0238 -0.9239>}
 FACIAL_L_NasolabialFurrow {parent=FACIAL_C_FacialRoot, loc=<2.784 2.831 7.3103>, rot=<-0.0333 -0.1258 -0.3638 -0.9223>}
 FACIAL_R_NasolabialFurrow {parent=FACIAL_C_FacialRoot, loc=<-2.784 2.831 7.3103>, rot=<-0.0333 0.1258 0.3638 -0.9223>}
 FACIAL_L_CheekLower {parent=FACIAL_C_FacialRoot, loc=<2.8233 4.2672 4.9245>, rot=<0.0345 -0.427 0.0577 -0.9018>}
 FACIAL_R_CheekLower {parent=FACIAL_C_FacialRoot, loc=<-2.8233 4.2672 4.9245>, rot=<0.0345 0.427 -0.0577 -0.9018>}
 FACIAL_C_Nose {parent=FACIAL_C_FacialRoot, loc=<0 -0.2121 7.8049>, rot=<-0.1305 0 0 -0.9914>}
 FACIAL_C_NoseLower {parent=FACIAL_C_Nose, loc=<0 3.1181 -0.2503>, rot=<0.1305 0 0 -0.9914>}
 FACIAL_L_Nostril {parent=FACIAL_C_Nose, loc=<0.8863 2.4797 -0.2362>, rot=<0.0173 0.1305 0.0023 -0.9913>}
 FACIAL_R_Nostril {parent=FACIAL_C_Nose, loc=<-0.8863 2.4797 -0.2362>, rot=<0.0173 -0.1305 -0.0023 -0.9913>}
 FACIAL_C_NoseTip {parent=FACIAL_C_Nose, loc=<0 2.2789 0.4055>, rot=<0.0872 0 0 -0.9962>}
 FACIAL_C_MouthUpper {parent=FACIAL_C_FacialRoot, loc=<0 4.0377 6.1483>, rot=<0 0 0 -1>}
 FACIAL_C_LipUpper {parent=FACIAL_C_MouthUpper, loc=<0 -0.019 1.7106>, rot=<-0.0872 0 0 -0.9962>}
 FACIAL_L_LipUpper {parent=FACIAL_C_MouthUpper, loc=<0.7947 -0.0208 1.5729>, rot=<-0.0778 -0.1601 -0.0379 -0.9833>}
 FACIAL_R_LipUpper {parent=FACIAL_C_MouthUpper, loc=<-0.7947 -0.0208 1.5729>, rot=<-0.0778 0.1601 0.0379 -0.9833>}
 FACIAL_L_LipUpperOuter {parent=FACIAL_C_MouthUpper, loc=<1.4616 0.0479 1.132>, rot=<-0.0983 -0.2294 -0.0983 -0.9634>}
 FACIAL_R_LipUpperOuter {parent=FACIAL_C_MouthUpper, loc=<-1.4616 0.0479 1.132>, rot=<-0.0983 0.2294 0.0983 -0.9634>}
 FACIAL_L_LipCorner {parent=FACIAL_C_MouthUpper, loc=<2.072 0.4528 0.4869>, rot=<0 -0.2079 0 -0.9781>}
 FACIAL_R_LipCorner {parent=FACIAL_C_MouthUpper, loc=<-2.072 0.4528 0.4869>, rot=<0 0.2079 0 -0.9781>}
 FACIAL_L_Masseter {parent=FACIAL_C_FacialRoot, loc=<6.337 1.5441 2.6867>, rot=<0.0693 0.1129 0.0079 -0.9912>}
 FACIAL_R_Masseter {parent=FACIAL_C_FacialRoot, loc=<-6.337 1.5441 2.6867>, rot=<0.0693 -0.1129 -0.0079 -0.9912>}
 FACIAL_C_UnderChin {parent=FACIAL_C_FacialRoot, loc=<0 7.5205 2.5434>, rot=<0.5 0 0 -0.866>}
 FACIAL_L_UnderChin {parent=FACIAL_C_FacialRoot, loc=<2.2177 6.5131 1.6168>, rot=<0.4025 -0.355 0.0823 -0.8398>}
 FACIAL_R_UnderChin {parent=FACIAL_C_FacialRoot, loc=<-2.2177 6.5131 1.6168>, rot=<0.4025 0.355 -0.0823 -0.8398>}
 FACIAL_C_TeethUpper {parent=FACIAL_C_FacialRoot, loc=<0 4.628 6.7264>, rot=<0 0 0 1>}
 FACIAL_C_LowerLipRotation {parent=FACIAL_C_FacialRoot, loc=<0 1.8437 0.9134>, rot=<0.342 0 0 -0.9397>}
 FACIAL_C_MouthLower {parent=FACIAL_C_LowerLipRotation, loc=<0 -1.6842 5.4205>, rot=<-0.342 0 0 -0.9397>}
 FACIAL_C_LipLower {parent=FACIAL_C_MouthLower, loc=<0 0.9045 1.5246>, rot=<0.0872 0 0 -0.9962>}
 FACIAL_L_LipLower {parent=FACIAL_C_MouthLower, loc=<0.688 0.8781 1.3608>, rot=<0.0927 -0.1295 0.0273 -0.9869>}
 FACIAL_R_LipLower {parent=FACIAL_C_MouthLower, loc=<-0.688 0.8781 1.3608>, rot=<0.0927 0.1295 -0.0273 -0.9869>}
 FACIAL_L_LipLowerOuter {parent=FACIAL_C_MouthLower, loc=<1.343 0.8111 0.9872>, rot=<0.0913 -0.2322 0.0318 -0.9679>}
 FACIAL_R_LipLowerOuter {parent=FACIAL_C_MouthLower, loc=<-1.343 0.8111 0.9872>, rot=<0.0913 0.2322 -0.0318 -0.9679>}
 FACIAL_C_TeethLower {parent=FACIAL_C_LowerLipRotation, loc=<0 -1.5397 6.1913>, rot=<-0.342 0 0 -0.9397>}
 FACIAL_C_Tongue1 {parent=FACIAL_C_TeethLower, loc=<0 0.3587 -4.85>, rot=<-0.2588 0 0 -0.9659>}
 FACIAL_C_Tongue2 {parent=FACIAL_C_Tongue1, loc=<0 0.3451 1.4312>, rot=<0.2588 0 0 -0.9659>}
 FACIAL_C_Tongue3 {parent=FACIAL_C_Tongue2, loc=<0 0.1493 1.2843>, rot=<0.1305 0 0 -0.9914>}
 FACIAL_C_Tongue4 {parent=FACIAL_C_Tongue3, loc=<0.0006 -0.2194 1.0685>, rot=<0.0436 0 0 -0.999>}
 FACIAL_C_TongueUpper3 {parent=FACIAL_C_Tongue3, loc=<0 -0.2941 -0.2398>, rot=<0 0 0 1>}
 FACIAL_C_TongueLower3 {parent=FACIAL_C_Tongue3, loc=<0 0.8294 0.0462>, rot=<0 0 0 1>}
 FACIAL_L_TongueSide3 {parent=FACIAL_C_Tongue3, loc=<1.4941 -0.0273 0.0614>, rot=<0 0 0 1>}
 FACIAL_R_TongueSide3 {parent=FACIAL_C_Tongue3, loc=<-1.4942 -0.0273 0.0614>, rot=<0 0 0 1>}
 FACIAL_C_TongueUpper2 {parent=FACIAL_C_Tongue2, loc=<0 -0.5618 -0.3851>, rot=<0 0 0 1>}
 FACIAL_L_TongueSide2 {parent=FACIAL_C_Tongue2, loc=<1.7814 -0.0872 -0.0014>, rot=<0 0 0 1>}
 FACIAL_R_TongueSide2 {parent=FACIAL_C_Tongue2, loc=<-1.7814 -0.0872 -0.0014>, rot=<0 0 0 1>}
 FACIAL_C_Jaw {parent=FACIAL_C_FacialRoot, loc=<0 1.8437 0.9134>, rot=<0.342 0 0 -0.9397>}
 FACIAL_C_Jawline {parent=FACIAL_C_Jaw, loc=<0 2.0889 6.5401>, rot=<0.2588 0 0 -0.9659>}
 FACIAL_L_ChinSide {parent=FACIAL_C_Jaw, loc=<2.3517 -0.0838 6.6099>, rot=<-0.2207 -0.3739 0.05 -0.8994>}
 FACIAL_R_ChinSide {parent=FACIAL_C_Jaw, loc=<-2.3517 -0.0838 6.6099>, rot=<-0.2207 0.3739 -0.05 -0.8994>}
 FACIAL_C_Chin {parent=FACIAL_C_Jaw, loc=<0 0.3234 6.5331>, rot=<-0.2588 0 0 -0.9659>}
 FACIAL_L_Ear {parent=FACIAL_C_FacialRoot, loc=<6.6374 -0.3424 -0.6787>, rot=<-0.0262 -0.0007 -0.0262 -0.9993>}
 FACIAL_R_Ear {parent=FACIAL_C_FacialRoot, loc=<-6.6374 -0.3424 -0.6787>, rot=<-0.0262 0.0007 0.0262 -0.9993>}
 FACIAL_L_Jawline {parent=FACIAL_C_FacialRoot, loc=<4.1303 4.9936 3.5752>, rot=<0.2092 0.1578 -0.0547 -0.9635>}
 FACIAL_R_Jawline {parent=FACIAL_C_FacialRoot, loc=<-4.1303 4.9936 3.5752>, rot=<0.2092 -0.1578 0.0547 -0.9635>}
 FACIAL_C_NoseBridge {parent=FACIAL_C_FacialRoot, loc=<0 -2.2966 8.7077>, rot=<0.0436 0 0 -0.999>}
 FACIAL_L_EyelidUpperB {parent=FACIAL_C_FacialRoot, loc=<2.9845 -1.3704 4.9427>, rot=<-0.225 0 0 -0.9744>}
 FACIAL_L_EyelidLowerB {parent=FACIAL_C_FacialRoot, loc=<2.9845 -1.3704 4.9427>, rot=<0.3173 0 0 -0.9483>}
 FACIAL_R_EyelidLowerB {parent=FACIAL_C_FacialRoot, loc=<-2.9845 -1.3704 4.9427>, rot=<0.3256 0 0 -0.9455>}
 FACIAL_R_EyelidUpperB {parent=FACIAL_C_FacialRoot, loc=<-2.9845 -1.3704 4.9427>, rot=<-0.225 0 0 -0.9744>}
 head_null {parent=neck_02, type=Null, loc=<5.3914 0 0>, rot=<0 0 0.115 -0.9934>}
 clavicle_l {parent=spine_05, loc=<5.4831 -0.1053 -3.2946>, rot=<0.0465 -0.7289 -0.009 -0.6829>}
 upperarm_l {parent=clavicle_l, loc=<9.8658 -1.2136 -0.4105>, rot=<0.0164 -0.3245 -0.0269 -0.9454>}
 lowerarm_l {parent=upperarm_l, loc=<26.465 -0.8345 0.259>, rot=<0.0036 0.0161 -0.2298 -0.9731>}
 hand_l {parent=lowerarm_l, loc=<22.7355 -0.3203 -0.6105>, rot=<-0.6784 -0.0223 0.0735 -0.7307>}
 thumb_01_l {parent=hand_l, loc=<1.3118 -0.3191 -2.459>, rot=<0.6546 -0.4042 0.0664 -0.6354>}
 thumb_02_l {parent=thumb_01_l, loc=<5.9336 -0.0448 0.0401>, rot=<-0.0037 0.0213 0.298 -0.9543>}
 thumb_03_l {parent=thumb_02_l, loc=<2.2223 0.3682 -0.0907>, rot=<0.0118 -0.013 0.0066 -0.9998>}
 index_metacarpal_l {parent=hand_l, loc=<1.6535 0.0183 -2.96>, rot=<0.0158 -0.0124 -0.0173 -0.9996>}
 index_01_l {parent=index_metacarpal_l, loc=<8.3008 0.0357 -0.0524>, rot=<-0.0044 0.0485 0.0924 -0.9945>}
 index_02_l {parent=index_01_l, loc=<3.3587 -0.3004 0.0452>, rot=<-0.002 0.0027 0.3225 -0.9466>}
 index_03_l {parent=index_02_l, loc=<2.5578 -0.123 -0.0252>, rot=<-0.002 0.0067 0.2558 -0.9667>}
 middle_metacarpal_l {parent=hand_l, loc=<2.1856 -0.1725 -0.5984>, rot=<0.006 0.0168 -0.0329 -0.9993>}
 middle_01_l {parent=middle_metacarpal_l, loc=<7.3275 -0.2447 0.0129>, rot=<-0.0208 0.0042 0.0858 -0.9961>}
 middle_02_l {parent=middle_01_l, loc=<4.0123 0.1401 0.1175>, rot=<-0.0218 0.0258 0.3512 -0.9357>}
 middle_03_l {parent=middle_02_l, loc=<2.6077 0.0534 0.1373>, rot=<0.0082 0.0188 0.1708 -0.9851>}
 ring_metacarpal_l {parent=hand_l, loc=<2.2074 -0.3625 1.3638>, rot=<-0.0284 0.0722 -0.0458 -0.9959>}
 ring_01_l {parent=ring_metacarpal_l, loc=<7.0295 -0.6197 -0.1446>, rot=<-0.0489 -0.0252 0.1892 -0.9804>}
 ring_02_l {parent=ring_01_l, loc=<3.6174 0.4647 -0.1027>, rot=<0 0 0.2806 -0.9598>}
 ring_03_l {parent=ring_02_l, loc=<2.4152 0.114 0.0199>, rot=<0 0 0.1189 -0.9929>}
 pinky_metacarpal_l {parent=hand_l, loc=<1.7616 -0.9753 3.3814>, rot=<-0.0397 0.0814 -0.069 -0.9935>}
 pinky_01_l {parent=pinky_metacarpal_l, loc=<7.1709 -0.4831 -0.1094>, rot=<-0.0654 -0.0035 0.2242 -0.9723>}
 pinky_02_l {parent=pinky_01_l, loc=<2.8274 0.0893 -0.1043>, rot=<0.0037 0.0011 0.2854 -0.9584>}
 pinky_03_l {parent=pinky_02_l, loc=<2.2046 0.0307 -0.0078>, rot=<-0.0002 0.0098 0.1445 -0.9895>}
 lowerarm_twist_02_l {parent=lowerarm_l, loc=<10.6707 0 0>, rot=<0 0 0 1>}
 lowerarm_twist_01_l {parent=lowerarm_l, loc=<18.316 0 0>, rot=<0 0 0 1>}
 upperarm_twist_01_l {parent=upperarm_l, loc=<7.2135 -0.1134 0>, rot=<0 0 0 1>}
 upperarm_twist_02_l {parent=upperarm_l, loc=<19.7998 -0.1134 0>, rot=<0 0 0 1>}
 clavicle_r {parent=spine_05, loc=<5.4831 -0.1053 3.2946>, rot=<0.7289 0.0465 -0.6829 0.009>}
 upperarm_r {parent=clavicle_r, loc=<-9.8658 1.2136 0.4105>, rot=<0.0164 -0.3245 -0.0269 -0.9454>}
 lowerarm_r {parent=upperarm_r, loc=<-26.465 0.8345 -0.259>, rot=<0.0036 0.0161 -0.2298 -0.9731>}
 hand_r {parent=lowerarm_r, loc=<-22.7355 0.3203 0.6106>, rot=<-0.6784 -0.0223 0.0735 -0.7307>}
 thumb_01_r {parent=hand_r, loc=<-1.3118 0.3191 2.459>, rot=<0.6546 -0.4042 0.0664 -0.6354>}
 thumb_02_r {parent=thumb_01_r, loc=<-5.9336 0.0448 -0.0401>, rot=<-0.0037 0.0213 0.298 -0.9543>}
 thumb_03_r {parent=thumb_02_r, loc=<-2.2223 -0.3682 0.0907>, rot=<0.0118 -0.013 0.0066 -0.9998>}
 index_metacarpal_r {parent=hand_r, loc=<-1.6535 -0.0183 2.96>, rot=<0.0158 -0.0124 -0.0173 -0.9996>}
 index_01_r {parent=index_metacarpal_r, loc=<-8.3008 -0.0357 0.0524>, rot=<-0.0044 0.0485 0.0924 -0.9945>}
 index_02_r {parent=index_01_r, loc=<-3.3587 0.3004 -0.0452>, rot=<-0.002 0.0027 0.3225 -0.9466>}
 index_03_r {parent=index_02_r, loc=<-2.5578 0.123 0.0252>, rot=<-0.002 0.0067 0.2558 -0.9667>}
 middle_metacarpal_r {parent=hand_r, loc=<-2.1856 0.1725 0.5984>, rot=<0.006 0.0168 -0.0329 -0.9993>}
 middle_01_r {parent=middle_metacarpal_r, loc=<-7.3275 0.2447 -0.0129>, rot=<-0.0208 0.0042 0.0858 -0.9961>}
 middle_02_r {parent=middle_01_r, loc=<-4.0123 -0.1401 -0.1175>, rot=<-0.0218 0.0258 0.3512 -0.9357>}
 middle_03_r {parent=middle_02_r, loc=<-2.6077 -0.0534 -0.1373>, rot=<0.0082 0.0188 0.1708 -0.9851>}
 ring_metacarpal_r {parent=hand_r, loc=<-2.2074 0.3625 -1.3638>, rot=<-0.0284 0.0722 -0.0458 -0.9959>}
 ring_01_r {parent=ring_metacarpal_r, loc=<-7.0295 0.6197 0.1446>, rot=<-0.0489 -0.0252 0.1892 -0.9804>}
 ring_02_r {parent=ring_01_r, loc=<-3.6174 -0.4647 0.1027>, rot=<0 0 0.2806 -0.9598>}
 ring_03_r {parent=ring_02_r, loc=<-2.4152 -0.114 -0.0199>, rot=<0 0 0.1189 -0.9929>}
 pinky_metacarpal_r {parent=hand_r, loc=<-1.7616 0.9753 -3.3814>, rot=<-0.0397 0.0814 -0.069 -0.9935>}
 pinky_01_r {parent=pinky_metacarpal_r, loc=<-7.1709 0.4831 0.1094>, rot=<-0.0654 -0.0035 0.2242 -0.9723>}
 pinky_02_r {parent=pinky_01_r, loc=<-2.8274 -0.0893 0.1043>, rot=<0.0037 0.0011 0.2854 -0.9584>}
 pinky_03_r {parent=pinky_02_r, loc=<-2.2046 -0.0307 0.0078>, rot=<-0.0002 0.0098 0.1445 -0.9895>}
 lowerarm_twist_02_r {parent=lowerarm_r, loc=<-10.6707 0 0>, rot=<0 0 0 -1>}
 lowerarm_twist_01_r {parent=lowerarm_r, loc=<-18.316 0 0>, rot=<0 0 0 -1>}
 upperarm_twist_01_r {parent=upperarm_r, loc=<-7.2135 0.1134 0>, rot=<0 0 0 -1>}
 upperarm_twist_02_r {parent=upperarm_r, loc=<-19.7998 0.1134 0>, rot=<0 0 0 -1>}
 aim_const {parent=spine_05, loc=<0 0 0>, rot=<0 0 0 1>}
 thigh_l {parent=pelvis, loc=<-1.0855 -0.2109 -8.7053>, rot=<0.0009 0.0355 0.0155 -0.9993>}
 calf_l {parent=thigh_l, loc=<-42.5284 -0.0019 -0.7297>, rot=<0.0006 -0.0161 -0.0391 -0.9991>}
 foot_l {parent=calf_l, loc=<-46.7401 -2.5014 -0.592>, rot=<0.0427 -0.0238 0.0353 -0.9982>}
 ball_l {parent=foot_l, loc=<-3.5658 10.8066 -0.1544>, rot=<0.0001 -0.0001 -0.7063 -0.7079>}
 calf_twist_02_l {parent=calf_l, loc=<-19.1546 -1.7943 -0.2638>, rot=<0.0237 0.0095 -0.0035 -0.9997>}
 calf_twist_01_l {parent=calf_l, loc=<-35.8483 -1.9197 -0.5794>, rot=<0.0237 0.0095 -0.0035 -0.9997>}
 base_shoe_wing_l {parent=calf_l, loc=<-37.5378 -1.9789 -4.6473>, rot=<0.9332 -0.1703 -0.0916 -0.3029>}
 dyn_shoe_wing_l_01 {parent=base_shoe_wing_l, loc=<0 0 0>, rot=<0 0 0 1>}
 dyn_shoe_wing_l_02 {parent=dyn_shoe_wing_l_01, loc=<5.0889 0 0>, rot=<0 0 0 1>}
 thigh_twist_01_l {parent=thigh_l, loc=<-27.952 0 0>, rot=<0.0013 0 0 -1>}
 thigh_r {parent=pelvis, loc=<-1.0855 -0.2109 8.7053>, rot=<0.0355 -0.0009 0.9993 0.0155>}
 calf_r {parent=thigh_r, loc=<42.5284 0.0019 0.7297>, rot=<0.0006 -0.0161 -0.0391 -0.9991>}
 foot_r {parent=calf_r, loc=<46.7401 2.5014 0.592>, rot=<0.0427 -0.0238 0.0353 -0.9982>}
 ball_r {parent=foot_r, loc=<3.5658 -10.8066 0.1544>, rot=<0.0001 -0.0001 -0.7063 -0.7079>}
 calf_twist_02_r {parent=calf_r, loc=<19.1546 1.7943 0.2638>, rot=<0.0237 0.0095 -0.0035 -0.9997>}
 calf_twist_01_r {parent=calf_r, loc=<35.8483 1.9197 0.5794>, rot=<0.0237 0.0095 -0.0035 -0.9997>}
 base_shoe_wing_r {parent=calf_r, loc=<37.5378 1.9788 4.6473>, rot=<0.0916 -0.3029 0.9332 0.1703>}
 dyn_shoe_wing_r_01 {parent=base_shoe_wing_r, loc=<0 0 0>, rot=<0 0 0 1>}
 dyn_shoe_wing_r_02 {parent=dyn_shoe_wing_r_01, loc=<5.0889 0 0>, rot=<0 0 0 1>}
 thigh_twist_01_r {parent=thigh_r, loc=<27.952 0 0>, rot=<0.0013 0 0 -1>}
 Mstr_Skirt_Frnt_Mid {parent=pelvis, loc=<4.5583 8.0352 0>, rot=<0.0413 0.7059 0.7059 0.0413>}
 dyn_Skirt_Frnt_Mid_1 {parent=Mstr_Skirt_Frnt_Mid, loc=<0 0 0>, rot=<0 0 0 1>}
 dyn_Skirt_Frnt_Mid_2 {parent=dyn_Skirt_Frnt_Mid_1, loc=<5 0 0>, rot=<0 0 0 1>}
 dyn_Skirt_Frnt_Mid_3 {parent=dyn_Skirt_Frnt_Mid_2, loc=<10 0 0>, rot=<0 0 0 1>}
 dyn_Skirt_Frnt_Mid_4 {parent=dyn_Skirt_Frnt_Mid_3, loc=<10 0 0>, rot=<0 0 0 1>}
 dyn_Skirt_Frnt_Mid_5 {parent=dyn_Skirt_Frnt_Mid_4, loc=<10 0 0>, rot=<0 0 0 1>}
 Mstr_Skirt_Frnt_l {parent=pelvis, loc=<4.444 4.5271 -8.8579>, rot=<-0.0227 -0.8844 -0.4551 -0.1009>}
 dyn_Skirt_Frnt_l_1 {parent=Mstr_Skirt_Frnt_l, loc=<0 0 0>, rot=<0 0 0 1>}
 dyn_Skirt_Frnt_l_2 {parent=dyn_Skirt_Frnt_l_1, loc=<5 0 0>, rot=<0 0 0 1>}
 dyn_Skirt_Frnt_l_3 {parent=dyn_Skirt_Frnt_l_2, loc=<10 0 0>, rot=<0 0 0 1>}
 dyn_Skirt_Frnt_l_4 {parent=dyn_Skirt_Frnt_l_3, loc=<10 0 0>, rot=<0 0 0 1>}
 dyn_Skirt_Frnt_l_5 {parent=dyn_Skirt_Frnt_l_4, loc=<10 0 0>, rot=<0 0 0 1>}
 Mstr_Skirt_Side_l {parent=pelvis, loc=<4.1987 -1.9972 -13.4728>, rot=<-0.0073 -0.9891 0.0886 -0.1171>}
 dyn_Skirt_Side_l_1 {parent=Mstr_Skirt_Side_l, loc=<0 0 0>, rot=<0 0 0 1>}
 dyn_Skirt_Side_l_2 {parent=dyn_Skirt_Side_l_1, loc=<5 0 0>, rot=<0 0 0 1>}
 dyn_Skirt_Side_l_3 {parent=dyn_Skirt_Side_l_2, loc=<10 0 0>, rot=<0 0 0 1>}
 dyn_Skirt_Side_l_4 {parent=dyn_Skirt_Side_l_3, loc=<10 0 0>, rot=<0 0 0 1>}
 dyn_Skirt_Side_l_5 {parent=dyn_Skirt_Side_l_4, loc=<10 0 0>, rot=<0 0 0 1>}
 Mstr_Skirt_Bck_l {parent=pelvis, loc=<3.9552 -8.7914 -8.7937>, rot=<0.023 -0.903 0.4222 -0.0765>}
 dyn_Skirt_Bck_l_1 {parent=Mstr_Skirt_Bck_l, loc=<0 0 0>, rot=<0 0 0 1>}
 dyn_Skirt_Bck_l_2 {parent=dyn_Skirt_Bck_l_1, loc=<5 0 0>, rot=<0 0 0 1>}
 dyn_Skirt_Bck_l_3 {parent=dyn_Skirt_Bck_l_2, loc=<10 0 0>, rot=<0 0 0 1>}
 dyn_Skirt_Bck_l_4 {parent=dyn_Skirt_Bck_l_3, loc=<10 0 0>, rot=<0 0 0 1>}
 dyn_Skirt_Bck_l_5 {parent=dyn_Skirt_Bck_l_4, loc=<10 0 0>, rot=<0 0 0 1>}
 Mstr_Skirt_Mid_Bck {parent=pelvis, loc=<3.8313 -12.246 0>, rot=<0.0276 -0.7066 0.7066 -0.0276>}
 dyn_Skirt_Mid_Bck_1 {parent=Mstr_Skirt_Mid_Bck, loc=<0 0 0>, rot=<0 0 0 1>}
 dyn_Skirt_Mid_Bck_2 {parent=dyn_Skirt_Mid_Bck_1, loc=<5 0 0>, rot=<0 0 0 1>}
 dyn_Skirt_Mid_Bck_3 {parent=dyn_Skirt_Mid_Bck_2, loc=<10 0 0>, rot=<0 0 0 1>}
 dyn_Skirt_Mid_Bck_4 {parent=dyn_Skirt_Mid_Bck_3, loc=<10 0 0>, rot=<0 0 0 1>}
 dyn_Skirt_Mid_Bck_5 {parent=dyn_Skirt_Mid_Bck_4, loc=<10 0 0>, rot=<0 0 0 1>}
 Mstr_Skirt_Bck_r {parent=pelvis, loc=<3.9552 -8.7914 8.7937>, rot=<0.0913 -0.4218 0.9016 -0.03>}
 dyn_Skirt_Bck_r_1 {parent=Mstr_Skirt_Bck_r, loc=<0 0 0>, rot=<0 0 0 1>}
 dyn_Skirt_Bck_r_2 {parent=dyn_Skirt_Bck_r_1, loc=<5 0 0>, rot=<0 0 0 1>}
 dyn_Skirt_Bck_r_3 {parent=dyn_Skirt_Bck_r_2, loc=<10 0 0>, rot=<0 0 0 1>}
 dyn_Skirt_Bck_r_4 {parent=dyn_Skirt_Bck_r_3, loc=<10 0 0>, rot=<0 0 0 1>}
 dyn_Skirt_Bck_r_5 {parent=dyn_Skirt_Bck_r_4, loc=<10 0 0>, rot=<0 0 0 1>}
 Mstr_Skirt_Side_r {parent=pelvis, loc=<4.1987 -1.9972 13.4728>, rot=<0.1188 -0.0887 0.9889 0.0072>}
 dyn_Skirt_Side_r_1 {parent=Mstr_Skirt_Side_r, loc=<0 0 0>, rot=<0 0 0 1>}
 dyn_Skirt_Side_r_2 {parent=dyn_Skirt_Side_r_1, loc=<5 0 0>, rot=<0 0 0 1>}
 dyn_Skirt_Side_r_3 {parent=dyn_Skirt_Side_r_2, loc=<10 0 0>, rot=<0 0 0 1>}
 dyn_Skirt_Side_r_4 {parent=dyn_Skirt_Side_r_3, loc=<10 0 0>, rot=<0 0 0 1>}
 dyn_Skirt_Side_r_5 {parent=dyn_Skirt_Side_r_4, loc=<10 0 0>, rot=<0 0 0 1>}
 Mstr_Skirt_Frnt_r {parent=pelvis, loc=<4.4328 4.5344 8.7733>, rot=<0.1009 0.4551 0.8844 0.0226>}
 dyn_Skirt_Frnt_r_1 {parent=Mstr_Skirt_Frnt_r, loc=<0 0 0>, rot=<0 0 0 1>}
 dyn_Skirt_Frnt_r_2 {parent=dyn_Skirt_Frnt_r_1, loc=<5 0 0>, rot=<0 0 0 1>}
 dyn_Skirt_Frnt_r_3 {parent=dyn_Skirt_Frnt_r_2, loc=<10 0 0>, rot=<0 0 0 1>}
 dyn_Skirt_Frnt_r_4 {parent=dyn_Skirt_Frnt_r_3, loc=<10 0 0>, rot=<0 0 0 1>}
 dyn_Skirt_Frnt_r_5 {parent=dyn_Skirt_Frnt_r_4, loc=<10 0 0>, rot=<0 0 0 1>}
 base_hipchain_r {parent=pelvis, loc=<7.1451 -4.3565 11.8669>, rot=<0.1837 -0.2958 0.9354 -0.0619>}
 dyn_hipchain_r_01 {parent=base_hipchain_r, loc=<0 0 0>, rot=<0 0 0 1>}
 dyn_hipchain_r_02 {parent=dyn_hipchain_r_01, loc=<2.5727 0.1198 0>, rot=<0 0 0 1>}
 dyn_hipchain_r_03 {parent=dyn_hipchain_r_02, loc=<2.5578 0.1508 0.1038>, rot=<0 0 0 1>}
 base_hipchain_bk1 {parent=pelvis, loc=<8.2445 -9.5513 -0.7314>, rot=<0.154 -0.6901 0.6901 -0.154>}
 dyn_hipchain_bk_01 {parent=base_hipchain_bk1, loc=<0 0 0>, rot=<0 0 0 1>}
 dyn_hipchain_bk_02 {parent=dyn_hipchain_bk_01, loc=<3.0805 0 0>, rot=<0 0 0 1>}
 Mstr_Skirt_Cloth_Frnt {parent=pelvis, loc=<4.7247 6.053 -9.2468>, rot=<0.0073 -0.8194 -0.5685 -0.0724>}
 dyn_Skirt_Cloth_Frnt_l_1 {parent=Mstr_Skirt_Cloth_Frnt, loc=<0 0 0>, rot=<0 0 0 1>}
 dyn_Skirt_Cloth_Frnt_l_2 {parent=dyn_Skirt_Cloth_Frnt_l_1, loc=<5 0 0>, rot=<0 0 0 1>}
 dyn_Skirt_Cloth_Frnt_l_3 {parent=dyn_Skirt_Cloth_Frnt_l_2, loc=<10 0 0>, rot=<0 0 0 1>}
 dyn_Skirt_Cloth_Frnt_l_4 {parent=dyn_Skirt_Cloth_Frnt_l_3, loc=<10 0 0>, rot=<0 0 0 1>}
 dyn_Skirt_Cloth_Frnt_l_5 {parent=dyn_Skirt_Cloth_Frnt_l_4, loc=<10 0 0>, rot=<0 0 0 1>}
 Mstr_Skirt_Cloth_Side {parent=pelvis, loc=<4.961 5.1678 -11.5776>, rot=<0.1281 -0.847 -0.48 -0.1889>}
 dyn_Skirt_Cloth_Side_l_1 {parent=Mstr_Skirt_Cloth_Side, loc=<0 0 0>, rot=<0 0 0 1>}
 dyn_Skirt_Cloth_Side_l_2 {parent=dyn_Skirt_Cloth_Side_l_1, loc=<5 0 0>, rot=<0 0 0 1>}
 dyn_Skirt_Cloth_Side_l_3 {parent=dyn_Skirt_Cloth_Side_l_2, loc=<10 0 0>, rot=<0 0 0 1>}
 attach {parent=root, loc=<0 0 0>, rot=<0 0 0 1>}
 ik_hand_root {parent=root, loc=<0 0 0>, rot=<0 0 0 1>}
 ik_hand_gun {parent=ik_hand_root, loc=<-48.1743 3.3545 104.4381>, rot=<0.6276 0.1309 0.4183 -0.6434>}
 ik_hand_r {parent=ik_hand_gun, loc=<0 0 0>, rot=<0 0 0 1>}
 ik_hand_l {parent=ik_hand_gun, loc=<59.3271 67.6997 34.3524>, rot=<-0.6158 -0.7027 -0.3565 0>}
 ik_foot_root {parent=root, loc=<0 0 0>, rot=<0 0 0 1>}
 ik_foot_l {parent=ik_foot_root, loc=<14.8262 -7.3545 8.296>, rot=<0.0357 0.7045 0.0284 -0.7082>}
 ik_foot_r {parent=ik_foot_root, loc=<-14.8262 -7.3545 8.296>, rot=<0.7082 0.0284 -0.7045 0.0357>}
 root_fp {parent=root, loc=<0 0 0>, rot=<0 0 0 1>}
 attach_fp {parent=root_fp, loc=<0 0 0>, rot=<0 0 0 1>}
 camera_root {parent=root_fp, loc=<0 0 0>, rot=<0 0 0 1>}
 camera {parent=camera_root, loc=<0 0 0>, rot=<0 0 0 1>}
 attach_emote_01 {parent=root, loc=<0 0 0>, rot=<0 0 0 1>}
 attach_emote_02 {parent=root, loc=<0 0 0>, rot=<0 0 0 1>}
 attach_emote_03 {parent=root, loc=<0 0 0>, rot=<0 0 0 1>}
 warp_point_01 {parent=root, loc=<0 0 0>, rot=<0 0 0 1>}
 warp_point_02 {parent=root, loc=<0 0 0>, rot=<0 0 0 1>}
 warp_point_03 {parent=root, loc=<0 0 0>, rot=<0 0 0 1>}
]
variables [parachuting_alpha {type=double, default=0}]
