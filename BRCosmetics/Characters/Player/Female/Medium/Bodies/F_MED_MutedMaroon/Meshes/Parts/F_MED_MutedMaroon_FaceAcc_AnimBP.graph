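# Animation Blueprint for the MutedMaroon female face-accessory mesh, animating hair, earrings and head wings on top of copied parent and body poses.
# Head-turn and look-up values drive additive back-hair corrections alongside spring-based AnimDynamics.
# A cached start pose feeds a parachute-aware Control Rig.
# Cape and backbling flags select RigidBody simulations or aim/crouch offsets, then hair bends and a custom-idle slot layers over front and top hair.
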AnimBlueprint /BRCosmetics/Characters/Player/Female/Medium/Bodies/F_MED_MutedMaroon/Meshes/Parts/F_MED_MutedMaroon_FaceAcc_AnimBP
target_skeleton=&/F_MED_MutedMaroon_FaceAcc_Skeleton
graphs [
 AnimGraph {nodes [
   Anim Dynamics - Bone: dyn_FrntHair_C_1!(ComponentPose, Alpha=1, GravityOverride:Vector="0,0,-980", ExternalForce="0,0,0") -> (Pose) {
    angularDampingOverride=1
    angularSpringConstant=4000
    bAngularSpring=true
    bOverrideAngularDamping=true
    bUseGravityOverride=true
    gravityScale=0.35
    simulationSpace=RootRelative
    alphaBoolBlend {blendOption=Linear}
    boundBone {boneName=dyn_FrntHair_C_1}
    componentAppliedLinearAccClamp=<5000 5000 5000>
    componentLinearAccScale=<0.5 0.5 0.5>
    componentLinearVelScale=<0.5 0.5 0.5>
    gravityOverride {z=-980}
    physicsBodyDefinitions [{
      boundBone {boneName=dyn_FrntHair_C_1}
      boxExtents=<15 15 15>
      constraintSetup {angularLimitsMax=<0 0 2>, angularLimitsMin=<0 0 0>, angularTarget=<1 0 0>}
      localJointOffset=<1 0 0>
     }]
   }
   n1 = Anim Dynamics - Bone: dyn_FrntHair_A_3!(ComponentPose, Alpha=1, GravityOverride:Vector="0,0,-980", ExternalForce="0,0,0") -> (Pose) {
    angularDampingOverride=1
    angularSpringConstant=7000
    bAngularSpring=true
    bOverrideAngularDamping=true
    bUseGravityOverride=true
    gravityScale=0.35
    simulationSpace=RootRelative
    alphaBoolBlend {blendOption=Linear}
    boundBone {boneName=dyn_FrntHair_A_3}
    componentAppliedLinearAccClamp=<5000 5000 5000>
    gravityOverride {z=-980}
    physicsBodyDefinitions [{
      boundBone {boneName=dyn_FrntHair_A_3}
      boxExtents=<15 15 15>
      constraintSetup {angularLimitsMax=<0 0 2>, angularLimitsMin=<0 0 0>, angularTarget=<1 0 0>}
      localJointOffset=<1 0 0>
     }]
   }
   n2 = Anim Dynamics - Bone: dyn_FrntHair_A_3(ComponentPose=@n1, Alpha=1, GravityOverride="0,0,-980", ExternalForce="0,0,0") {
    angularDampingOverride=1
    angularSpringConstant=7000
    bAngularSpring=true
    bOverrideAngularDamping=true
    bUseGravityOverride=true
    gravityScale=0.35
    simulationSpace=RootRelative
    alphaBoolBlend {blendOption=Linear}
    boundBone {boneName=dyn_FrntHair_A_3}
    componentAppliedLinearAccClamp=<5000 5000 5000>
    gravityOverride {z=-980}
    physicsBodyDefinitions [{
      boundBone {boneName=dyn_FrntHair_A_3}
      boxExtents=<15 15 15>
      constraintSetup {angularLimitsMax=<0 0 2>, angularLimitsMin=<0 0 0>, angularTarget=<1 0 0>}
      localJointOffset=<1 0 0>
     }]
   }
   n3 = Anim Dynamics - Bone: dyn_FrntHair_A_4!(ComponentPose=@n2, Alpha=1, GravityOverride:Vector="0,0,-980", ExternalForce="0,0,0") -> (Pose) {
    angularDampingOverride=1
    angularSpringConstant=7000
    bAngularSpring=true
    bOverrideAngularDamping=true
    bUseGravityOverride=true
    gravityScale=0.35
    simulationSpace=RootRelative
    alphaBoolBlend {blendOption=Linear}
    boundBone {boneName=dyn_FrntHair_A_4}
    componentAppliedLinearAccClamp=<5000 5000 5000>
    gravityOverride {z=-980}
    physicsBodyDefinitions [{
      boundBone {boneName=dyn_FrntHair_A_4}
      boxExtents=<15 15 15>
      constraintSetup {angularLimitsMax=<0 0 2>, angularLimitsMin=<0 0 0>, angularTarget=<1 0 0>}
      localJointOffset=<1 0 0>
     }]
   }
   n4 = Anim Dynamics - Bone: dyn_FrntHair_B_2!(ComponentPose=@n3, Alpha=1, GravityOverride:Vector="0,0,-980", ExternalForce="0,0,0") -> (Pose) {
    angularDampingOverride=1
    angularSpringConstant=7000
    bAngularSpring=true
    bOverrideAngularDamping=true
    bUseGravityOverride=true
    gravityScale=0.35
    simulationSpace=RootRelative
    alphaBoolBlend {blendOption=Linear}
    boundBone {boneName=dyn_FrntHair_B_2}
    componentAppliedLinearAccClamp=<5000 5000 5000>
    componentLinearAccScale=<0.5 0.5 0.5>
    componentLinearVelScale=<0.5 0.5 0.5>
    gravityOverride {z=-980}
    physicsBodyDefinitions [{
      boundBone {boneName=dyn_FrntHair_B_2}
      boxExtents=<15 15 15>
      constraintSetup {angularLimitsMax=<0 0 -5>, angularLimitsMin=<0 0 -5>, angularTarget=<1 0 0>}
      localJointOffset=<1 0 0>
     }]
   }
   n5 = Anim Dynamics - Bone: dyn_FrntHair_B_3!(ComponentPose=@n4, Alpha=1, GravityOverride:Vector="0,0,-980", ExternalForce="0,0,0") -> (Pose) {
    angularDampingOverride=1
    angularSpringConstant=7000
    bAngularSpring=true
    bOverrideAngularDamping=true
    bUseGravityOverride=true
    gravityScale=0.35
    simulationSpace=RootRelative
    alphaBoolBlend {blendOption=Linear}
    boundBone {boneName=dyn_FrntHair_B_3}
    componentAppliedLinearAccClamp=<5000 5000 5000>
    componentLinearAccScale=<0.5 0.5 0.5>
    componentLinearVelScale=<0.5 0.5 0.5>
    gravityOverride {z=-980}
    physicsBodyDefinitions [{
      boundBone {boneName=dyn_FrntHair_B_3}
      boxExtents=<15 15 15>
      constraintSetup {angularLimitsMax=<0 0 -5>, angularLimitsMin=<0 0 -5>, angularTarget=<1 0 0>}
      localJointOffset=<1 0 0>
     }]
   }
   n6 = Anim Dynamics - Bone: dyn_FrntHair_C_2!(ComponentPose=@n5, Alpha=1, GravityOverride:Vector="0,0,-980", ExternalForce="0,0,0") -> (Pose) {
    angularDampingOverride=1
    angularSpringConstant=4000
    bAngularSpring=true
    bOverrideAngularDamping=true
    bUseGravityOverride=true
    gravityScale=0.35
    simulationSpace=RootRelative
    alphaBoolBlend {blendOption=Linear}
    boundBone {boneName=dyn_FrntHair_C_2}
    componentAppliedLinearAccClamp=<5000 5000 5000>
    componentLinearAccScale=<0.5 0.5 0.5>
    componentLinearVelScale=<0.5 0.5 0.5>
    gravityOverride {z=-980}
    physicsBodyDefinitions [{
      boundBone {boneName=dyn_FrntHair_C_2}
      boxExtents=<15 15 15>
      constraintSetup {angularLimitsMax=<0 0 2>, angularLimitsMin=<0 0 0>, angularTarget=<1 0 0>}
      localJointOffset=<1 0 0>
     }]
   }
   n7 = Anim Dynamics - Bone: Mstr_earring_l!(ComponentPose=@n6, Alpha=1, GravityOverride:Vector="0,0,-980", ExternalForce="0,0,0") -> (Pose) {
    angularDampingOverride=1
    angularSpringConstant=15000
    bAngularSpring=true
    bOverrideAngularDamping=true
    bUseGravityOverride=true
    gravityScale=0.35
    simulationSpace=RootRelative
    alphaBoolBlend {blendOption=Linear}
    boundBone {boneName=Mstr_earring_l}
    componentLinearAccScale=<1 1 1>
    componentLinearVelScale=<1 1 1>
    gravityOverride {z=-980}
    physicsBodyDefinitions [{
      boundBone {boneName=Mstr_earring_l}
      boxExtents=<20 30 30>
      constraintSetup {angularLimitsMax=<50 70 50>, angularLimitsMin=<-50 -70 -50>, angularTarget=<1 0 0>}
      localJointOffset=<2 0 0>
     }]
   }
   n8 = Anim Dynamics - Bone: Mstr_earring_r!(ComponentPose=@n7, Alpha=1, GravityOverride:Vector="0,0,-980", ExternalForce="0,0,0") -> (Pose) {
    angularDampingOverride=1
    angularSpringConstant=15000
    bAngularSpring=true
    bOverrideAngularDamping=true
    bUseGravityOverride=true
    gravityScale=0.35
    simulationSpace=RootRelative
    alphaBoolBlend {blendOption=Linear}
    boundBone {boneName=Mstr_earring_r}
    componentLinearAccScale=<1 1 1>
    componentLinearVelScale=<1 1 1>
    gravityOverride {z=-980}
    physicsBodyDefinitions [{
      boundBone {boneName=Mstr_earring_r}
      boxExtents=<20 30 30>
      constraintSetup {angularLimitsMax=<50 70 50>, angularLimitsMin=<-50 -70 -50>, angularTarget=<1 0 0>}
      localJointOffset=<2 0 0>
     }]
   }
   n9 = "Transform (Modify) Bone - Bone: Mstr_Back_Hair_A"(ComponentPose=@n8, Alpha=1, Translation="0,0,0", Rotation="0,0,10", Scale="1,1,1") {
    rotationMode=BMM_Additive
    scaleMode=BMM_Additive
    scaleSpace=BCS_BoneSpace
    translationMode=BMM_Additive
    alphaBoolBlend {blendOption=Linear}
    boneToModify {boneName=Mstr_Back_Hair_A}
    rotation {roll=10}
   }
   n10 = "Transform (Modify) Bone - Bone: Mstr_Back_Hair_B"(ComponentPose=@n9, Alpha=1, Translation="0,0,0", Rotation="0,0,10", Scale="1,1,1") {
    rotationMode=BMM_Additive
    scaleMode=BMM_Additive
    scaleSpace=BCS_BoneSpace
    translationMode=BMM_Additive
    alphaBoolBlend {blendOption=Linear}
    boneToModify {boneName=Mstr_Back_Hair_B}
    rotation {roll=10}
   }
   n11 = "Transform (Modify) Bone - Bone: Mstr_Back_Hair_C"(ComponentPose=@n10, Alpha=1, Translation="0,-5,0", Rotation="0,0,-10", Scale="1,1,1") {
    rotationMode=BMM_Additive
    scaleMode=BMM_Additive
    scaleSpace=BCS_BoneSpace
    translationMode=BMM_Additive
    alphaBoolBlend {blendOption=Linear}
    boneToModify {boneName=Mstr_Back_Hair_C}
    rotation {roll=-10}
    translation {y=-5}
   }
   n12 = "Transform (Modify) Bone - Bone: dyn_Back_Hair_A_2"(ComponentPose=@n11, Alpha=1, Translation="0,0,0", Rotation="0,0,-20", Scale="1,1,1") {
    rotationMode=BMM_Additive
    scaleMode=BMM_Additive
    scaleSpace=BCS_BoneSpace
    translationMode=BMM_Additive
    alphaBoolBlend {blendOption=Linear}
    boneToModify {boneName=dyn_Back_Hair_A_2}
    rotation {roll=-20}
   }
   n13 = "Transform (Modify) Bone - Bone: dyn_Back_Hair_B_2"(ComponentPose=@n12, Alpha=1, Translation="0,0,0", Rotation="0,0,-20", Scale="1,1,1") {
    rotationMode=BMM_Additive
    scaleMode=BMM_Additive
    scaleSpace=BCS_BoneSpace
    translationMode=BMM_Additive
    alphaBoolBlend {blendOption=Linear}
    boneToModify {boneName=dyn_Back_Hair_B_2}
    rotation {roll=-20}
   }
   n14 = "Transform (Modify) Bone - Bone: dyn_Back_Hair_C_2"(ComponentPose=@n13, Alpha=1, Translation="0,0,0", Rotation="0,0,-50", Scale="1,1,1") {
    rotationMode=BMM_Additive
    scaleMode=BMM_Additive
    scaleSpace=BCS_BoneSpace
    translationMode=BMM_Additive
    alphaBoolBlend {blendOption=Linear}
    boneToModify {boneName=dyn_Back_Hair_C_2}
    rotation {roll=-50}
   }
   n15 = "Transform (Modify) Bone - Bone: dyn_Back_Hair_A_3"(ComponentPose=@n14, Alpha=1, Translation="0,0,0", Rotation="0,0,0", Scale="1,1,1") {
    rotationMode=BMM_Additive
    scaleMode=BMM_Additive
    scaleSpace=BCS_BoneSpace
    translationMode=BMM_Additive
    alphaBoolBlend {blendOption=Linear}
    boneToModify {boneName=dyn_Back_Hair_A_3}
   }
   n16 = "Transform (Modify) Bone - Bone: dyn_Back_Hair_B_3"(ComponentPose=@n15, Alpha=1, Translation="0,0,0", Rotation="0,0,0", Scale="1,1,1") {
    rotationMode=BMM_Additive
    scaleMode=BMM_Additive
    scaleSpace=BCS_BoneSpace
    translationMode=BMM_Additive
    alphaBoolBlend {blendOption=Linear}
    boneToModify {boneName=dyn_Back_Hair_B_3}
   }
   n17 = "Transform (Modify) Bone - Bone: dyn_Back_Hair_C_3"(ComponentPose=@n16, Alpha=1, Translation="0,0,0", Rotation="0,0,50", Scale="1,1,1") {
    rotationMode=BMM_Additive
    scaleMode=BMM_Additive
    scaleSpace=BCS_BoneSpace
    translationMode=BMM_Additive
    alphaBoolBlend {blendOption=Linear}
    boneToModify {boneName=dyn_Back_Hair_C_3}
    rotation {roll=50}
   }
   n18 = "Constraint - dyn_Back_Hair_A_1 (PPPP)"!(
    ComponentPose=@n17
    bAlphaBoolEnabled:bool=True
    ConstraintWeights_0=0.4422
    ConstraintWeights_1:float=0.2115
    ConstraintWeights_2:float=0.2075
    ConstraintWeights_3:float=0.1388
   ) -> (Pose) {
    alphaInputType=Bool
    alphaBoolBlend {blendOption=Linear}
    boneToModify {boneName=dyn_Back_Hair_A_1}
    constraintSetup [{targetBone {boneName=neck_02}}, {targetBone {boneName=head}}, {targetBone {boneName=neck_01}}, {targetBone {boneName=spine_05}}]
    constraintWeights [1,1,1,1]
   }
   n19 = "Constraint - dyn_Back_Hair_B_1 (PPPP)"!(
    ComponentPose=@n18
    bAlphaBoolEnabled:bool=True
    ConstraintWeights_0=0.5042
    ConstraintWeights_1:float=0.2494
    ConstraintWeights_2:float=0.1834
    ConstraintWeights_3:float=0.063
   ) -> (Pose) {
    alphaInputType=Bool
    alphaBoolBlend {blendOption=Linear}
    boneToModify {boneName=dyn_Back_Hair_B_1}
    constraintSetup [{targetBone {boneName=neck_02}}, {targetBone {boneName=head}}, {targetBone {boneName=neck_01}}, {targetBone {boneName=spine_05}}]
    constraintWeights [1,1,1,1]
   }
   n20 = "Constraint - dyn_Back_Hair_C_1 (PPPP)"!(
    ComponentPose=@n19
    bAlphaBoolEnabled:bool=True
    ConstraintWeights_0=0.4427
    ConstraintWeights_1:float=0.2117
    ConstraintWeights_2:float=0.2076
    ConstraintWeights_3:float=0.138
   ) -> (Pose) {
    alphaInputType=Bool
    alphaBoolBlend {blendOption=Linear}
    boneToModify {boneName=dyn_Back_Hair_C_1}
    constraintSetup [{targetBone {boneName=neck_02}}, {targetBone {boneName=head}}, {targetBone {boneName=neck_01}}, {targetBone {boneName=spine_05}}]
    constraintWeights [1,1,1,1]
   }
   n21 = "Constraint - dyn_Back_Hair_A_2 (PPPP)"!(
    ComponentPose=@n20
    bAlphaBoolEnabled:bool=True
    ConstraintWeights_0=0.2422
    ConstraintWeights_1:float=0.1115
    ConstraintWeights_2:float=0.1075
    ConstraintWeights_3:float=0.5388
   ) -> (Pose) {
    alphaInputType=Bool
    alphaBoolBlend {blendOption=Linear}
    boneToModify {boneName=dyn_Back_Hair_A_2}
    constraintSetup [{targetBone {boneName=neck_02}}, {targetBone {boneName=head}}, {targetBone {boneName=neck_01}}, {targetBone {boneName=spine_05}}]
    constraintWeights [1,1,1,1]
   }
   n22 = "Constraint - dyn_Back_Hair_A_3 (PPP)"!(
    ComponentPose=@n21
    bAlphaBoolEnabled:bool=True
    ConstraintWeights_0=0.8329
    ConstraintWeights_1:float=0.0878
    ConstraintWeights_2:float=0.0792
   ) -> (Pose) {
    alphaInputType=Bool
    alphaBoolBlend {blendOption=Linear}
    boneToModify {boneName=dyn_Back_Hair_A_3}
    constraintSetup [{targetBone {boneName=clavicle_l}}, {targetBone {boneName=upperarm_twist_01_l}}, {targetBone {boneName=spine_05}}]
    constraintWeights [1,1,1]
   }
   n23 = "Constraint - dyn_Back_Hair_A_4 (PPPP)"!(
    ComponentPose=@n22
    bAlphaBoolEnabled:bool=True
    ConstraintWeights_0=0.4667
    ConstraintWeights_1:float=0.1856
    ConstraintWeights_2:float=0.1845
    ConstraintWeights_3:float=0.1632
   ) -> (Pose) {
    alphaInputType=Bool
    alphaBoolBlend {blendOption=Linear}
    boneToModify {boneName=dyn_Back_Hair_A_4}
    constraintSetup [
     {targetBone {boneName=clavicle_l}}
     {targetBone {boneName=spine_05}}
     {targetBone {boneName=upperarm_twist_01_l}}
     {targetBone {boneName=spine_04}}
    ]
    constraintWeights [1,1,1,1]
   }
   n24 = "Constraint - dyn_Back_Hair_B_2 (PPPP)"!(
    ComponentPose=@n23
    bAlphaBoolEnabled:bool=True
    ConstraintWeights_0=0.4373
    ConstraintWeights_1:float=0.035
    ConstraintWeights_2:float=0.0276
    ConstraintWeights_3:float=0.5
   ) -> (Pose) {
    alphaInputType=Bool
    alphaBoolBlend {blendOption=Linear}
    boneToModify {boneName=dyn_Back_Hair_B_2}
    constraintSetup [
     {targetBone {boneName=dyn_Back_Hair_B_1}}
     {targetBone {boneName=head}}
     {targetBone {boneName=dyn_Back_Hair_C_1}}
     {targetBone {boneName=spine_05}}
    ]
    constraintWeights [1,1,1,1]
   }
   n25 = "Constraint - dyn_Back_Hair_B_3 (PPP)"!(
    ComponentPose=@n24
    bAlphaBoolEnabled:bool=True
    ConstraintWeights_0=0.7815
    ConstraintWeights_1:float=0.1526
    ConstraintWeights_2:float=0.0658
   ) -> (Pose) {
    alphaInputType=Bool
    alphaBoolBlend {blendOption=Linear}
    boneToModify {boneName=dyn_Back_Hair_B_3}
    constraintSetup [{targetBone {boneName=spine_05}}, {targetBone {boneName=clavicle_r}}, {targetBone {boneName=clavicle_l}}]
    constraintWeights [1,1,1]
   }
   n26 = "Constraint - dyn_Back_Hair_B_4 (PPPP)"!(
    ComponentPose=@n25
    bAlphaBoolEnabled:bool=True
    ConstraintWeights_0=0.5088
    ConstraintWeights_1:float=0.2593
    ConstraintWeights_2:float=0.1413
    ConstraintWeights_3:float=0.0905
   ) -> (Pose) {
    alphaInputType=Bool
    alphaBoolBlend {blendOption=Linear}
    boneToModify {boneName=dyn_Back_Hair_B_4}
    constraintSetup [{targetBone {boneName=spine_05}}, {targetBone {boneName=spine_04}}, {targetBone {boneName=clavicle_r}}, {targetBone {boneName=clavicle_l}}]
    constraintWeights [1,1,1,1]
   }
   n27 = "Constraint - dyn_Back_Hair_C_2 (PPPP)"!(
    ComponentPose=@n26
    bAlphaBoolEnabled:bool=True
    ConstraintWeights_0=0.2427
    ConstraintWeights_1:float=0.1117
    ConstraintWeights_2:float=0.1076
    ConstraintWeights_3:float=0.538
   ) -> (Pose) {
    alphaInputType=Bool
    alphaBoolBlend {blendOption=Linear}
    boneToModify {boneName=dyn_Back_Hair_C_2}
    constraintSetup [{targetBone {boneName=neck_02}}, {targetBone {boneName=head}}, {targetBone {boneName=neck_01}}, {targetBone {boneName=spine_05}}]
    constraintWeights [1,1,1,1]
   }
   n28 = "Constraint - dyn_Back_Hair_C_3 (PPP)"!(
    ComponentPose=@n27
    bAlphaBoolEnabled:bool=True
    ConstraintWeights_0=0.8329
    ConstraintWeights_1:float=0.0878
    ConstraintWeights_2:float=0.0792
   ) -> (Pose) {
    alphaInputType=Bool
    alphaBoolBlend {blendOption=Linear}
    boneToModify {boneName=dyn_Back_Hair_C_3}
    constraintSetup [{targetBone {boneName=clavicle_r}}, {targetBone {boneName=upperarm_twist_01_r}}, {targetBone {boneName=spine_05}}]
    constraintWeights [1,1,1]
   }
   n29 = "Constraint - dyn_Back_Hair_C_4 (PPPP)"!(
    ComponentPose=@n28
    bAlphaBoolEnabled:bool=True
    ConstraintWeights_0=0.4667
    ConstraintWeights_1:float=0.1856
    ConstraintWeights_2:float=0.1845
    ConstraintWeights_3:float=0.1632
   ) -> (Pose) {
    alphaInputType=Bool
    alphaBoolBlend {blendOption=Linear}
    boneToModify {boneName=dyn_Back_Hair_C_4}
    constraintSetup [
     {targetBone {boneName=clavicle_r}}
     {targetBone {boneName=spine_05}}
     {targetBone {boneName=upperarm_twist_01_r}}
     {targetBone {boneName=spine_04}}
    ]
    constraintWeights [1,1,1,1]
   }
   n30 = "Transform (Modify) Bone - Bone: dyn_Back_Hair_C_1"!(ComponentPose=@n29, bAlphaBoolEnabled:bool=True, Translation="0,0,0", Rotation="20,0,0", Scale="1,1,1") -> (Pose) {
    alphaInputType=Bool
    rotationMode=BMM_Additive
    rotationSpace=BCS_BoneSpace
    translationMode=BMM_Additive
    translationSpace=BCS_BoneSpace
    alphaBoolBlend {blendInTime=0.2, blendOption=Linear, blendOutTime=0.2}
    boneToModify {boneName=dyn_Back_Hair_C_1}
    rotation {pitch=20}
   }
   Component To Local(ComponentPose=@n30)
   "Transform (Modify) Bone - Bone: dyn_Back_Hair_C_1"!(ComponentPose, Alpha:float=1, Translation="0,0,0", Rotation="0,0,0", Scale="1,1,1") -> (Pose) {
    rotationMode=BMM_Additive
    rotationSpace=BCS_BoneSpace
    translationMode=BMM_Additive
    translationSpace=BCS_BoneSpace
    alphaBoolBlend {blendOption=Linear}
    boneToModify {boneName=dyn_Back_Hair_C_1}
   }
   n31 = "Transform (Modify) Bone - Bone: base_hair_back_r"!(ComponentPose, Alpha=1, Translation="0,0,0", Rotation="0,-70,0") -> (Pose) {
    rotationMode=BMM_Additive
    rotationSpace=BCS_BoneSpace
    translationMode=BMM_Additive
    translationSpace=BCS_BoneSpace
    alphaBoolBlend {blendOption=Linear}
    boneToModify {boneName=base_hair_back_r}
    rotation {yaw=-70}
   }
   n32 = "Transform (Modify) Bone - Bone: base_hair_back_m"!(ComponentPose=@n31, Alpha=1, Rotation="0,-50,0") -> (Pose) {
    rotationMode=BMM_Additive
    rotationSpace=BCS_BoneSpace
    alphaBoolBlend {blendOption=Linear}
    boneToModify {boneName=base_hair_back_m}
    rotation {yaw=-50}
   }
   n33 = "Transform (Modify) Bone - Bone: base_hair_back_l"!(ComponentPose=@n32, Alpha=1, Rotation="0,-40,0") -> (Pose) {
    rotationMode=BMM_Additive
    rotationSpace=BCS_BoneSpace
    alphaBoolBlend {blendOption=Linear}
    boneToModify {boneName=base_hair_back_l}
    rotation {yaw=-40}
   }
   n34 = "Transform (Modify) Bone - Bone: base_hair_back_r"!(ComponentPose=@n33, Alpha=1, Rotation="-20,0,0") -> (Pose) {
    rotationMode=BMM_Additive
    rotationSpace=BCS_BoneSpace
    alphaBoolBlend {blendOption=Linear}
    boneToModify {boneName=base_hair_back_r}
    rotation {pitch=-20}
   }
   n35 = "Transform (Modify) Bone - Bone: base_hair_back_m"(ComponentPose=@n34, Alpha=1, Rotation="-30,0,0") {
    rotationMode=BMM_Additive
    rotationSpace=BCS_BoneSpace
    alphaBoolBlend {blendOption=Linear}
    boneToModify {boneName=base_hair_back_m}
    rotation {pitch=-30}
   }
   n36 = "Transform (Modify) Bone - Bone: base_hair_back_l"(ComponentPose=@n35, Alpha=1, Rotation="-30,0,0") {
    rotationMode=BMM_Additive
    rotationSpace=BCS_BoneSpace
    alphaBoolBlend {blendOption=Linear}
    boneToModify {boneName=base_hair_back_l}
    rotation {pitch=-30}
   }
   n37 = "Transform (Modify) Bone - Bone: base_hair_back_r"(ComponentPose=@n36, Alpha=1, Rotation="20,0,0") {
    rotationMode=BMM_Additive
    rotationSpace=BCS_BoneSpace
    alphaBoolBlend {blendOption=Linear}
    boneToModify {boneName=base_hair_back_r}
    rotation {pitch=20}
   }
   n38 = "Transform (Modify) Bone - Bone: base_hair_back_m"(ComponentPose=@n37, Alpha=1, Rotation="30,0,0") {
    rotationMode=BMM_Additive
    rotationSpace=BCS_BoneSpace
    alphaBoolBlend {blendOption=Linear}
    boneToModify {boneName=base_hair_back_m}
    rotation {pitch=30}
   }
   "Transform (Modify) Bone - Bone: base_hair_back_l"(ComponentPose=@n38, Alpha=1, Rotation="0,0,0") {rotationMode=BMM_Additive, rotationSpace=BCS_BoneSpace, alphaBoolBlend {blendOption=Linear}, boneToModify {boneName=base_hair_back_l}}
   n39 = "Transform (Modify) Bone - Bone: dyn_Back_Hair_C_2"!(ComponentPose, bAlphaBoolEnabled:bool=True, Translation="0,0,0", Rotation="0,0,0", Scale="1,1,1") -> (Pose) {
    alphaInputType=Bool
    rotationMode=BMM_Additive
    rotationSpace=BCS_BoneSpace
    translationMode=BMM_Additive
    translationSpace=BCS_BoneSpace
    alphaBoolBlend {blendInTime=0.2, blendOption=Linear, blendOutTime=0.2}
    boneToModify {boneName=dyn_Back_Hair_C_2}
   }
   "Transform (Modify) Bone - Bone: dyn_Back_Hair_C_3"!(ComponentPose=@n39, bAlphaBoolEnabled:bool=True, Translation="0,0,0", Rotation="0,0,0", Scale="1,1,1") -> (Pose) {
    alphaInputType=Bool
    rotationMode=BMM_Additive
    rotationSpace=BCS_BoneSpace
    translationMode=BMM_Additive
    translationSpace=BCS_BoneSpace
    alphaBoolBlend {blendInTime=0.2, blendOption=Linear, blendOutTime=0.2}
    boneToModify {boneName=dyn_Back_Hair_C_3}
   }
   n40 = Get Head Down() -> (Head Down:double)
   n41 = "Transform (Modify) Bone - Bone: base_hair_back_r"!(ComponentPose, Alpha=@n40, Translation:Vector="0,100,0", Rotation="30,-50,0") -> (Pose) {
    rotationMode=BMM_Additive
    rotationSpace=BCS_BoneSpace
    translationMode=BMM_Additive
    translationSpace=BCS_BoneSpace
    alphaBoolBlend {blendOption=Linear}
    boneToModify {boneName=base_hair_back_r}
    rotation {pitch=30, yaw=-50}
    translation {y=100}
   }
   n42 = "Transform (Modify) Bone - Bone: base_hair_back_m"!(ComponentPose=@n41, Alpha=@n40, Translation:Vector="0,0,0", Rotation="0,-60,0") -> (Pose) {
    rotationMode=BMM_Additive
    rotationSpace=BCS_BoneSpace
    translationMode=BMM_Additive
    translationSpace=BCS_BoneSpace
    alphaBoolBlend {blendOption=Linear}
    boneToModify {boneName=base_hair_back_m}
    rotation {yaw=-60}
   }
   "Transform (Modify) Bone - Bone: base_hair_back_l"!(ComponentPose=@n42, Alpha=@n40, Translation:Vector="0,0,0", Rotation="0,-50,0") -> (Pose) {
    rotationMode=BMM_Additive
    rotationSpace=BCS_BoneSpace
    translationMode=BMM_Additive
    translationSpace=BCS_BoneSpace
    alphaBoolBlend {blendOption=Linear}
    boneToModify {boneName=base_hair_back_l}
    rotation {yaw=-50}
   }
   n43 = "Transform (Modify) Bone - Bone: base_hair_back_r"!(ComponentPose, Alpha=1, Rotation="10,16.558,-42.051") -> (Pose) {
    rotationMode=BMM_Additive
    rotationSpace=BCS_BoneSpace
    alphaBoolBlend {blendOption=Linear}
    boneToModify {boneName=base_hair_back_r}
    rotation=<rot 10 16.558 -42.051>
   }
   n44 = "Transform (Modify) Bone - Bone: base_hair_back_m"!(ComponentPose=@n43, Alpha=1, Rotation="20,6.367,-25.587") -> (Pose) {
    rotationMode=BMM_Additive
    rotationSpace=BCS_BoneSpace
    alphaBoolBlend {blendOption=Linear}
    boneToModify {boneName=base_hair_back_m}
    rotation=<rot 20 6.367 -25.587>
   }
   "Transform (Modify) Bone - Bone: base_hair_back_l"!(ComponentPose=@n44, Alpha=1, Rotation="10,-8.227,-23.316") -> (Pose) {
    rotationMode=BMM_Additive
    rotationSpace=BCS_BoneSpace
    alphaBoolBlend {blendOption=Linear}
    boneToModify {boneName=base_hair_back_l}
    rotation=<rot 10 -8.227 -23.316>
   }
   n45 = RigidBody!(
    ComponentPose=Local To Component(LocalPose=Use cached pose 'Cache_BackpackState'())
    Alpha
    OverrideWorldGravity:Vector="0,0,-980"
    ExternalForce
    ComponentLinearAccScale:Vector="0.01,0.01,0.01"
    ComponentLinearVelScale:Vector="0.01,0.01,0.01"
   ) -> (Pose) {
    bClampLinearTranslationLimitToRefPose=true
    bOverrideWorldGravity=true
    overridePhysicsAsset=&/F_MED_MutedMaroon_FaceAcc_Physics
    simulationSpace=BaseBoneSpace
    alphaBoolBlend {blendOption=Linear}
    baseBoneRef {boneName=neck_01}
    componentAppliedLinearAccClamp=<2000 2000 2000>
    componentLinearAccScale=<0.6 0.2 0.2>
    componentLinearVelScale=<0.6 0.2 0.2>
    simSpaceSettings {worldAlpha=0.01}
   }
   n46 = RigidBody(
    ComponentPose=Local To Component(LocalPose=Use cached pose 'Cache_StartPose'())
    OverrideWorldGravity=Get GravityOverride()
    ComponentLinearAccScale="0,0,0"
    ComponentLinearVelScale="0,0,0"
   ) {
    bClampLinearTranslationLimitToRefPose=true
    bOverrideWorldGravity=true
    overridePhysicsAsset=&F_MED_MutedMaroon_FaceAcc_Physics
    simulationSpace=BaseBoneSpace
    alphaBoolBlend {blendOption=Linear}
    baseBoneRef {boneName=neck_01}
    simSpaceSettings {worldAlpha=0.05}
   }
   n47 = Blend Poses by bool(
    BlendPose_0=Component To Local(ComponentPose=@n45)
    BlendPose_1=Component To Local(ComponentPose=@n46)
    bActiveValue=Get isWearingBackbling()
   ) {blendType=Linear}
   n48 = Get AimingUp_ShortW() -> (AimingUp_ShortW:double)
   n49 = "Transform (Modify) Bone - Bone: base_hair_back_r"!(
    ComponentPose=Local To Component(LocalPose=Use cached pose 'Cache_BackpackState'())
    Alpha=@n48
    Translation:Vector="2,1,0"
    Rotation="0,0,0"
    Scale:Vector="1,1,1"
   ) -> (Pose) {
    rotationMode=BMM_Additive
    rotationSpace=BCS_BoneSpace
    translationMode=BMM_Additive
    translationSpace=BCS_BoneSpace
    alphaBoolBlend {blendOption=Linear}
    alphaScaleBiasClamp {bInterpResult=true}
    boneToModify {boneName=base_hair_back_r}
    translation=<2 1>
   }
   n50 = "Transform (Modify) Bone - Bone: base_hair_back_m"!(ComponentPose=@n49, Alpha=@n48, Translation:Vector="3,1,0", Rotation="0,0,0", Scale:Vector="1,1,1") -> (Pose) {
    rotationMode=BMM_Additive
    rotationSpace=BCS_BoneSpace
    translationMode=BMM_Additive
    translationSpace=BCS_BoneSpace
    alphaBoolBlend {blendOption=Linear}
    alphaScaleBiasClamp {bInterpResult=true}
    boneToModify {boneName=base_hair_back_m}
    translation=<3 1>
   }
   n51 = "Transform (Modify) Bone - Bone: base_hair_back_r"(
    ComponentPose=@n50
    Alpha=Get Crouching_BackpackCape_SW() -> (Crouching_BackpackCape_SW:double)
    Translation="0,4,0"
    Rotation="0,-5,0"
    Scale="1,1,1"
   ) {
    rotationMode=BMM_Additive
    rotationSpace=BCS_BoneSpace
    scaleMode=BMM_Additive
    scaleSpace=BCS_BoneSpace
    translationMode=BMM_Additive
    translationSpace=BCS_BoneSpace
    alphaBoolBlend {blendOption=Linear}
    alphaScaleBiasClamp {bInterpResult=true}
    boneToModify {boneName=base_hair_back_r}
    rotation {yaw=-5}
    translation {y=4}
   }
   n52 = "Transform (Modify) Bone - Bone: base_hair_back_r"(
    ComponentPose=@n51
    Alpha=Get Crouching_BackpackCape_LW() -> (Crouching_BackpackCape_LW:double)
    Translation="-1,5,0"
    Rotation="0,-2,0"
    Scale="1,1,1"
   ) {
    rotationMode=BMM_Additive
    rotationSpace=BCS_BoneSpace
    scaleMode=BMM_Additive
    scaleSpace=BCS_BoneSpace
    translationMode=BMM_Additive
    translationSpace=BCS_BoneSpace
    alphaBoolBlend {blendOption=Linear}
    alphaScaleBiasClamp {bInterpResult=true}
    boneToModify {boneName=base_hair_back_r}
    rotation {yaw=-2}
    translation=<-1 5>
   }
   n53 = Component To Local(ComponentPose=@n52)
   n54 = Get Is Wearing Cape() -> (Is Wearing Cape:bool)
   n55 = "Transform (Modify) Bone - Bone: dyn_hair_back_02_m"!(
    ComponentPose=Local To Component(LocalPose=Blend Poses by bool(BlendPose_0=@n53, BlendPose_1=@n47, bActiveValue=@n54) {blendType=Linear})
    bAlphaBoolEnabled:bool=True
    Translation="0,0,0"
    Rotation="0,-10,0"
    Scale="1,1,1"
   ) -> (Pose) {
    alphaInputType=Bool
    rotationMode=BMM_Additive
    rotationSpace=BCS_BoneSpace
    translationMode=BMM_Additive
    translationSpace=BCS_BoneSpace
    alphaBoolBlend {blendInTime=0.5, blendOption=Linear, blendOutTime=0.8}
    boneToModify {boneName=dyn_hair_back_02_m}
    rotation {yaw=-10}
   }
   n56 = "Transform (Modify) Bone - Bone: dyn_hair_back_03_m"!(ComponentPose=@n55, bAlphaBoolEnabled:bool=True, Translation="0,0,0", Rotation="0,-15,0", Scale="1,1,1") -> (Pose) {
    alphaInputType=Bool
    rotationMode=BMM_Additive
    rotationSpace=BCS_BoneSpace
    translationMode=BMM_Additive
    translationSpace=BCS_BoneSpace
    alphaBoolBlend {blendInTime=0.5, blendOption=Linear, blendOutTime=0.8}
    boneToModify {boneName=dyn_hair_back_03_m}
    rotation {yaw=-15}
   }
   n57 = "Transform (Modify) Bone - Bone: dyn_hair_back_04_m"!(ComponentPose=@n56, bAlphaBoolEnabled:bool=True, Translation="0,0,0", Rotation="0,-20,0", Scale="1,1,1") -> (Pose) {
    alphaInputType=Bool
    rotationMode=BMM_Additive
    rotationSpace=BCS_BoneSpace
    translationMode=BMM_Additive
    translationSpace=BCS_BoneSpace
    alphaBoolBlend {blendInTime=0.5, blendOption=Linear, blendOutTime=0.8}
    boneToModify {boneName=dyn_hair_back_04_m}
    rotation {yaw=-20}
   }
   n58 = "Transform (Modify) Bone - Bone: dyn_hair_back_05_m"!(ComponentPose=@n57, bAlphaBoolEnabled:bool=True, Translation="0,0,0", Rotation="0,-25,0", Scale="1,1,1") -> (Pose) {
    alphaInputType=Bool
    rotationMode=BMM_Additive
    rotationSpace=BCS_BoneSpace
    translationMode=BMM_Additive
    translationSpace=BCS_BoneSpace
    alphaBoolBlend {blendInTime=0.5, blendOption=Linear, blendOutTime=0.8}
    boneToModify {boneName=dyn_hair_back_05_m}
    rotation {yaw=-25}
   }
   n59 = "Transform (Modify) Bone - Bone: dyn_hair_back_02_l"!(ComponentPose=@n58, bAlphaBoolEnabled:bool=True, Translation="0,0,0", Rotation="0,-10,0", Scale="1,1,1") -> (Pose) {
    alphaInputType=Bool
    rotationMode=BMM_Additive
    rotationSpace=BCS_BoneSpace
    translationMode=BMM_Additive
    translationSpace=BCS_BoneSpace
    alphaBoolBlend {blendInTime=0.5, blendOption=Linear, blendOutTime=0.8}
    boneToModify {boneName=dyn_hair_back_02_l}
    rotation {yaw=-10}
   }
   n60 = "Transform (Modify) Bone - Bone: dyn_hair_back_03_l"!(ComponentPose=@n59, bAlphaBoolEnabled:bool=True, Translation="0,0,0", Rotation="0,-15,0", Scale="1,1,1") -> (Pose) {
    alphaInputType=Bool
    rotationMode=BMM_Additive
    rotationSpace=BCS_BoneSpace
    translationMode=BMM_Additive
    translationSpace=BCS_BoneSpace
    alphaBoolBlend {blendInTime=0.5, blendOption=Linear, blendOutTime=0.8}
    boneToModify {boneName=dyn_hair_back_03_l}
    rotation {yaw=-15}
   }
   n61 = "Transform (Modify) Bone - Bone: dyn_hair_back_04_l"!(ComponentPose=@n60, bAlphaBoolEnabled:bool=True, Translation="0,0,0", Rotation="0,-20,0", Scale="1,1,1") -> (Pose) {
    alphaInputType=Bool
    rotationMode=BMM_Additive
    rotationSpace=BCS_BoneSpace
    translationMode=BMM_Additive
    translationSpace=BCS_BoneSpace
    alphaBoolBlend {blendInTime=0.5, blendOption=Linear, blendOutTime=0.8}
    boneToModify {boneName=dyn_hair_back_04_l}
    rotation {yaw=-20}
   }
   n62 = "Transform (Modify) Bone - Bone: dyn_hair_back_05_l"!(ComponentPose=@n61, bAlphaBoolEnabled:bool=True, Translation="0,0,0", Rotation="0,-25,0", Scale="1,1,1") -> (Pose) {
    alphaInputType=Bool
    rotationMode=BMM_Additive
    rotationSpace=BCS_BoneSpace
    translationMode=BMM_Additive
    translationSpace=BCS_BoneSpace
    alphaBoolBlend {blendInTime=0.5, blendOption=Linear, blendOutTime=0.8}
    boneToModify {boneName=dyn_hair_back_05_l}
    rotation {yaw=-25}
   }
   n63 = "Transform (Modify) Bone - Bone: dyn_hair_back_02_r"!(ComponentPose=@n62, bAlphaBoolEnabled:bool=True, Translation="0,0,0", Rotation="0,-10,0", Scale="1,1,1") -> (Pose) {
    alphaInputType=Bool
    rotationMode=BMM_Additive
    rotationSpace=BCS_BoneSpace
    translationMode=BMM_Additive
    translationSpace=BCS_BoneSpace
    alphaBoolBlend {blendInTime=0.5, blendOption=Linear, blendOutTime=0.8}
    boneToModify {boneName=dyn_hair_back_02_r}
    rotation {yaw=-10}
   }
   n64 = "Transform (Modify) Bone - Bone: dyn_hair_back_03_r"!(ComponentPose=@n63, bAlphaBoolEnabled:bool=True, Translation="0,0,0", Rotation="0,-15,0", Scale="1,1,1") -> (Pose) {
    alphaInputType=Bool
    rotationMode=BMM_Additive
    rotationSpace=BCS_BoneSpace
    translationMode=BMM_Additive
    translationSpace=BCS_BoneSpace
    alphaBoolBlend {blendInTime=0.5, blendOption=Linear, blendOutTime=0.8}
    boneToModify {boneName=dyn_hair_back_03_r}
    rotation {yaw=-15}
   }
   n65 = "Transform (Modify) Bone - Bone: dyn_hair_back_04_r"!(ComponentPose=@n64, bAlphaBoolEnabled:bool=True, Translation="0,0,0", Rotation="0,-20,0", Scale="1,1,1") -> (Pose) {
    alphaInputType=Bool
    rotationMode=BMM_Additive
    rotationSpace=BCS_BoneSpace
    translationMode=BMM_Additive
    translationSpace=BCS_BoneSpace
    alphaBoolBlend {blendInTime=0.5, blendOption=Linear, blendOutTime=0.8}
    boneToModify {boneName=dyn_hair_back_04_r}
    rotation {yaw=-20}
   }
   n66 = "Transform (Modify) Bone - Bone: dyn_hair_back_05_r"!(ComponentPose=@n65, bAlphaBoolEnabled:bool=True, Translation="0,0,0", Rotation="0,-25,0", Scale="1,1,1") -> (Pose) {
    alphaInputType=Bool
    rotationMode=BMM_Additive
    rotationSpace=BCS_BoneSpace
    translationMode=BMM_Additive
    translationSpace=BCS_BoneSpace
    alphaBoolBlend {blendInTime=0.5, blendOption=Linear, blendOutTime=0.8}
    boneToModify {boneName=dyn_hair_back_05_r}
    rotation {yaw=-25}
   }
   n67 = Layered blend per bone(
    BasePose=Component To Local(ComponentPose=@n66)
    BlendPoses_0=Slot 'DefaultSlot'()
    BlendWeights_0=Get Custom Idle is Playing() -> (Custom Idle is Playing:double)
   ) {layerSetup [{branchFilters [{boneName="Mstr_FrntHair_A, Mstr_FrntHair_B"}, {boneName=dyn_Top_Hair_1}, {boneName=Mstr_FrntHair_C}, {boneName=dyn_FrntHair_A_B_1}]}]}
   Output Pose(Result=@n67) {layerGroup=None}
   n68 = Control Rig!(Source=Use cached pose 'Cache_StartPose'(), Alpha, parachuting_alpha:double=Get isParachuting_alpha() -> (isParachuting_alpha:double)) -> (Pose) {
    alphaBoolBlend {blendOption=Linear}
    controlRigAssetReference {blueprintRigClass=&/F_MED_MutedMaroon_FaceAcc_CtrlRig.F_MED_MutedMaroon_FaceAcc_CtrlRig_C}
   }
   Save cached pose 'Cache_BackpackState'(Pose=@n68) {cacheName=Cache_BackpackState}
   n69 = Get Head_Turn_L() -> (Head_Turn_L:double)
   n70 = Get Head_Turn_R() -> (Head_Turn_R:double)
   n71 = Layered blend per bone(
    BasePose=Copy Pose From Mesh() {bCopyCurves=true, bUseAttachedParent=true}
    BlendPoses_0=Copy Pose From Mesh(SourceMeshComponent=Get Body_Mesh())
   ) {layerSetup [{branchFilters [
       {boneName=clavicle_l}
       {boneName=clavicle_r}
       {boneName=upperarm_l}
       {boneName=upperarm_r}
       {boneName=upperarm_twist_01_l}
       {boneName=upperarm_twist_01_r}
       {boneName=deform_upperarm_bk_base_l}
       {boneName=deform_upperarm_bk_l}
       {boneName=deform_upperarm_fr_base_l}
       {boneName=deform_upperarm_fr_l}
       {boneName=deform_upperarm_up_base_l}
       {boneName=deform_upperarm_up_l}
       {boneName=deform_upperarm_bk_base_r}
       {boneName=deform_upperarm_bk_r}
       {boneName=deform_upperarm_fr_base_r}
       {boneName=deform_upperarm_fr_r}
       {boneName=deform_upperarm_up_base_r}
       {boneName=deform_upperarm_up_r}
      ]}]}
   n72 = Get HeadUpFloat() -> (HeadUpFloat:double)
   n73 = "Transform (Modify) Bone - Bone: base_hair_back_r"!(ComponentPose=Local To Component(LocalPose=@n71), Alpha=@n72, Rotation="0,-50,0") -> (Pose) {
    rotationMode=BMM_Additive
    rotationSpace=BCS_BoneSpace
    alphaBoolBlend {blendOption=Linear}
    boneToModify {boneName=base_hair_back_r}
    rotation {yaw=-50}
   }
   n74 = "Transform (Modify) Bone - Bone: base_hair_back_m"!(ComponentPose=@n73, Alpha=@n72, Rotation="0,-60,0") -> (Pose) {
    rotationMode=BMM_Additive
    rotationSpace=BCS_BoneSpace
    alphaBoolBlend {blendOption=Linear}
    boneToModify {boneName=base_hair_back_m}
    rotation {yaw=-60}
   }
   n75 = "Transform (Modify) Bone - Bone: base_hair_back_l"(ComponentPose=@n74, Alpha=@n72, Rotation="0,-50,0") {
    rotationMode=BMM_Additive
    rotationSpace=BCS_BoneSpace
    alphaBoolBlend {blendOption=Linear}
    boneToModify {boneName=base_hair_back_l}
    rotation {yaw=-50}
   }
   n76 = "Transform (Modify) Bone - Bone: base_hair_back_r"!(ComponentPose=@n75, Alpha=@n70, Translation:Vector="0,0,0", Rotation="20,0,0") -> (Pose) {
    rotationMode=BMM_Additive
    rotationSpace=BCS_BoneSpace
    translationMode=BMM_Additive
    translationSpace=BCS_BoneSpace
    alphaBoolBlend {blendOption=Linear}
    boneToModify {boneName=base_hair_back_r}
    rotation {pitch=20}
   }
   n77 = "Transform (Modify) Bone - Bone: base_hair_back_m"!(ComponentPose=@n76, Alpha=@n70, Translation:Vector="0,0,0", Rotation="20,0,0") -> (Pose) {
    rotationMode=BMM_Additive
    rotationSpace=BCS_BoneSpace
    translationMode=BMM_Additive
    translationSpace=BCS_BoneSpace
    alphaBoolBlend {blendOption=Linear}
    boneToModify {boneName=base_hair_back_m}
    rotation {pitch=20}
   }
   n78 = "Transform (Modify) Bone - Bone: base_hair_back_l"!(ComponentPose=@n77, Alpha=@n70, Translation:Vector="0,0,0", Rotation="10,0,0") -> (Pose) {
    rotationMode=BMM_Additive
    rotationSpace=BCS_BoneSpace
    translationMode=BMM_Additive
    translationSpace=BCS_BoneSpace
    alphaBoolBlend {blendOption=Linear}
    boneToModify {boneName=base_hair_back_l}
    rotation {pitch=10}
   }
   n79 = "Transform (Modify) Bone - Bone: base_hair_back_r"!(ComponentPose=@n78, Alpha=@n69, Rotation="-10,-7.489,0") -> (Pose) {
    rotationMode=BMM_Additive
    rotationSpace=BCS_BoneSpace
    alphaBoolBlend {blendOption=Linear}
    alphaScaleBiasClamp {bInterpResult=true}
    boneToModify {boneName=base_hair_back_r}
    rotation {pitch=-10, yaw=-7.489}
   }
   n80 = "Transform (Modify) Bone - Bone: base_hair_back_m"!(ComponentPose=@n79, Alpha=@n69, Rotation="12.226,2.246,0") -> (Pose) {
    rotationMode=BMM_Additive
    rotationSpace=BCS_BoneSpace
    alphaBoolBlend {blendOption=Linear}
    boneToModify {boneName=base_hair_back_m}
    rotation {pitch=12.226, yaw=2.246}
   }
   n81 = "Transform (Modify) Bone - Bone: base_hair_back_l"!(ComponentPose=@n80, Alpha=@n69, Rotation="6.411,8.559,0") -> (Pose) {
    rotationMode=BMM_Additive
    rotationSpace=BCS_BoneSpace
    alphaBoolBlend {blendOption=Linear}
    boneToModify {boneName=base_hair_back_l}
    rotation {pitch=6.411, yaw=8.559}
   }
   n82 = "Transform (Modify) Bone - Bone: base_hair_back_r"(ComponentPose=@n81, Alpha=1, Rotation="0,10,15") {
    rotationMode=BMM_Additive
    rotationSpace=BCS_BoneSpace
    alphaBoolBlend {blendOption=Linear}
    boneToModify {boneName=base_hair_back_r}
    rotation {roll=15, yaw=10}
   }
   n83 = "Transform (Modify) Bone - Bone: base_hair_back_l"(ComponentPose=@n82, Alpha=1, Rotation="0,0,-8") {
    rotationMode=BMM_Additive
    rotationSpace=BCS_BoneSpace
    alphaBoolBlend {blendOption=Linear}
    boneToModify {boneName=base_hair_back_l}
    rotation {roll=-8}
   }
   n84 = Anim Dynamics - Bone: dyn_headwing_r_01!(ComponentPose=@n83, Alpha=1, GravityOverride:Vector="0,0,-980", ExternalForce="0,0,0") -> (Pose) {
    angularDampingOverride=1
    angularSpringConstant=15000
    bAngularSpring=true
    bOverrideAngularDamping=true
    bUseGravityOverride=true
    gravityScale=0.35
    simulationSpace=RootRelative
    alphaBoolBlend {blendOption=Linear}
    boundBone {boneName=dyn_headwing_r_01}
    componentLinearAccScale=<1 1 1>
    componentLinearVelScale=<1 1 1>
    gravityOverride {z=-980}
    physicsBodyDefinitions [{
      boundBone {boneName=dyn_headwing_r_01}
      boxExtents=<20 30 30>
      constraintSetup {angularLimitsMax=<0 0 5>, angularLimitsMin=<0 0 -5>, angularTarget=<1 0 0>}
      localJointOffset=<2 0 0>
     }]
   }
   n85 = Anim Dynamics - Bone: dyn_headwing_r_02!(ComponentPose=@n84, Alpha=1, GravityOverride:Vector="0,0,-980", ExternalForce="0,0,0") -> (Pose) {
    angularDampingOverride=1
    angularSpringConstant=15000
    bAngularSpring=true
    bOverrideAngularDamping=true
    bUseGravityOverride=true
    gravityScale=0.35
    simulationSpace=RootRelative
    alphaBoolBlend {blendOption=Linear}
    boundBone {boneName=dyn_headwing_r_02}
    componentLinearAccScale=<1 1 1>
    componentLinearVelScale=<1 1 1>
    gravityOverride {z=-980}
    physicsBodyDefinitions [{
      boundBone {boneName=dyn_headwing_r_02}
      boxExtents=<20 30 30>
      constraintSetup {angularLimitsMax=<0 0 5>, angularLimitsMin=<0 0 -5>, angularTarget=<1 0 0>}
      localJointOffset=<2 0 0>
     }]
   }
   n86 = Anim Dynamics - Bone: dyn_headwing_l_01!(ComponentPose=@n85, Alpha=1, GravityOverride:Vector="0,0,-980", ExternalForce="0,0,0") -> (Pose) {
    angularDampingOverride=1
    angularSpringConstant=15000
    bAngularSpring=true
    bOverrideAngularDamping=true
    bUseGravityOverride=true
    gravityScale=0.35
    simulationSpace=RootRelative
    alphaBoolBlend {blendOption=Linear}
    boundBone {boneName=dyn_headwing_l_01}
    componentLinearAccScale=<1 1 1>
    componentLinearVelScale=<1 1 1>
    gravityOverride {z=-980}
    physicsBodyDefinitions [{
      boundBone {boneName=dyn_headwing_l_01}
      boxExtents=<20 30 30>
      constraintSetup {angularLimitsMax=<0 0 5>, angularLimitsMin=<0 0 -5>, angularTarget=<1 0 0>}
      localJointOffset=<2 0 0>
     }]
   }
   n87 = Anim Dynamics - Bone: dyn_headwing_l_02!(ComponentPose=@n86, Alpha=1, GravityOverride:Vector="0,0,-980", ExternalForce="0,0,0") -> (Pose) {
    angularDampingOverride=1
    angularSpringConstant=15000
    bAngularSpring=true
    bOverrideAngularDamping=true
    bUseGravityOverride=true
    gravityScale=0.35
    simulationSpace=RootRelative
    alphaBoolBlend {blendOption=Linear}
    boundBone {boneName=dyn_headwing_l_02}
    componentLinearAccScale=<1 1 1>
    componentLinearVelScale=<1 1 1>
    gravityOverride {z=-980}
    physicsBodyDefinitions [{
      boundBone {boneName=dyn_headwing_l_02}
      boxExtents=<20 30 30>
      constraintSetup {angularLimitsMax=<0 0 5>, angularLimitsMin=<0 0 -5>, angularTarget=<1 0 0>}
      localJointOffset=<2 0 0>
     }]
   }
   n88 = Anim Dynamics - Bone: dyn_Top_Hair_1!(ComponentPose=@n87, Alpha=1, GravityOverride:Vector="0,0,-980", ExternalForce="0,0,0") -> (Pose) {
    angularDampingOverride=1
    angularSpringConstant=10000
    bAngularSpring=true
    bOverrideAngularDamping=true
    bUseGravityOverride=true
    gravityScale=0.35
    simulationSpace=RootRelative
    alphaBoolBlend {blendOption=Linear}
    boundBone {boneName=dyn_Top_Hair_1}
    componentAppliedLinearAccClamp=<5000 5000 5000>
    componentLinearAccScale=<0.5 0.5 0.5>
    componentLinearVelScale=<0.5 0.5 0.5>
    gravityOverride {z=-980}
    physicsBodyDefinitions [{
      boundBone {boneName=dyn_Top_Hair_1}
      boxExtents=<20 30 30>
      constraintSetup {angularLimitsMax=<0 0 2>, angularLimitsMin=<0 0 -2>, angularTarget=<1 0 0>}
      localJointOffset=<2 0 0>
     }]
   }
   n89 = Anim Dynamics - Bone: dyn_Top_Hair_2!(ComponentPose=@n88, Alpha=1, GravityOverride:Vector="0,0,-980", ExternalForce="0,0,0") -> (Pose) {
    angularDampingOverride=1
    angularSpringConstant=10000
    bAngularSpring=true
    bOverrideAngularDamping=true
    bUseGravityOverride=true
    gravityScale=0.35
    simulationSpace=RootRelative
    alphaBoolBlend {blendOption=Linear}
    boundBone {boneName=dyn_Top_Hair_2}
    componentAppliedLinearAccClamp=<5000 5000 5000>
    componentLinearAccScale=<0.5 0.5 0.5>
    componentLinearVelScale=<0.5 0.5 0.5>
    gravityOverride {z=-980}
    physicsBodyDefinitions [{
      boundBone {boneName=dyn_Top_Hair_2}
      boxExtents=<20 30 30>
      constraintSetup {angularLimitsMax=<0 0 2>, angularLimitsMin=<0 0 -2>, angularTarget=<1 0 0>}
      localJointOffset=<2 0 0>
     }]
   }
   n90 = Anim Dynamics - Bone: dyn_Top_Hair_3!(ComponentPose=@n89, Alpha=1, GravityOverride:Vector="0,0,-980", ExternalForce="0,0,0") -> (Pose) {
    angularDampingOverride=1
    angularSpringConstant=10000
    bAngularSpring=true
    bOverrideAngularDamping=true
    bUseGravityOverride=true
    gravityScale=0.35
    simulationSpace=RootRelative
    alphaBoolBlend {blendOption=Linear}
    boundBone {boneName=dyn_Top_Hair_3}
    componentAppliedLinearAccClamp=<5000 5000 5000>
    componentLinearAccScale=<0.5 0.5 0.5>
    componentLinearVelScale=<0.5 0.5 0.5>
    gravityOverride {z=-980}
    physicsBodyDefinitions [{
      boundBone {boneName=dyn_Top_Hair_3}
      boxExtents=<20 30 30>
      constraintSetup {angularLimitsMax=<0 0 2>, angularLimitsMin=<0 0 -2>, angularTarget=<1 0 0>}
      localJointOffset=<2 0 0>
     }]
   }
   n91 = Anim Dynamics - Bone: dyn_FrntHair_A_B_1!(ComponentPose=@n90, Alpha=1, GravityOverride:Vector="0,0,-980", ExternalForce="0,0,0") -> (Pose) {
    angularDampingOverride=1
    angularSpringConstant=15000
    bAngularSpring=true
    bOverrideAngularDamping=true
    bUseGravityOverride=true
    gravityScale=0.35
    simulationSpace=RootRelative
    alphaBoolBlend {blendOption=Linear}
    boundBone {boneName=dyn_FrntHair_A_B_1}
    componentLinearAccScale=<1 1 1>
    componentLinearVelScale=<1 1 1>
    gravityOverride {z=-980}
    physicsBodyDefinitions [{
      boundBone {boneName=dyn_FrntHair_A_B_1}
      boxExtents=<20 30 30>
      constraintSetup {angularLimitsMax=<0 10 10>, angularLimitsMin=<0 -10 -10>, angularTarget=<1 0 0>}
      localJointOffset=<4 0 0>
     }]
   }
   n92 = Anim Dynamics - Bone: dyn_FrntHair_A_B_2!(ComponentPose=@n91, Alpha=1, GravityOverride:Vector="0,0,-980", ExternalForce="0,0,0") -> (Pose) {
    angularDampingOverride=1
    angularSpringConstant=15000
    bAngularSpring=true
    bOverrideAngularDamping=true
    bUseGravityOverride=true
    gravityScale=0.35
    simulationSpace=RootRelative
    alphaBoolBlend {blendOption=Linear}
    boundBone {boneName=dyn_FrntHair_A_B_2}
    componentLinearAccScale=<1 1 1>
    componentLinearVelScale=<1 1 1>
    gravityOverride {z=-980}
    physicsBodyDefinitions [{
      boundBone {boneName=dyn_FrntHair_A_B_2}
      boxExtents=<20 30 30>
      constraintSetup {angularLimitsMax=<0 10 10>, angularLimitsMin=<0 -10 -10>, angularTarget=<1 0 0>}
      localJointOffset=<4 0 0>
     }]
   }
   n93 = Anim Dynamics - Bone: dyn_Back_Hair_C_A!(ComponentPose=@n92, Alpha=1, GravityOverride:Vector="0,0,-980", ExternalForce="0,0,0") -> (Pose) {
    angularDampingOverride=1
    angularSpringConstant=5000
    bAngularSpring=true
    bOverrideAngularDamping=true
    bUseGravityOverride=true
    gravityScale=0.35
    simulationSpace=RootRelative
    alphaBoolBlend {blendOption=Linear}
    boundBone {boneName=dyn_Back_Hair_C_A}
    componentAppliedLinearAccClamp=<5000 5000 5000>
    componentLinearAccScale=<0.5 0.5 0.5>
    componentLinearVelScale=<0.5 0.5 0.5>
    gravityOverride {z=-980}
    physicsBodyDefinitions [{
      boundBone {boneName=dyn_Back_Hair_C_A}
      boxExtents=<15 15 15>
      constraintSetup {angularLimitsMax=<0 5 0>, angularLimitsMin=<0 -5 0>, angularTarget=<1 0 0>}
      localJointOffset=<1 0 0>
     }]
   }
   n94 = Anim Dynamics - Bone: dyn_Back_Hair_A_A!(ComponentPose=@n93, Alpha=1, GravityOverride:Vector="0,0,-980", ExternalForce="0,0,0") -> (Pose) {
    angularDampingOverride=1
    angularSpringConstant=5000
    bAngularSpring=true
    bOverrideAngularDamping=true
    bUseGravityOverride=true
    gravityScale=0.35
    simulationSpace=RootRelative
    alphaBoolBlend {blendOption=Linear}
    boundBone {boneName=dyn_Back_Hair_A_A}
    componentAppliedLinearAccClamp=<5000 5000 5000>
    componentLinearAccScale=<0.5 0.5 0.5>
    componentLinearVelScale=<0.5 0.5 0.5>
    gravityOverride {z=-980}
    physicsBodyDefinitions [{
      boundBone {boneName=dyn_Back_Hair_A_A}
      boxExtents=<15 15 15>
      constraintSetup {angularLimitsMax=<0 10 0>, angularLimitsMin=<0 -10 0>, angularTarget=<1 0 0>}
      localJointOffset=<1 0 0>
     }]
   }
   n95 = Anim Dynamics - Bone: dyn_FrntHair_A_3(ComponentPose=@n94, Alpha=1, GravityOverride="0,0,-980", ExternalForce="0,0,0") {
    angularDampingOverride=1
    angularSpringConstant=7000
    bAngularSpring=true
    bOverrideAngularDamping=true
    bUseGravityOverride=true
    gravityScale=0.35
    simulationSpace=RootRelative
    alphaBoolBlend {blendOption=Linear}
    boundBone {boneName=dyn_FrntHair_A_3}
    gravityOverride {z=-980}
    physicsBodyDefinitions [{
      boundBone {boneName=dyn_FrntHair_A_3}
      boxExtents=<15 15 15>
      constraintSetup {angularLimitsMax=<0 0 2>, angularLimitsMin=<0 0 -2>, angularTarget=<1 0 0>}
      localJointOffset=<1 0 0>
     }]
   }
   n96 = Anim Dynamics - Bone: dyn_FrntHair_A_4(ComponentPose=@n95, Alpha=1, GravityOverride="0,0,-980", ExternalForce="0,0,0") {
    angularDampingOverride=1
    angularSpringConstant=7000
    bAngularSpring=true
    bOverrideAngularDamping=true
    bUseGravityOverride=true
    gravityScale=0.35
    simulationSpace=RootRelative
    alphaBoolBlend {blendOption=Linear}
    boundBone {boneName=dyn_FrntHair_A_4}
    gravityOverride {z=-980}
    physicsBodyDefinitions [{
      boundBone {boneName=dyn_FrntHair_A_4}
      boxExtents=<15 15 15>
      constraintSetup {angularLimitsMax=<0 0 2>, angularLimitsMin=<0 0 -2>, angularTarget=<1 0 0>}
      localJointOffset=<1 0 0>
     }]
   }
   n97 = Anim Dynamics - Bone: dyn_FrntHair_B_2(ComponentPose=@n96, Alpha=1, GravityOverride="0,0,-980", ExternalForce="0,0,0") {
    angularDampingOverride=1
    angularSpringConstant=10000
    bAngularSpring=true
    bOverrideAngularDamping=true
    bUseGravityOverride=true
    bUseSphericalLimits=true
    gravityScale=0.35
    simulationSpace=RootRelative
    alphaBoolBlend {blendOption=Linear}
    boundBone {boneName=dyn_FrntHair_B_2}
    componentAppliedLinearAccClamp=<5000 5000 5000>
    componentLinearAccScale=<0.5 0.5 0.5>
    componentLinearVelScale=<0.5 0.5 0.5>
    gravityOverride {z=-980}
    physicsBodyDefinitions [{
      boundBone {boneName=dyn_FrntHair_B_2}
      boxExtents=<15 15 15>
      constraintSetup {angularLimitsMax=<0 0 2>, angularLimitsMin=<0 0 -2>, angularTarget=<1 0 0>}
      localJointOffset=<1 0 0>
     }]
    sphericalLimits [{limitRadius=10, drivingBone {boneName=head}, sphereLocalOffset=<4.8 1.3 7>}]
   }
   n98 = Anim Dynamics - Bone: dyn_FrntHair_B_3(ComponentPose=@n97, Alpha=1, GravityOverride="0,0,-980", ExternalForce="0,0,0") {
    angularDampingOverride=1
    angularSpringConstant=10000
    bAngularSpring=true
    bOverrideAngularDamping=true
    bUseGravityOverride=true
    bUseSphericalLimits=true
    gravityScale=0.35
    simulationSpace=RootRelative
    alphaBoolBlend {blendOption=Linear}
    boundBone {boneName=dyn_FrntHair_B_3}
    componentAppliedLinearAccClamp=<5000 5000 5000>
    componentLinearAccScale=<0.5 0.5 0.5>
    componentLinearVelScale=<0.5 0.5 0.5>
    gravityOverride {z=-980}
    physicsBodyDefinitions [{
      boundBone {boneName=dyn_FrntHair_B_3}
      boxExtents=<15 15 15>
      constraintSetup {angularLimitsMax=<0 0 2>, angularLimitsMin=<0 0 -2>, angularTarget=<1 0 0>}
      localJointOffset=<1 0 0>
     }]
    sphericalLimits [{limitRadius=10, drivingBone {boneName=head}, sphereLocalOffset=<4.8 1.3 0.7>}]
   }
   n99 = Anim Dynamics - Bone: dyn_FrntHair_C_1(ComponentPose=@n98, Alpha=1, GravityOverride="0,0,-980", ExternalForce="0,0,0") {
    angularDampingOverride=1
    angularSpringConstant=10000
    bAngularSpring=true
    bOverrideAngularDamping=true
    bUseGravityOverride=true
    gravityScale=0.35
    simulationSpace=RootRelative
    alphaBoolBlend {blendOption=Linear}
    boundBone {boneName=dyn_FrntHair_C_1}
    componentAppliedLinearAccClamp=<5000 5000 5000>
    componentLinearAccScale=<0.5 0.5 0.5>
    componentLinearVelScale=<0.5 0.5 0.5>
    gravityOverride {z=-980}
    physicsBodyDefinitions [{
      boundBone {boneName=dyn_FrntHair_C_1}
      boxExtents=<15 15 15>
      constraintSetup {angularLimitsMax=<0 0 1>, angularLimitsMin=<0 0 -1>, angularTarget=<1 0 0>}
      localJointOffset=<1 0 0>
     }]
   }
   n100 = Anim Dynamics - Bone: dyn_FrntHair_C_2(ComponentPose=@n99, Alpha=1, GravityOverride="0,0,-980", ExternalForce="0,0,0") {
    angularDampingOverride=1
    angularSpringConstant=10000
    bAngularSpring=true
    bOverrideAngularDamping=true
    bUseGravityOverride=true
    bUseSphericalLimits=true
    gravityScale=0.35
    simulationSpace=RootRelative
    alphaBoolBlend {blendOption=Linear}
    boundBone {boneName=dyn_FrntHair_C_2}
    componentAppliedLinearAccClamp=<5000 5000 5000>
    componentLinearAccScale=<0.5 0.5 0.5>
    componentLinearVelScale=<0.5 0.5 0.5>
    gravityOverride {z=-980}
    physicsBodyDefinitions [{
      boundBone {boneName=dyn_FrntHair_C_2}
      boxExtents=<15 15 15>
      constraintSetup {angularLimitsMax=<0 0 1>, angularLimitsMin=<0 0 -1>, angularTarget=<1 0 0>}
      localJointOffset=<1 0 0>
     }]
    sphericalLimits [{limitRadius=10.5, drivingBone {boneName=head}, sphereLocalOffset=<4.9173 0.996 1.789>}]
   }
   n101 = Anim Dynamics - Bone: dyn_earring_l!(ComponentPose=@n100, Alpha=1, GravityOverride:Vector="0,0,-980", ExternalForce="0,0,0") -> (Pose) {
    angularDampingOverride=1
    angularSpringConstant=5000
    bAngularSpring=true
    bOverrideAngularDamping=true
    bUseGravityOverride=true
    gravityScale=0.35
    simulationSpace=RootRelative
    alphaBoolBlend {blendOption=Linear}
    boundBone {boneName=dyn_earring_l}
    gravityOverride {z=-980}
    physicsBodyDefinitions [{
      boundBone {boneName=dyn_earring_l}
      boxExtents=<50 50 50>
      constraintSetup {angularLimitsMax=<30 30 30>, angularLimitsMin=<-30 -30 -30>, angularTarget=<1 0 0>}
      localJointOffset=<5 0 0>
     }]
   }
   n102 = Anim Dynamics - Bone: dyn_earring_r!(ComponentPose=@n101, Alpha=1, GravityOverride:Vector="0,0,-980", ExternalForce="0,0,0") -> (Pose) {
    angularDampingOverride=1
    angularSpringConstant=5000
    bAngularSpring=true
    bOverrideAngularDamping=true
    bUseGravityOverride=true
    gravityScale=0.35
    simulationSpace=RootRelative
    alphaBoolBlend {blendOption=Linear}
    boundBone {boneName=dyn_earring_r}
    gravityOverride {z=-980}
    physicsBodyDefinitions [{
      boundBone {boneName=dyn_earring_r}
      boxExtents=<50 50 50>
      constraintSetup {angularLimitsMax=<30 30 30>, angularLimitsMin=<-30 -30 -30>, angularTarget=<1 0 0>}
      localJointOffset=<5 0 0>
     }]
   }
   Save cached pose 'Cache_StartPose'(Pose=Component To Local(ComponentPose=@n102)) {cacheName=Cache_StartPose}
  ]}
 EventGraph {nodes [
   Event BlueprintUpdateAnimation()
   > n1 = Sequence!(execute) -> (then_0, then_1, then_2:Exec, then_3:Exec, then_4:Exec, then_5:Exec, then_6:Exec, then_7:Exec, then_8:Exec, then_9:Exec)
   n2 = "To Float (Boolean)"(InBool=Get LongW Equipped() -> (LongW Equipped:bool))
   n3 = float * float(
    A=GetAlphaFromReferenceTransformDelta(SkeletalMeshComponent=Get Skel Mesh(), BoneName=spine_05, BaseName=head, InputRange=(X=0,Y=90), TwistAxis="0,0,1")
    B=3.3
   ) {is=title K2Node_PromotableOperator}
   Set RiffleAimingUp(execute=@n1.then_5, RiffleAimingUp:double="Clamp (Float)"(Value=float * float(A=@n3, B=@n2), Min=0, Max=1)) -> (Output_Get:double)
   Branch(execute=@n1.then_7, Condition=Get bIsInFrontEnd())
   > Set Custom Idle is Playing(Custom Idle is Playing:double=GetCurveValue(self=GetOwnerFrontEndAnimInstance(), CurveName=Frontend_Idle_Bypass)) -> (Output_Get:double)
   n4 = "Gravity Override State Function (Parameter Asset) (Nativized)"(
    execute=@n1.then_0
    CharacterPartAnimInstance=Self-Reference()
    GameTimeInSeconds=GetGameTimeInSeconds()
    Parameters=&/F_MED_MutedMaroon_FaceAcc_Parameters
   )
   > Set GravityOverride(GravityOverride=@n4)
   n5 = CheckForBackbling()
   > Set isWearingBackbling(isWearingBackbling=AND Boolean!(A=@n5, B=NOT Boolean(A=@n5.bIsBackblingACape), C:bool=NOT Boolean(A=@n5.bIsBackblingLong)) -> (ReturnValue))
   > Set Is Wearing Cape(Is Wearing Cape=OR Boolean(A=@n5.bIsBackblingACape, B=@n5.bIsBackblingLong))
   n6 = Get Skel Mesh()
   Set Head Down(
    execute=@n1.then_4
    Head Down:double=GetAlphaFromReferenceTransformDelta(SkeletalMeshComponent=@n6, BoneName=spine_05, BaseName="head ", InputRange=(X=0,Y=-90), OutputRange=(X=0,Y=-1), TwistAxis="0,0,1")
   ) -> (Output_Get:double)
   > Set Head_Turn_L(Head_Turn_L:double=GetAlphaFromReferenceTransformDelta(SkeletalMeshComponent=@n6, BoneName=spine_05, BaseName=head, DeltaComponentType=EulerX, InputRange=(X=0,Y=-22), TwistAxis="1,0,0")) -> (Output_Get:double)
   > Set Head_Turn_R(Head_Turn_R:double=GetAlphaFromReferenceTransformDelta(SkeletalMeshComponent=@n6, BoneName=spine_05, BaseName=head, DeltaComponentType=EulerX, InputRange=(X=0,Y=22), TwistAxis="1,0,0")) -> (Output_Get:double)
   n7 = GetAlphaFromReferenceTransformDelta(SkeletalMeshComponent=@n6, BoneName=spine_05, BaseName=head, InputRange=(X=10,Y=70), TwistAxis="0,0,1")
   Set HeadUpFloat(execute=@n1.then_3, HeadUpFloat:double=@n7) -> (Output_Get:double)
   n8 = float * float(A=@n7, B=3)
   > n9 = Set HeadUp(HeadUp:bool="To Boolean (Integer)"!(InInt:int=Truncate(A=@n8)) -> (ReturnValue:bool)) -> (Output_Get:bool)
   n10 = NOT Boolean(A=Get bIsCrouching())
   n11 = Get Pistol Equipped() -> (Pistol Equipped:bool)
   > Set AimingUp_ShortW(AimingUp_ShortW:double="To Float (Boolean)"(InBool=AND Boolean(A=@n9, B=@n11, C=@n10))) -> (Output_Get:double)
   n12 = NOT Boolean(A=Get bIsMoving2D())
   n13 = Get LongW Equipped()
   n14 = float * float(A=@n7, B=2)
   n15 = AND Boolean(A=OR Boolean(A=Get isWearingBackbling(), B=Get Is Wearing Cape()), B="To Boolean (Integer)"(InInt=Truncate(A=@n14)), C=Get bIsCrouching())
   n16 = Branch(execute=@n1.then_2, Condition=AND Boolean(A=@n15, B=@n13, C=@n12))
   > Delay(Duration=0.4)
   > Set Crouching_BackpackCape_LW(Crouching_BackpackCape_LW:double=1) -> (Output_Get:double)
   Set Crouching_BackpackCape_LW(execute=@n16.else, Crouching_BackpackCape_LW=0)
   > n17 = Branch(Condition=AND Boolean(A=@n15, B=NOT Boolean(A=@n13), C=@n12))
   > Delay(Duration=0.4)
   > Set Crouching_BackpackCape_SW(Crouching_BackpackCape_SW:double=1) -> (Output_Get:double)
   Set Crouching_BackpackCape_SW(execute=@n17.else, Crouching_BackpackCape_SW=0)
   n18 = Cast To FortPlayerAnimInstance(execute=@n1.then_6, Object=GetAnimInstance(self=GetOwnerSkeletalMeshComponent()))
   > Set LongW Equipped(LongW Equipped:bool=false) -> (Output_Get:bool)
   > n19 = Switch on EFortWeaponCoreAnimation!(execute, Selection:EFortWeaponCoreAnimation Enum=Get WeaponCoreAnim(self=@n18)) -> (
    Melee:Exec
    Pistol:Exec
    Shotgun:Exec
    PaperBlueprint:Exec
    Rifle:Exec
    MeleeOneHand:Exec
    MachinePistol:Exec
    RocketLauncher:Exec
    GrenadeLauncher:Exec
    GoingCommando:Exec
    AssaultRifle:Exec
    TacticalShotgun:Exec
    SniperRifle:Exec
    TrapPlacement:Exec
    ShoulderLauncher:Exec
    AbilityDecoTool:Exec
    Crossbow:Exec
    C4:Exec
    RemoteControl:Exec
    DualWield:Exec
    AR_BullPup:Exec
    AR_ForwardGrip:Exec
    MedPackPaddles:Exec
    SMG_P90:Exec
    AR_DrumGun:Exec
    Consumable_Small:Exec
    Consumable_Large:Exec
    Balloon:Exec
    MountedTurret:Exec
    CreativeTool:Exec
    ExplosiveBow:Exec
    AshtonIndigo:Exec
    AshtonChicago:Exec
    MeleeDualWield:Exec
    Unarmed:Exec
   )
   n20 = Branch(
    execute=[@n19.AssaultRifle, @n19.GoingCommando, @n19.GrenadeLauncher, @n19.Rifle, @n19.Shotgun, @n19.SniperRifle, @n19.TacticalShotgun]
    Condition=Get bIsTargeting()
   )
   > Set LongW Equipped(LongW Equipped=true)
   Set LongW Equipped(execute=@n20.else, LongW Equipped=false)
   n21 = Branch(execute=[@n19.DualWield, @n19.Pistol], Condition=Get bIsTargeting())
   > Set Pistol Equipped(Pistol Equipped:bool=true) -> (Output_Get:bool)
   Set Pistol Equipped(execute=@n21.else, Pistol Equipped=false)
   Event OnCharacterCustomizationLoaded()
   > n22 = Sequence!(execute) -> (then_0, then_1)
   Delay(execute=@n22.then_1, Duration=0.2)
   > ResetDynamics(InTeleportType=TeleportPhysics)
   > ResetDynamics(InTeleportType=ResetPhysics)
   n23 = Cast To PlayerPawn_Athena(execute=@n22.then_0, Object=TryGetPawnOwner())
   > Set Body_Mesh(Body_Mesh=GetSkeletalMeshForPartType(self=@n23))
   Event BlueprintInitializeAnimation()
   > Set Skel Mesh(Skel Mesh=GetOwnerSkeletalMeshComponent())
   GetAlphaFromReferenceTransformDelta(BoneName=spine_05, BaseName=head, DeltaComponentType=EulerX, InputRange=(X=0,Y=22), TwistAxis="1,0,0")
   GetAlphaFromReferenceTransformDelta(BoneName=spine_05, BaseName=head, DeltaComponentType=EulerX, InputRange=(X=0,Y=-22), TwistAxis="1,0,0")
  ]}
]
variables [
 AimingUp_ShortW {type=number}
 Body_Mesh {type=SkeletalMeshComponent}
 Crouching_BackpackCape_LW {type=number}
 Crouching_BackpackCape_SW {type=number}
 Custom Idle is Playing {type=number}
 GravityOverride {type=Vector}
 Head Down {type=number}
 Head_Turn_L {type=number}
 Head_Turn_R {type=number}
 HeadUp {type=boolean}
 HeadUpFloat {type=number}
 Is Wearing Cape {type=boolean}
 isParachuting_alpha {type=number}
 isWearingBackbling {type=boolean}
 LongW Equipped {type=boolean}
 Pistol Equipped {type=boolean}
 RiffleAimingUp {type=number}
 Skel Mesh {type=SkeletalMeshComponent}
]
