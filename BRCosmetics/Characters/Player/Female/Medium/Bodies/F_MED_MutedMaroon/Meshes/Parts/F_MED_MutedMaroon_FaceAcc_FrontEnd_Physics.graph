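# Physics asset for the front-end variant of the F_MED_MutedMaroon face accessory, setting up dynamic back hair on a medium female character.
# Left, middle and right chains of six capsule bones hang from base bones locked to spine_05, with swing and twist limits set per joint.
# Free constraints tie the outer chain tips to the middle tip.
# Capsules on the upper spine, clavicles and upper arms form the body shapes around the hair.
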PhysicsAsset /BRCosmetics/Characters/Player/Female/Medium/Bodies/F_MED_MutedMaroon/Meshes/Parts/F_MED_MutedMaroon_FaceAcc_FrontEnd_Physics
bodies {
 base_hair_back_l [{shapeType=Capsule, length=0.5, radius=1, shapeName=base_hair_back_l_capsule, rotation=<rot 90 0 0>}]
 base_hair_back_m [{shapeType=Capsule, length=0.5, radius=1, shapeName=base_hair_back_m_capsule, rotation=<rot 90 0 0>}]
 base_hair_back_r [{shapeType=Capsule, length=0.5, radius=1, shapeName=base_hair_back_r_capsule, rotation=<rot 90 0 0>}]
 clavicle_l [{shapeType=Capsule, length=4.9651, radius=5.3616, shapeName=clavicle_l_capsule, center=<3 0 0>, rotation=<rot 90 0 0>}]
 clavicle_r [{shapeType=Capsule, length=4.9651, radius=5.3616, shapeName=clavicle_r_capsule, center=<-3 0 0>, rotation=<rot 90 0 0>}]
 dyn_hair_back_01_l [{shapeType=Capsule, length=5, radius=2, shapeName=dyn_hair_back_01_l_capsule, center=<4.4493 0 0>, rotation=<rot 90 0 0>}]
 dyn_hair_back_01_m [{shapeType=Capsule, length=5, radius=2, shapeName=dyn_hair_back_01_m_capsule, center=<4.038 0 0>, rotation=<rot 90 0 0>}]
 dyn_hair_back_01_r [{shapeType=Capsule, length=5, radius=2, shapeName=dyn_hair_back_01_r_capsule, center=<4.3344 0 0>, rotation=<rot 90 0 0>}]
 dyn_hair_back_02_l [{shapeType=Capsule, length=3, radius=3.9, shapeName=dyn_hair_back_02_l_capsule, center=<2.73 0 0>, rotation=<rot 80 -165.9638 -155.9638>}]
 dyn_hair_back_02_m [{shapeType=Capsule, length=6, radius=3, shapeName=dyn_hair_back_02_m_capsule, center=<3 0 0>}]
 dyn_hair_back_02_r [{shapeType=Capsule, length=2, radius=3.5, shapeName=dyn_hair_back_02_r_capsule, center=<3.9597 0 0>, rotation=<rot 80 0 0>}]
 dyn_hair_back_03_l [{shapeType=Capsule, length=4, radius=3.9, shapeName=dyn_hair_back_04_r_capsule, center=<4.5849 -0.5 1>}]
 dyn_hair_back_03_m [{shapeType=Capsule, length=9, radius=3.5, shapeName=dyn_hair_back_03_m_capsule, center=<3.2794 0 0>}]
 dyn_hair_back_03_r [{shapeType=Capsule, length=6.5811, radius=4, shapeName=dyn_hair_back_04_r_capsule, center=<4.5922 -0.5 -0.25>}]
 dyn_hair_back_04_l [{shapeType=Capsule, length=4, radius=3.9, shapeName=dyn_hair_back_04_r_capsule, center=<4 -1 0.5>}]
 dyn_hair_back_04_m [{shapeType=Capsule, length=9, radius=4, shapeName=dyn_hair_back_04_m_capsule, center=<3.7441 -1.8 0>}]
 dyn_hair_back_04_r [{shapeType=Capsule, length=6.5811, radius=4, shapeName=dyn_hair_back_04_r_capsule, center=<4 -1 0>}]
 dyn_hair_back_05_l [{shapeType=Capsule, length=6, radius=3.9, shapeName=dyn_hair_back_04_r_capsule, center=<4 -1 -0.5>}]
 dyn_hair_back_05_m [{shapeType=Capsule, length=6, radius=4, shapeName=dyn_hair_back_04_m_capsule, center=<4 -1.6 0>}]
 dyn_hair_back_05_r [
  {shapeType=Capsule, length=8.5811, radius=4, shapeName=dyn_hair_back_04_r_capsule, center=<4 -1 1>}
  {
   shapeType=Capsule
   length=6.75
   radius=3.375
   shapeName=dyn_hair_back_05_r_capsule
   center=<4.6226 4.2151 -0.4594>
   rotation=<rot 70 0.0002 0>
  }
 ]
 dyn_hair_back_06_l [{shapeType=Capsule, length=11, radius=2.9, shapeName=dyn_hair_back_04_l_capsule, center=<4.0725 -1.3 -3>, rotation=<rot 45 0 0>}]
 dyn_hair_back_06_m [{shapeType=Capsule, length=11, radius=3, shapeName=dyn_hair_back_04_l_capsule, center=<7 -1.3 0>, rotation=<rot 90 0 0>}]
 dyn_hair_back_06_r [{shapeType=Capsule, length=11, radius=3, shapeName=dyn_hair_back_04_l_capsule, center=<5.2884 -1.3 3.8257>, rotation=<rot -38.0226 0.0001 0>}]
 spine_04 [{shapeType=Capsule, length=43.5, radius=12.5, shapeName=spine_04_capsule, center=<-9.4227 1.0507 0>}]
 spine_05 [
  {shapeType=Capsule, length=47.125, radius=9.975, shapeName=spine_05_capsule, center=<-3.3009 4.4524 0>}
  {shapeType=Capsule, length=45.5, radius=13.25, shapeName=spine_05_capsule, center=<9.687 14.3909 0>}
 ]
 upperarm_l [{shapeType=Capsule, length=35.4298, radius=6.7768, shapeName=upperarm_l_capsule, center=<9.0117 4.5168 -0.4323>, rotation=<rot 90 0 0>}]
 upperarm_r [{shapeType=Capsule, length=35.4298, radius=6.7768, shapeName=upperarm_r_capsule, center=<-9.0117 -4.5168 0.4323>, rotation=<rot 90 180 0>}]
}
constraints [
 {bone1Name=dyn_hair_back_01_l, bone2Name=base_hair_back_l, swing1LimitDegrees=10, swing2LimitDegrees=1, twistLimitDegrees=0}
 {
  bone1Name=dyn_hair_back_02_l
  bone2Name=dyn_hair_back_01_l
  swing1LimitDegrees=20
  swing2LimitDegrees=2
  twistLimitDegrees=1
  twistMotion=Locked
 }
 {
  bone1Name=dyn_hair_back_03_l
  bone2Name=dyn_hair_back_02_l
  swing1LimitDegrees=33
  swing2LimitDegrees=8
  twistLimitDegrees=1
  twistMotion=Locked
 }
 {
  bone1Name=dyn_hair_back_04_l
  bone2Name=dyn_hair_back_03_l
  swing1LimitDegrees=18
  swing2LimitDegrees=5
  twistLimitDegrees=1
  twistMotion=Locked
 }
 {bone1Name=dyn_hair_back_05_l, bone2Name=dyn_hair_back_04_l, swing1LimitDegrees=0, swing2LimitDegrees=7, twistLimitDegrees=1}
 {bone1Name=dyn_hair_back_01_m, bone2Name=base_hair_back_m, swing1LimitDegrees=10, swing2LimitDegrees=1, twistLimitDegrees=0}
 {
  bone1Name=dyn_hair_back_02_m
  bone2Name=dyn_hair_back_01_m
  swing1LimitDegrees=20
  swing2LimitDegrees=2
  twistLimitDegrees=1
  twistMotion=Locked
 }
 {bone1Name=dyn_hair_back_03_m, bone2Name=dyn_hair_back_02_m, swing1LimitDegrees=33, swing2LimitDegrees=8, twistLimitDegrees=1}
 {bone1Name=dyn_hair_back_04_m, bone2Name=dyn_hair_back_03_m, swing1LimitDegrees=18, swing2LimitDegrees=10, twistLimitDegrees=10}
 {bone1Name=dyn_hair_back_05_m, bone2Name=dyn_hair_back_04_m, swing1LimitDegrees=0, swing2LimitDegrees=7, twistLimitDegrees=1}
 {bone1Name=dyn_hair_back_01_r, bone2Name=base_hair_back_r, swing1LimitDegrees=10, swing2LimitDegrees=1, twistLimitDegrees=0}
 {
  bone1Name=dyn_hair_back_02_r
  bone2Name=dyn_hair_back_01_r
  swing1LimitDegrees=20
  swing2LimitDegrees=2
  twistLimitDegrees=1
  twistMotion=Locked
 }
 {
  bone1Name=dyn_hair_back_03_r
  bone2Name=dyn_hair_back_02_r
  swing1LimitDegrees=33
  swing2LimitDegrees=8
  twistLimitDegrees=1
  twistMotion=Locked
 }
 {
  bone1Name=dyn_hair_back_04_r
  bone2Name=dyn_hair_back_03_r
  swing1LimitDegrees=18
  swing2LimitDegrees=5
  twistLimitDegrees=1
  twistMotion=Locked
 }
 {bone1Name=dyn_hair_back_05_r, bone2Name=dyn_hair_back_04_r, swing1LimitDegrees=0, swing2LimitDegrees=7, twistLimitDegrees=1}
 {bone1Name=dyn_hair_back_06_l, bone2Name=dyn_hair_back_05_l, swing1LimitDegrees=0, swing2LimitDegrees=7, twistLimitDegrees=1}
 {bone1Name=dyn_hair_back_06_m, bone2Name=dyn_hair_back_05_m, swing1LimitDegrees=0, swing2LimitDegrees=7, twistLimitDegrees=1}
 {bone1Name=dyn_hair_back_06_r, bone2Name=dyn_hair_back_05_r, swing1LimitDegrees=0, swing2LimitDegrees=7, twistLimitDegrees=1}
 {
  bone1Name=dyn_hair_back_06_r
  bone2Name=dyn_hair_back_06_m
  swing1LimitDegrees=45
  swing1Motion=Free
  swing2LimitDegrees=45
  swing2Motion=Free
  twistLimitDegrees=45
  twistMotion=Free
 }
 {
  bone1Name=dyn_hair_back_06_l
  bone2Name=dyn_hair_back_06_m
  swing1LimitDegrees=45
  swing1Motion=Free
  swing2LimitDegrees=45
  swing2Motion=Free
  twistLimitDegrees=45
  twistMotion=Free
 }
 {
  bone1Name=base_hair_back_l
  bone2Name=spine_05
  swing1LimitDegrees=10
  swing1Motion=Locked
  swing2LimitDegrees=10
  swing2Motion=Locked
  twistLimitDegrees=5
  twistMotion=Locked
 }
 {
  bone1Name=base_hair_back_r
  bone2Name=spine_05
  swing1LimitDegrees=10
  swing1Motion=Locked
  swing2LimitDegrees=10
  swing2Motion=Locked
  twistLimitDegrees=5
  twistMotion=Locked
 }
 {
  bone1Name=base_hair_back_m
  bone2Name=spine_05
  swing1LimitDegrees=10
  swing1Motion=Locked
  swing2LimitDegrees=10
  swing2Motion=Locked
  twistLimitDegrees=5
  twistMotion=Locked
 }
 {
  bone1Name=spine_05
  bone2Name=spine_04
  swing1LimitDegrees=45
  swing1Motion=Locked
  swing2LimitDegrees=45
  swing2Motion=Locked
  twistLimitDegrees=45
  twistMotion=Locked
 }
 {
  bone1Name=clavicle_l
  bone2Name=spine_05
  swing1LimitDegrees=45
  swing1Motion=Free
  swing2LimitDegrees=45
  swing2Motion=Free
  twistLimitDegrees=45
  twistMotion=Free
 }
 {
  bone1Name=upperarm_l
  bone2Name=clavicle_l
  swing1LimitDegrees=45
  swing1Motion=Free
  swing2LimitDegrees=45
  swing2Motion=Free
  twistLimitDegrees=45
  twistMotion=Free
 }
 {
  bone1Name=upperarm_r
  bone2Name=clavicle_r
  swing1LimitDegrees=45
  swing1Motion=Free
  swing2LimitDegrees=45
  swing2Motion=Free
  twistLimitDegrees=45
  twistMotion=Free
 }
 {
  bone1Name=clavicle_r
  bone2Name=spine_05
  swing1LimitDegrees=45
  swing1Motion=Free
  swing2LimitDegrees=45
  swing2Motion=Free
  twistLimitDegrees=45
  twistMotion=Free
 }
]
