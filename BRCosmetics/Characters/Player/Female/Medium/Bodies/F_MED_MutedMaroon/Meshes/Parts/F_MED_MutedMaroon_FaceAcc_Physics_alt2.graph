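# Physics asset for the MutedMaroon female medium body's FaceAcc part (alt2), setting up back-hair dynamics against capsules for torso, neck, clavicles and upper arms.
# Three six-bone hair chains (left, middle, right) hang from base bones angularly locked to neck_02, with limited swing and locked or near-zero twist.
# Free-angle constraints cross-link side chains to the middle at lower segments; segment 05 duplicates the left link and lacks a right one.
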PhysicsAsset /BRCosmetics/Characters/Player/Female/Medium/Bodies/F_MED_MutedMaroon/Meshes/Parts/F_MED_MutedMaroon_FaceAcc_Physics_alt2
bodies {
 base_hair_back_l [{shapeType=Capsule, length=0.5, radius=1, shapeName=base_hair_back_l_capsule, rotation=<rot 90 0 0>}]
 base_hair_back_m [{shapeType=Capsule, length=0.5, radius=1, shapeName=base_hair_back_m_capsule, rotation=<rot 90 0 0>}]
 base_hair_back_r [{shapeType=Capsule, length=0.5, radius=1, shapeName=base_hair_back_r_capsule, rotation=<rot 90 0 0>}]
 clavicle_l [{shapeType=Capsule, length=7, radius=4, shapeName=clavicle_l_capsule, center=<4 0 0>, rotation=<rot 90 0 0>}]
 clavicle_r [{shapeType=Capsule, length=9, radius=5, shapeName=clavicle_r_capsule, center=<-4 0 0>, rotation=<rot 90 0 0>}]
 dyn_hair_back_01_l [{shapeType=Capsule, length=1, radius=6, shapeName=dyn_hair_back_01_l_capsule, center=<5 -2 0>}]
 dyn_hair_back_01_m [{shapeType=Capsule, length=1, radius=6, shapeName=dyn_hair_back_01_m_capsule, center=<5 -2 0>}]
 dyn_hair_back_01_r [{shapeType=Capsule, length=2, radius=7.5, shapeName=dyn_hair_back_01_r_capsule, center=<5 -2 0>}]
 dyn_hair_back_02_l [{shapeType=Capsule, length=1, radius=5.5, shapeName=dyn_hair_back_02_l_capsule, center=<5 -2 0>}]
 dyn_hair_back_02_m [{shapeType=Capsule, length=1, radius=6, shapeName=dyn_hair_back_02_m_capsule, center=<5 -2 0>}]
 dyn_hair_back_02_r [{shapeType=Capsule, length=2, radius=6.5, shapeName=dyn_hair_back_02_r_capsule, center=<5 -2 0>}]
 dyn_hair_back_03_l [{shapeType=Capsule, length=1, radius=5.7, shapeName=dyn_hair_back_03_l_capsule, center=<5 -2 0>}]
 dyn_hair_back_03_m [{shapeType=Capsule, length=1, radius=6, shapeName=dyn_hair_back_03_m_capsule, center=<5 -2 0>}]
 dyn_hair_back_03_r [{shapeType=Capsule, length=2, radius=6, shapeName=dyn_hair_back_03_r_capsule, center=<5 -2 0>}]
 dyn_hair_back_04_l [{shapeType=Capsule, length=1, radius=6, shapeName=dyn_hair_back_04_l_capsule, center=<5 -2 0>}]
 dyn_hair_back_04_m [{shapeType=Capsule, length=1, radius=6, shapeName=dyn_hair_back_04_m_capsule, center=<5 -2 0>}]
 dyn_hair_back_04_r [{shapeType=Capsule, length=2, radius=6, shapeName=dyn_hair_back_04_r_capsule, center=<5 -2 0>}]
 dyn_hair_back_05_l [{shapeType=Capsule, length=1, radius=6, shapeName=dyn_hair_back_05_l_capsule, center=<5 -2 0>}]
 dyn_hair_back_05_m [{shapeType=Capsule, length=1, radius=6, shapeName=dyn_hair_back_05_m_capsule, center=<5 -2 0>}]
 dyn_hair_back_05_r [{shapeType=Capsule, length=2, radius=6, shapeName=dyn_hair_back_05_r_capsule, center=<5 -2 0>}]
 dyn_hair_back_06_l [{shapeType=Capsule, length=1, radius=5, shapeName=dyn_hair_back_06_l_capsule, center=<5 -2 0>}]
 dyn_hair_back_06_m [{shapeType=Capsule, length=1, radius=5, shapeName=dyn_hair_back_06_m_capsule, center=<5 -2 0>}]
 dyn_hair_back_06_r [{shapeType=Capsule, length=2, radius=5, shapeName=dyn_hair_back_06_r_capsule, center=<5 -2 0>}]
 neck_01 [{shapeType=Capsule, length=11.5, radius=4.5, shapeName=neck_01_capsule, center=<0 1.5014 0>}]
 neck_02 [{shapeType=Capsule, length=6, radius=4.5, shapeName=neck_02_capsule, center=<0 1.5014 0>}]
 pelvis [{shapeType=Capsule, length=10, radius=14, shapeName=pelvis_capsule, rotation=<rot 0 -90 0>}]
 spine_01 [{shapeType=Capsule, length=10, radius=11, shapeName=spine_01_capsule, rotation=<rot 0 -90 0>}]
 spine_02 [{shapeType=Capsule, length=10, radius=11, shapeName=spine_02_capsule, rotation=<rot 0 -90 0>}]
 spine_03 [{shapeType=Capsule, length=10, radius=11, shapeName=spine_03_capsule, rotation=<rot 0 -90 0>}]
 spine_05 [
  {shapeType=Capsule, length=18, radius=8.5, shapeName=spine_05_capsule, center=<-3.255 -0.3 0>}
  {shapeType=Capsule, length=25.5, radius=11.25, shapeName=spine_05_capsule, center=<9.687 14.3909 0>}
 ]
 upperarm_l [{shapeType=Capsule, length=18.6798, radius=5, shapeName=upperarm_l_capsule, center=<11 0 -0.5>, rotation=<rot 90 0 0>}]
 upperarm_r [{shapeType=Capsule, length=20, radius=4, shapeName=upperarm_r_capsule, center=<-12 0 0.5>, rotation=<rot 90 0 0>}]
}
constraints [
 {
  bone1Name=base_hair_back_l
  bone2Name=neck_02
  swing1LimitDegrees=10
  swing1Motion=Locked
  swing2LimitDegrees=10
  swing2Motion=Locked
  twistLimitDegrees=5
  twistMotion=Locked
 }
 {
  bone1Name=base_hair_back_r
  bone2Name=neck_02
  swing1LimitDegrees=10
  swing1Motion=Locked
  swing2LimitDegrees=10
  swing2Motion=Locked
  twistLimitDegrees=5
  twistMotion=Locked
 }
 {
  bone1Name=base_hair_back_m
  bone2Name=neck_02
  swing1LimitDegrees=10
  swing1Motion=Locked
  swing2LimitDegrees=10
  swing2Motion=Locked
  twistLimitDegrees=5
  twistMotion=Locked
 }
 {
  bone1Name=spine_05
  bone2Name=spine_03
  swing1LimitDegrees=45
  swing1Motion=Locked
  swing2LimitDegrees=45
  swing2Motion=Locked
  twistLimitDegrees=45
  twistMotion=Locked
 }
 {
  bone1Name=clavicle_l
  bone2Name=spine_05
  swing1LimitDegrees=45
  swing1Motion=Free
  swing2LimitDegrees=45
  swing2Motion=Free
  twistLimitDegrees=45
  twistMotion=Free
 }
 {
  bone1Name=upperarm_l
  bone2Name=clavicle_l
  swing1LimitDegrees=45
  swing1Motion=Free
  swing2LimitDegrees=45
  swing2Motion=Free
  twistLimitDegrees=45
  twistMotion=Free
 }
 {
  bone1Name=upperarm_r
  bone2Name=clavicle_r
  swing1LimitDegrees=45
  swing1Motion=Free
  swing2LimitDegrees=45
  swing2Motion=Free
  twistLimitDegrees=45
  twistMotion=Free
 }
 {
  bone1Name=clavicle_r
  bone2Name=spine_05
  swing1LimitDegrees=45
  swing1Motion=Free
  swing2LimitDegrees=45
  swing2Motion=Free
  twistLimitDegrees=45
  twistMotion=Free
 }
 {
  bone1Name=neck_01
  bone2Name=spine_05
  swing1LimitDegrees=45
  swing1Motion=Free
  swing2LimitDegrees=45
  swing2Motion=Free
  twistLimitDegrees=45
  twistMotion=Free
 }
 {
  bone1Name=neck_02
  bone2Name=neck_01
  swing1LimitDegrees=45
  swing1Motion=Free
  swing2LimitDegrees=45
  swing2Motion=Free
  twistLimitDegrees=45
  twistMotion=Free
 }
 {
  bone1Name=dyn_hair_back_01_m
  bone2Name=base_hair_back_m
  swing1LimitDegrees=30
  swing2LimitDegrees=5
  twistLimitDegrees=0
  twistMotion=Locked
 }
 {
  bone1Name=dyn_hair_back_02_m
  bone2Name=dyn_hair_back_01_m
  swing1LimitDegrees=30
  swing2LimitDegrees=10
  twistLimitDegrees=0
  twistMotion=Locked
 }
 {
  bone1Name=dyn_hair_back_03_m
  bone2Name=dyn_hair_back_02_m
  swing1LimitDegrees=30
  swing2LimitDegrees=15
  twistLimitDegrees=0
  twistMotion=Locked
 }
 {bone1Name=dyn_hair_back_04_m, bone2Name=dyn_hair_back_03_m, swing1LimitDegrees=30, swing2LimitDegrees=15, twistLimitDegrees=0.1}
 {bone1Name=dyn_hair_back_05_m, bone2Name=dyn_hair_back_04_m, swing1LimitDegrees=30, swing2LimitDegrees=5, twistLimitDegrees=0.1}
 {bone1Name=dyn_hair_back_06_m, bone2Name=dyn_hair_back_05_m, swing1LimitDegrees=30, swing2LimitDegrees=5, twistLimitDegrees=0.1}
 {
  bone1Name=dyn_hair_back_02_l
  bone2Name=dyn_hair_back_01_l
  swing1LimitDegrees=30
  swing2LimitDegrees=10
  twistLimitDegrees=0
  twistMotion=Locked
 }
 {
  bone1Name=dyn_hair_back_03_l
  bone2Name=dyn_hair_back_02_l
  swing1LimitDegrees=30
  swing2LimitDegrees=15
  twistLimitDegrees=0
  twistMotion=Locked
 }
 {bone1Name=dyn_hair_back_04_l, bone2Name=dyn_hair_back_03_l, swing1LimitDegrees=30, swing2LimitDegrees=15, twistLimitDegrees=0.1}
 {bone1Name=dyn_hair_back_05_l, bone2Name=dyn_hair_back_04_l, swing1LimitDegrees=30, swing2LimitDegrees=5, twistLimitDegrees=0.1}
 {bone1Name=dyn_hair_back_06_l, bone2Name=dyn_hair_back_05_l, swing1LimitDegrees=30, swing2LimitDegrees=5, twistLimitDegrees=0.1}
 {
  bone1Name=dyn_hair_back_01_r
  bone2Name=base_hair_back_r
  swing1LimitDegrees=30
  swing2LimitDegrees=5
  twistLimitDegrees=0
  twistMotion=Locked
 }
 {
  bone1Name=dyn_hair_back_02_r
  bone2Name=dyn_hair_back_01_r
  swing1LimitDegrees=30
  swing2LimitDegrees=10
  twistLimitDegrees=0
  twistMotion=Locked
 }
 {
  bone1Name=dyn_hair_back_03_r
  bone2Name=dyn_hair_back_02_r
  swing1LimitDegrees=30
  swing2LimitDegrees=15
  twistLimitDegrees=0
  twistMotion=Locked
 }
 {bone1Name=dyn_hair_back_04_r, bone2Name=dyn_hair_back_03_r, swing1LimitDegrees=30, swing2LimitDegrees=15, twistLimitDegrees=0.1}
 {bone1Name=dyn_hair_back_05_r, bone2Name=dyn_hair_back_04_r, swing1LimitDegrees=30, swing2LimitDegrees=5, twistLimitDegrees=0.1}
 {bone1Name=dyn_hair_back_06_r, bone2Name=dyn_hair_back_05_r, swing1LimitDegrees=30, swing2LimitDegrees=5, twistLimitDegrees=0.1}
 {
  bone1Name=dyn_hair_back_01_l
  bone2Name=base_hair_back_l
  swing1LimitDegrees=30
  swing2LimitDegrees=5
  twistLimitDegrees=0
  twistMotion=Locked
 }
 {
  bone1Name=spine_02
  bone2Name=spine_01
  swing1LimitDegrees=45
  swing1Motion=Free
  swing2LimitDegrees=45
  swing2Motion=Free
  twistLimitDegrees=45
  twistMotion=Free
 }
 {
  bone1Name=spine_03
  bone2Name=spine_02
  swing1LimitDegrees=45
  swing1Motion=Free
  swing2LimitDegrees=45
  swing2Motion=Free
  twistLimitDegrees=45
  twistMotion=Free
 }
 {
  bone1Name=spine_01
  bone2Name=pelvis
  swing1LimitDegrees=45
  swing1Motion=Free
  swing2LimitDegrees=45
  swing2Motion=Free
  twistLimitDegrees=45
  twistMotion=Free
 }
 {
  bone1Name=dyn_hair_back_06_r
  bone2Name=dyn_hair_back_06_m
  swing1LimitDegrees=45
  swing1Motion=Free
  swing2LimitDegrees=45
  swing2Motion=Free
  twistLimitDegrees=45
  twistMotion=Free
 }
 {
  bone1Name=dyn_hair_back_06_l
  bone2Name=dyn_hair_back_06_m
  swing1LimitDegrees=45
  swing1Motion=Free
  swing2LimitDegrees=45
  swing2Motion=Free
  twistLimitDegrees=45
  twistMotion=Free
 }
 {
  bone1Name=dyn_hair_back_03_r
  bone2Name=dyn_hair_back_03_m
  swing1LimitDegrees=45
  swing1Motion=Free
  swing2LimitDegrees=45
  swing2Motion=Free
  twistLimitDegrees=45
  twistMotion=Free
 }
 {
  bone1Name=dyn_hair_back_03_l
  bone2Name=dyn_hair_back_03_m
  swing1LimitDegrees=45
  swing1Motion=Free
  swing2LimitDegrees=45
  swing2Motion=Free
  twistLimitDegrees=45
  twistMotion=Free
 }
 {
  bone1Name=dyn_hair_back_05_l
  bone2Name=dyn_hair_back_05_m
  swing1LimitDegrees=45
  swing1Motion=Free
  swing2LimitDegrees=45
  swing2Motion=Free
  twistLimitDegrees=45
  twistMotion=Free
 }
 {
  bone1Name=dyn_hair_back_05_l
  bone2Name=dyn_hair_back_05_m
  swing1LimitDegrees=45
  swing1Motion=Free
  swing2LimitDegrees=45
  swing2Motion=Free
  twistLimitDegrees=45
  twistMotion=Free
 }
 {
  bone1Name=dyn_hair_back_04_r
  bone2Name=dyn_hair_back_04_m
  swing1LimitDegrees=45
  swing1Motion=Free
  swing2LimitDegrees=45
  swing2Motion=Free
  twistLimitDegrees=45
  twistMotion=Free
 }
 {
  bone1Name=dyn_hair_back_04_l
  bone2Name=dyn_hair_back_04_m
  swing1LimitDegrees=45
  swing1Motion=Free
  swing2LimitDegrees=45
  swing2Motion=Free
  twistLimitDegrees=45
  twistMotion=Free
 }
]
