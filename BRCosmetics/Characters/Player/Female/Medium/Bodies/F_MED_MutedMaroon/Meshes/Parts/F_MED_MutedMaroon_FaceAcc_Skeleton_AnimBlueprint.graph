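# Anim Blueprint for the MutedMaroon female face accessory skeleton, posing three back-hair chains (left, middle, right) around worn back items.
# Bool blends on cape and backbling state choose additive bone offsets, offsets plus weighted constraints to head, neck, spine and shoulders, or RigidBody simulation.
# The simulation runs in neck_01 space with gravity override, switching physics assets for the front end.
# A Control Rig post-processes the blended pose.
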AnimBlueprint /BRCosmetics/Characters/Player/Female/Medium/Bodies/F_MED_MutedMaroon/Meshes/Parts/F_MED_MutedMaroon_FaceAcc_Skeleton_AnimBlueprint
target_skeleton=&/F_MED_MutedMaroon_FaceAcc_Skeleton
graphs [
 AnimGraph {nodes [
   n1 = "Transform (Modify) Bone - Bone: dyn_hair_back_02_m"!(
    ComponentPose=Local To Component(LocalPose=Use cached pose 'Cache_BackpackState'())
    bAlphaBoolEnabled:bool=True
    Translation="0,0,0"
    Rotation="0,-10,0"
    Scale="1,1,1"
   ) -> (Pose) {
    alphaInputType=Bool
    rotationMode=BMM_Additive
    rotationSpace=BCS_BoneSpace
    translationMode=BMM_Additive
    translationSpace=BCS_BoneSpace
    alphaBoolBlend {blendInTime=0.5, blendOption=Linear, blendOutTime=0.8}
    boneToModify {boneName=dyn_hair_back_02_m}
    rotation {yaw=-10}
   }
   n2 = "Transform (Modify) Bone - Bone: dyn_hair_back_03_m"!(ComponentPose=@n1, bAlphaBoolEnabled:bool=True, Translation="0,0,0", Rotation="0,-15,0", Scale="1,1,1") -> (Pose) {
    alphaInputType=Bool
    rotationMode=BMM_Additive
    rotationSpace=BCS_BoneSpace
    translationMode=BMM_Additive
    translationSpace=BCS_BoneSpace
    alphaBoolBlend {blendInTime=0.5, blendOption=Linear, blendOutTime=0.8}
    boneToModify {boneName=dyn_hair_back_03_m}
    rotation {yaw=-15}
   }
   n3 = "Transform (Modify) Bone - Bone: dyn_hair_back_04_m"!(ComponentPose=@n2, bAlphaBoolEnabled:bool=True, Translation="0,0,0", Rotation="0,-20,0", Scale="1,1,1") -> (Pose) {
    alphaInputType=Bool
    rotationMode=BMM_Additive
    rotationSpace=BCS_BoneSpace
    translationMode=BMM_Additive
    translationSpace=BCS_BoneSpace
    alphaBoolBlend {blendInTime=0.5, blendOption=Linear, blendOutTime=0.8}
    boneToModify {boneName=dyn_hair_back_04_m}
    rotation {yaw=-20}
   }
   n4 = "Transform (Modify) Bone - Bone: dyn_hair_back_05_m"!(ComponentPose=@n3, bAlphaBoolEnabled:bool=True, Translation="0,0,0", Rotation="0,-25,0", Scale="1,1,1") -> (Pose) {
    alphaInputType=Bool
    rotationMode=BMM_Additive
    rotationSpace=BCS_BoneSpace
    translationMode=BMM_Additive
    translationSpace=BCS_BoneSpace
    alphaBoolBlend {blendInTime=0.5, blendOption=Linear, blendOutTime=0.8}
    boneToModify {boneName=dyn_hair_back_05_m}
    rotation {yaw=-25}
   }
   n5 = "Transform (Modify) Bone - Bone: dyn_hair_back_02_l"!(ComponentPose=@n4, bAlphaBoolEnabled:bool=True, Translation="0,0,0", Rotation="0,-10,0", Scale="1,1,1") -> (Pose) {
    alphaInputType=Bool
    rotationMode=BMM_Additive
    rotationSpace=BCS_BoneSpace
    translationMode=BMM_Additive
    translationSpace=BCS_BoneSpace
    alphaBoolBlend {blendInTime=0.5, blendOption=Linear, blendOutTime=0.8}
    boneToModify {boneName=dyn_hair_back_02_l}
    rotation {yaw=-10}
   }
   n6 = "Transform (Modify) Bone - Bone: dyn_hair_back_03_l"!(ComponentPose=@n5, bAlphaBoolEnabled:bool=True, Translation="0,0,0", Rotation="0,-15,0", Scale="1,1,1") -> (Pose) {
    alphaInputType=Bool
    rotationMode=BMM_Additive
    rotationSpace=BCS_BoneSpace
    translationMode=BMM_Additive
    translationSpace=BCS_BoneSpace
    alphaBoolBlend {blendInTime=0.5, blendOption=Linear, blendOutTime=0.8}
    boneToModify {boneName=dyn_hair_back_03_l}
    rotation {yaw=-15}
   }
   n7 = "Transform (Modify) Bone - Bone: dyn_hair_back_04_l"!(ComponentPose=@n6, bAlphaBoolEnabled:bool=True, Translation="0,0,0", Rotation="0,-20,0", Scale="1,1,1") -> (Pose) {
    alphaInputType=Bool
    rotationMode=BMM_Additive
    rotationSpace=BCS_BoneSpace
    translationMode=BMM_Additive
    translationSpace=BCS_BoneSpace
    alphaBoolBlend {blendInTime=0.5, blendOption=Linear, blendOutTime=0.8}
    boneToModify {boneName=dyn_hair_back_04_l}
    rotation {yaw=-20}
   }
   n8 = "Transform (Modify) Bone - Bone: dyn_hair_back_05_l"!(ComponentPose=@n7, bAlphaBoolEnabled:bool=True, Translation="0,0,0", Rotation="0,-25,0", Scale="1,1,1") -> (Pose) {
    alphaInputType=Bool
    rotationMode=BMM_Additive
    rotationSpace=BCS_BoneSpace
    translationMode=BMM_Additive
    translationSpace=BCS_BoneSpace
    alphaBoolBlend {blendInTime=0.5, blendOption=Linear, blendOutTime=0.8}
    boneToModify {boneName=dyn_hair_back_05_l}
    rotation {yaw=-25}
   }
   n9 = "Transform (Modify) Bone - Bone: dyn_hair_back_02_r"!(ComponentPose=@n8, bAlphaBoolEnabled:bool=True, Translation="0,0,0", Rotation="0,-10,0", Scale="1,1,1") -> (Pose) {
    alphaInputType=Bool
    rotationMode=BMM_Additive
    rotationSpace=BCS_BoneSpace
    translationMode=BMM_Additive
    translationSpace=BCS_BoneSpace
    alphaBoolBlend {blendInTime=0.5, blendOption=Linear, blendOutTime=0.8}
    boneToModify {boneName=dyn_hair_back_02_r}
    rotation {yaw=-10}
   }
   n10 = "Transform (Modify) Bone - Bone: dyn_hair_back_03_r"!(ComponentPose=@n9, bAlphaBoolEnabled:bool=True, Translation="0,0,0", Rotation="0,-15,0", Scale="1,1,1") -> (Pose) {
    alphaInputType=Bool
    rotationMode=BMM_Additive
    rotationSpace=BCS_BoneSpace
    translationMode=BMM_Additive
    translationSpace=BCS_BoneSpace
    alphaBoolBlend {blendInTime=0.5, blendOption=Linear, blendOutTime=0.8}
    boneToModify {boneName=dyn_hair_back_03_r}
    rotation {yaw=-15}
   }
   n11 = "Transform (Modify) Bone - Bone: dyn_hair_back_04_r"!(ComponentPose=@n10, bAlphaBoolEnabled:bool=True, Translation="0,0,0", Rotation="0,-20,0", Scale="1,1,1") -> (Pose) {
    alphaInputType=Bool
    rotationMode=BMM_Additive
    rotationSpace=BCS_BoneSpace
    translationMode=BMM_Additive
    translationSpace=BCS_BoneSpace
    alphaBoolBlend {blendInTime=0.5, blendOption=Linear, blendOutTime=0.8}
    boneToModify {boneName=dyn_hair_back_04_r}
    rotation {yaw=-20}
   }
   n12 = "Transform (Modify) Bone - Bone: dyn_hair_back_05_r"!(ComponentPose=@n11, bAlphaBoolEnabled:bool=True, Translation="0,0,0", Rotation="0,-25,0", Scale="1,1,1") -> (Pose) {
    alphaInputType=Bool
    rotationMode=BMM_Additive
    rotationSpace=BCS_BoneSpace
    translationMode=BMM_Additive
    translationSpace=BCS_BoneSpace
    alphaBoolBlend {blendInTime=0.5, blendOption=Linear, blendOutTime=0.8}
    boneToModify {boneName=dyn_hair_back_05_r}
    rotation {yaw=-25}
   }
   n13 = "Transform (Modify) Bone - Bone: base_hair_back_m"(ComponentPose=@n12, Alpha=1, Translation="0,0,0", Rotation="-30,0,0", Scale="1,1,1") {
    rotationMode=BMM_Additive
    translationMode=BMM_Additive
    alphaBoolBlend {blendOption=Linear}
    alphaScaleBiasClamp {bInterpResult=true}
    boneToModify {boneName=base_hair_back_m}
    rotation {pitch=-30}
   }
   n14 = "Transform (Modify) Bone - Bone: base_hair_back_r"(ComponentPose=@n13, Alpha=1, Translation="0,0,0", Rotation="-48.977,-0.032,0.061", Scale="1,1,1") {
    rotationMode=BMM_Additive
    rotationSpace=BCS_BoneSpace
    translationMode=BMM_Additive
    translationSpace=BCS_BoneSpace
    alphaBoolBlend {blendOption=Linear}
    alphaScaleBiasClamp {bInterpResult=true}
    boneToModify {boneName=base_hair_back_r}
    rotation=<rot -48.977 -0.0319 0.061>
   }
   n15 = "Transform (Modify) Bone - Bone: base_hair_back_l"(ComponentPose=@n14, Alpha=1, Translation="0,0,0", Rotation="-20,0,0", Scale="1,1,1") {
    rotationMode=BMM_Additive
    rotationSpace=BCS_BoneSpace
    translationMode=BMM_Additive
    translationSpace=BCS_BoneSpace
    alphaBoolBlend {blendOption=Linear}
    alphaScaleBiasClamp {bInterpResult=true}
    boneToModify {boneName=base_hair_back_l}
    rotation {pitch=-20}
   }
   n16 = "Transform (Modify) Bone - Bone: dyn_hair_back_03_m"!(ComponentPose=@n15, Alpha:float=1, Translation="0,0,0", Rotation="-10,0,0", Scale="1,1,1") -> (Pose) {
    rotationMode=BMM_Additive
    translationMode=BMM_Additive
    alphaBoolBlend {blendOption=Linear}
    alphaScaleBiasClamp {bInterpResult=true}
    boneToModify {boneName=dyn_hair_back_03_m}
    rotation {pitch=-10}
   }
   n17 = "Transform (Modify) Bone - Bone: dyn_hair_back_03_r"!(ComponentPose=@n16, Alpha:float=1, Translation="0,0,0", Rotation="-10,0,0", Scale="1,1,1") -> (Pose) {
    rotationMode=BMM_Additive
    rotationSpace=BCS_BoneSpace
    translationMode=BMM_Additive
    translationSpace=BCS_BoneSpace
    alphaBoolBlend {blendOption=Linear}
    alphaScaleBiasClamp {bInterpResult=true}
    boneToModify {boneName=dyn_hair_back_03_r}
    rotation {pitch=-10}
   }
   n18 = "Transform (Modify) Bone - Bone: dyn_hair_back_03_l"!(ComponentPose=@n17, Alpha:float=1, Translation="0,0,0", Rotation="-10,0,0", Scale="1,1,1") -> (Pose) {
    rotationMode=BMM_Additive
    rotationSpace=BCS_BoneSpace
    translationMode=BMM_Additive
    translationSpace=BCS_BoneSpace
    alphaBoolBlend {blendOption=Linear}
    alphaScaleBiasClamp {bInterpResult=true}
    boneToModify {boneName=dyn_hair_back_03_l}
    rotation {pitch=-10}
   }
   n19 = "Transform (Modify) Bone - Bone: base_hair_back_m"(ComponentPose=@n18, Alpha=1, Translation="0,0,0", Rotation="-40,-10,1.783", Scale="1,1,1") {
    rotationMode=BMM_Additive
    rotationSpace=BCS_BoneSpace
    translationMode=BMM_Additive
    translationSpace=BCS_BoneSpace
    alphaBoolBlend {blendOption=Linear}
    alphaScaleBiasClamp {bInterpResult=true}
    boneToModify {boneName=base_hair_back_m}
    rotation=<rot -40 -10 1.783>
   }
   n20 = "Transform (Modify) Bone - Bone: base_hair_back_r"(ComponentPose=@n19, Alpha=1, Translation="0,0,0", Rotation="-40,-10,0.021", Scale="1,1,1") {
    rotationMode=BMM_Additive
    rotationSpace=BCS_BoneSpace
    translationMode=BMM_Additive
    translationSpace=BCS_BoneSpace
    alphaBoolBlend {blendOption=Linear}
    alphaScaleBiasClamp {bInterpResult=true}
    boneToModify {boneName=base_hair_back_r}
    rotation=<rot -40 -10 0.021>
   }
   n21 = "Transform (Modify) Bone - Bone: base_hair_back_l"(ComponentPose=@n20, Alpha=1, Translation="0,0,0", Rotation="-40,-10,2.373", Scale="1,1,1") {
    rotationMode=BMM_Additive
    rotationSpace=BCS_BoneSpace
    translationMode=BMM_Additive
    translationSpace=BCS_BoneSpace
    alphaBoolBlend {blendOption=Linear}
    alphaScaleBiasClamp {bInterpResult=true}
    boneToModify {boneName=base_hair_back_l}
    rotation=<rot -40 -10 2.373>
   }
   n22 = "Transform (Modify) Bone - Bone: dyn_hair_back_05_m"!(ComponentPose=@n21, Alpha:float=1, Translation="0,0,0", Rotation="-40,-10,1.783", Scale="1,1,1") -> (Pose) {
    rotationMode=BMM_Additive
    rotationSpace=BCS_BoneSpace
    translationMode=BMM_Additive
    translationSpace=BCS_BoneSpace
    alphaBoolBlend {blendOption=Linear}
    alphaScaleBiasClamp {bInterpResult=true}
    boneToModify {boneName=dyn_hair_back_05_m}
    rotation=<rot -40 -10 1.783>
   }
   n23 = "Transform (Modify) Bone - Bone: dyn_hair_back_05_r"!(ComponentPose=@n22, Alpha:float=1, Translation="0,0,0", Rotation="-40,-10,0.021", Scale="1,1,1") -> (Pose) {
    rotationMode=BMM_Additive
    rotationSpace=BCS_BoneSpace
    translationMode=BMM_Additive
    translationSpace=BCS_BoneSpace
    alphaBoolBlend {blendOption=Linear}
    alphaScaleBiasClamp {bInterpResult=true}
    boneToModify {boneName=dyn_hair_back_05_r}
    rotation=<rot -40 -10 0.021>
   }
   n24 = "Transform (Modify) Bone - Bone: dyn_hair_back_05_l"!(ComponentPose=@n23, Alpha:float=1, Translation="0,0,0", Rotation="-40,-10,2.373", Scale="1,1,1") -> (Pose) {
    rotationMode=BMM_Additive
    rotationSpace=BCS_BoneSpace
    translationMode=BMM_Additive
    translationSpace=BCS_BoneSpace
    alphaBoolBlend {blendOption=Linear}
    alphaScaleBiasClamp {bInterpResult=true}
    boneToModify {boneName=dyn_hair_back_05_l}
    rotation=<rot -40 -10 2.373>
   }
   n25 = "Constraint - dyn_hair_back_02_l (PP)"!(ComponentPose=@n24, Alpha=1, ConstraintWeights_0=0.3, ConstraintWeights_1:float=0.7) -> (Pose) {
    alphaBoolBlend {blendOption=Linear}
    boneToModify {boneName=dyn_hair_back_02_l}
    constraintSetup [{targetBone {boneName=head}}, {targetBone {boneName=neck_02}}]
    constraintWeights [1,0]
   }
   n26 = "Constraint - dyn_hair_back_03_l (PPP)"!(ComponentPose=@n25, Alpha=1, ConstraintWeights_0=0.825, ConstraintWeights_1:float=0.175, ConstraintWeights_2:float=0) -> (Pose) {
    alphaBoolBlend {blendOption=Linear}
    boneToModify {boneName=dyn_hair_back_03_l}
    constraintSetup [{targetBone {boneName=spine_05}}, {targetBone {boneName=clavicle_l}}, {targetBone {boneName=neck_02}}]
    constraintWeights [1,1,1]
   }
   n27 = "Constraint - dyn_hair_back_04_l (PPPP)"!(
    ComponentPose=@n26
    Alpha=1
    ConstraintWeights_0=0.6653
    ConstraintWeights_1:float=0.3309
    ConstraintWeights_2:float=0
    ConstraintWeights_3:float=0.0038
   ) -> (Pose) {
    alphaBoolBlend {blendOption=Linear}
    boneToModify {boneName=dyn_hair_back_04_l}
    constraintSetup [
     {targetBone {boneName=clavicle_l}}
     {targetBone {boneName=spine_05}}
     {targetBone {boneName=upperarm_twist_01_l}}
     {targetBone {boneName=clavicle_r}}
    ]
    constraintWeights [1,1,1,1]
   }
   n28 = "Constraint - dyn_hair_back_05_l (PPPP)"!(
    ComponentPose=@n27
    Alpha=1
    ConstraintWeights_0=0.759
    ConstraintWeights_1:float=0.1995
    ConstraintWeights_2:float=0
    ConstraintWeights_3:float=0.0415
   ) -> (Pose) {
    alphaBoolBlend {blendOption=Linear}
    boneToModify {boneName=dyn_hair_back_05_l}
    constraintSetup [
     {targetBone {boneName=clavicle_l}}
     {targetBone {boneName=spine_05}}
     {targetBone {boneName=upperarm_twist_01_l}}
     {targetBone {boneName=spine_04}}
    ]
    constraintWeights [1,1,1,1]
   }
   n29 = "Constraint - dyn_hair_back_06_l (PPPPPP)"!(
    ComponentPose=@n28
    Alpha=1
    ConstraintWeights_0=0.5094
    ConstraintWeights_1:float=0.2398
    ConstraintWeights_2:float=0.239
    ConstraintWeights_3:float=0
    ConstraintWeights_4:float=0.0105
    ConstraintWeights_5:float=0.0013
   ) -> (Pose) {
    alphaBoolBlend {blendOption=Linear}
    boneToModify {boneName=dyn_hair_back_06_l}
    constraintSetup [
     {targetBone {boneName=clavicle_l}}
     {targetBone {boneName=spine_04}}
     {targetBone {boneName=spine_05}}
     {targetBone {boneName=upperarm_twist_01_l}}
     {targetBone {boneName=spine_03}}
     {targetBone {boneName=spine_02}}
    ]
    constraintWeights [1,1,1,1,1,1]
   }
   n30 = "Constraint - dyn_hair_back_02_m (PP)"!(ComponentPose=@n29, Alpha=1, ConstraintWeights_0=0.3, ConstraintWeights_1:float=0.7) -> (Pose) {
    alphaBoolBlend {blendOption=Linear}
    boneToModify {boneName=dyn_hair_back_02_m}
    constraintSetup [{targetBone {boneName=head}}, {targetBone {boneName=neck_02}}]
    constraintWeights [1,0]
   }
   n31 = "Constraint - dyn_hair_back_03_m (PPPP)"!(
    ComponentPose=@n30
    Alpha=1
    ConstraintWeights_0=0.3
    ConstraintWeights_1:float=0.3
    ConstraintWeights_2:float=0
    ConstraintWeights_3:float=0.4
   ) -> (Pose) {
    alphaBoolBlend {blendOption=Linear}
    boneToModify {boneName=dyn_hair_back_03_m}
    constraintSetup [{targetBone {boneName=neck_02}}, {targetBone {boneName=neck_01}}, {targetBone {boneName=head}}, {targetBone {boneName=spine_05}}]
    constraintWeights [1,1,1,1]
   }
   n32 = "Constraint - dyn_hair_back_04_m (PPP)"!(ComponentPose=@n31, Alpha=1, ConstraintWeights_0=0.9548, ConstraintWeights_1:float=0.0226, ConstraintWeights_2:float=0.0226) -> (Pose) {
    alphaBoolBlend {blendOption=Linear}
    boneToModify {boneName=dyn_hair_back_04_m}
    constraintSetup [{targetBone {boneName=spine_05}}, {targetBone {boneName=clavicle_r}}, {targetBone {boneName=clavicle_l}}]
    constraintWeights [1,1,1]
   }
   n33 = "Constraint - dyn_hair_back_05_m (PPPP)"!(
    ComponentPose=@n32
    Alpha=1
    ConstraintWeights_0=0.6851
    ConstraintWeights_1:float=0.1717
    ConstraintWeights_2:float=0.1328
    ConstraintWeights_3:float=0.0104
   ) -> (Pose) {
    alphaBoolBlend {blendOption=Linear}
    boneToModify {boneName=dyn_hair_back_05_m}
    constraintSetup [{targetBone {boneName=spine_05}}, {targetBone {boneName=clavicle_l}}, {targetBone {boneName=clavicle_r}}, {targetBone {boneName=spine_04}}]
    constraintWeights [1,1,1,1]
   }
   n34 = "Constraint - dyn_hair_back_06_m (PPPPP)"!(
    ComponentPose=@n33
    Alpha=1
    ConstraintWeights_0=0.4118
    ConstraintWeights_1:float=0.3788
    ConstraintWeights_2:float=0.1139
    ConstraintWeights_3:float=0.0799
    ConstraintWeights_4:float=0.0156
   ) -> (Pose) {
    alphaBoolBlend {blendOption=Linear}
    boneToModify {boneName=dyn_hair_back_06_m}
    constraintSetup [
     {targetBone {boneName=spine_04}}
     {targetBone {boneName=spine_05}}
     {targetBone {boneName=clavicle_l}}
     {targetBone {boneName=clavicle_r}}
     {targetBone {boneName=spine_03}}
    ]
    constraintWeights [1,1,1,1,1]
   }
   n35 = "Constraint - dyn_hair_back_02_r (PP)"!(ComponentPose=@n34, Alpha=1, ConstraintWeights_0=0.3, ConstraintWeights_1:float=0.7) -> (Pose) {
    alphaBoolBlend {blendOption=Linear}
    boneToModify {boneName=dyn_hair_back_02_r}
    constraintSetup [{targetBone {boneName=head}}, {targetBone {boneName=neck_02}}]
    constraintWeights [1,0]
   }
   n36 = "Constraint - dyn_hair_back_03_r (PPP)"!(ComponentPose=@n35, Alpha=1, ConstraintWeights_0=0.4, ConstraintWeights_1:float=0.3, ConstraintWeights_2:float=0.3) -> (Pose) {
    alphaBoolBlend {blendOption=Linear}
    boneToModify {boneName=dyn_hair_back_03_r}
    constraintSetup [{targetBone {boneName=spine_05}}, {targetBone {boneName=neck_02}}, {targetBone {boneName=neck_01}}]
    constraintWeights [1,1,1]
   }
   n37 = "Constraint - dyn_hair_back_04_r (PPPP)"!(
    ComponentPose=@n36
    Alpha=1
    ConstraintWeights_0=0.6653
    ConstraintWeights_1:float=0.3309
    ConstraintWeights_2:float=0
    ConstraintWeights_3:float=0.0038
   ) -> (Pose) {
    alphaBoolBlend {blendOption=Linear}
    boneToModify {boneName=dyn_hair_back_04_r}
    constraintSetup [
     {targetBone {boneName=clavicle_r}}
     {targetBone {boneName=spine_05}}
     {targetBone {boneName=upperarm_twist_01_r}}
     {targetBone {boneName=clavicle_l}}
    ]
    constraintWeights [1,1,1,1]
   }
   n38 = "Constraint - dyn_hair_back_05_r (PPPPP)"!(
    ComponentPose=@n37
    Alpha=1
    ConstraintWeights_0=0.6906
    ConstraintWeights_1:float=0.2575
    ConstraintWeights_2:float=0.0507
    ConstraintWeights_3:float=0
    ConstraintWeights_4:float=0.0012
   ) -> (Pose) {
    alphaBoolBlend {blendOption=Linear}
    boneToModify {boneName=dyn_hair_back_05_r}
    constraintSetup [
     {targetBone {boneName=clavicle_r}}
     {targetBone {boneName=spine_05}}
     {targetBone {boneName=spine_04}}
     {targetBone {boneName=upperarm_twist_01_r}}
     {targetBone {boneName=clavicle_l}}
    ]
    constraintWeights [1,1,1,1,1]
   }
   n39 = "Constraint - dyn_hair_back_06_r (PPPPP)"!(
    ComponentPose=@n38
    Alpha=1
    ConstraintWeights_0=0.5099
    ConstraintWeights_1:float=0.2403
    ConstraintWeights_2:float=0.2394
    ConstraintWeights_3:float=0
    ConstraintWeights_4:float=0.0105
   ) -> (Pose) {
    alphaBoolBlend {blendOption=Linear}
    boneToModify {boneName=dyn_hair_back_06_r}
    constraintSetup [
     {targetBone {boneName=clavicle_r}}
     {targetBone {boneName=spine_04}}
     {targetBone {boneName=spine_05}}
     {targetBone {boneName=upperarm_twist_01_r}}
     {targetBone {boneName=spine_03}}
    ]
    constraintWeights [1,1,1,1,1]
   }
   n40 = RigidBody!(
    ComponentPose=Local To Component(LocalPose=Use cached pose 'Cache_StartPose'())
    bAlphaBoolEnabled:bool=NOT Boolean(A=Get bIsInFrontEnd())
    OverrideWorldGravity:Vector=Get GravityOverride()
    ExternalForce
    ComponentLinearAccScale:Vector="0,0,0"
    ComponentLinearVelScale:Vector="0,0,0"
   ) -> (Pose) {
    alphaInputType=Bool
    bClampLinearTranslationLimitToRefPose=true
    bOverrideWorldGravity=true
    overridePhysicsAsset=&/F_MED_MutedMaroon_FaceAcc_Physics_alt
    simulationSpace=BaseBoneSpace
    alphaBoolBlend {blendOption=Linear}
    baseBoneRef {boneName=neck_01}
    componentAppliedLinearAccClamp=<0 0 0>
    simSpaceSettings {maxAngularAcceleration=40, maxAngularVelocity=8, maxLinearAcceleration=3500, maxLinearVelocity=700, velocityScaleZ=0.2, worldAlpha=0.5}
   }
   n41 = RigidBody(
    ComponentPose=@n40
    bAlphaBoolEnabled=Get bIsInFrontEnd()
    OverrideWorldGravity=Get GravityOverride()
    ComponentLinearAccScale="0,0,0"
    ComponentLinearVelScale="0,0,0"
   ) {
    alphaInputType=Bool
    bClampLinearTranslationLimitToRefPose=true
    bOverrideWorldGravity=true
    overridePhysicsAsset=&/F_MED_MutedMaroon_FaceAcc_FrontEnd_Physics
    simulationSpace=BaseBoneSpace
    alphaBoolBlend {blendOption=Linear}
    baseBoneRef {boneName=neck_01}
    simSpaceSettings {worldAlpha=0.05}
   }
   n42 = "Transform (Modify) Bone - Bone: base_hair_back_m"(
    ComponentPose=Local To Component(LocalPose=Use cached pose 'Cache_BackpackState'())
    Alpha=1
    Translation="0,0,0"
    Rotation="-10,-5.288,1.783"
    Scale="1,1,1"
   ) {
    rotationMode=BMM_Additive
    rotationSpace=BCS_BoneSpace
    translationMode=BMM_Additive
    translationSpace=BCS_BoneSpace
    alphaBoolBlend {blendOption=Linear}
    alphaScaleBiasClamp {bInterpResult=true}
    boneToModify {boneName=base_hair_back_m}
    rotation=<rot -10 -5.288 1.783>
   }
   n43 = "Transform (Modify) Bone - Bone: base_hair_back_r"(ComponentPose=@n42, Alpha=1, Translation="0,0,0", Rotation="-10,-5,0.021", Scale="1,1,1") {
    rotationMode=BMM_Additive
    rotationSpace=BCS_BoneSpace
    translationMode=BMM_Additive
    translationSpace=BCS_BoneSpace
    alphaBoolBlend {blendOption=Linear}
    alphaScaleBiasClamp {bInterpResult=true}
    boneToModify {boneName=base_hair_back_r}
    rotation=<rot -10 -5 0.021>
   }
   n44 = "Transform (Modify) Bone - Bone: base_hair_back_l"(ComponentPose=@n43, Alpha=1, Translation="0,0,0", Rotation="-10,-5.532,2.373", Scale="1,1,1") {
    rotationMode=BMM_Additive
    rotationSpace=BCS_BoneSpace
    translationMode=BMM_Additive
    translationSpace=BCS_BoneSpace
    alphaBoolBlend {blendOption=Linear}
    alphaScaleBiasClamp {bInterpResult=true}
    boneToModify {boneName=base_hair_back_l}
    rotation=<rot -10 -5.532 2.373>
   }
   n45 = "Transform (Modify) Bone - Bone: base_hair_back_l"(ComponentPose=@n44, Alpha=1, Translation="0,0,0", Rotation="-5.599,0.000,-0.000", Scale="1,1,1") {
    rotationMode=BMM_Additive
    rotationSpace=BCS_BoneSpace
    translationMode=BMM_Additive
    translationSpace=BCS_BoneSpace
    alphaBoolBlend {blendOption=Linear}
    alphaScaleBiasClamp {bInterpResult=true}
    boneToModify {boneName=base_hair_back_l}
    rotation=<rot -5.599 0 0>
   }
   n46 = "Transform (Modify) Bone - Bone: base_hair_back_m"(ComponentPose=@n45, Alpha=1, Translation="0,0,0", Rotation="-10,0,0", Scale="1,1,1") {
    rotationMode=BMM_Additive
    rotationSpace=BCS_BoneSpace
    translationMode=BMM_Additive
    translationSpace=BCS_BoneSpace
    alphaBoolBlend {blendOption=Linear}
    alphaScaleBiasClamp {bInterpResult=true}
    boneToModify {boneName=base_hair_back_m}
    rotation {pitch=-10}
   }
   n47 = "Transform (Modify) Bone - Bone: base_hair_back_m"(ComponentPose=@n46, Alpha=1, Translation="0,0,0", Rotation="30,0,0", Scale="1,1,1") {
    rotationMode=BMM_Additive
    rotationSpace=BCS_BoneSpace
    translationMode=BMM_Additive
    translationSpace=BCS_BoneSpace
    alphaBoolBlend {blendOption=Linear}
    alphaScaleBiasClamp {bInterpResult=true, interpSpeedDecreasing=5, interpSpeedIncreasing=20}
    boneToModify {boneName=base_hair_back_m}
    rotation {pitch=30}
   }
   n48 = "Transform (Modify) Bone - Bone: base_hair_back_r"(ComponentPose=@n47, Alpha=1, Translation="0,0,0", Rotation="45,0,0", Scale="1,1,1") {
    rotationMode=BMM_Additive
    rotationSpace=BCS_BoneSpace
    translationMode=BMM_Additive
    translationSpace=BCS_BoneSpace
    alphaBoolBlend {blendOption=Linear}
    alphaScaleBiasClamp {bInterpResult=true, interpSpeedDecreasing=5, interpSpeedIncreasing=20}
    boneToModify {boneName=base_hair_back_r}
    rotation {pitch=45}
   }
   n49 = "Transform (Modify) Bone - Bone: base_hair_back_l"(ComponentPose=@n48, Alpha=1, Translation="0,0,0", Rotation="30,0,0", Scale="1,1,1") {
    rotationMode=BMM_Additive
    rotationSpace=BCS_BoneSpace
    translationMode=BMM_Additive
    translationSpace=BCS_BoneSpace
    alphaBoolBlend {blendOption=Linear}
    alphaScaleBiasClamp {bInterpResult=true, interpSpeedDecreasing=5, interpSpeedIncreasing=20}
    boneToModify {boneName=base_hair_back_l}
    rotation {pitch=30}
   }
   n50 = "Transform (Modify) Bone - Bone: dyn_hair_back_01_l"(ComponentPose=@n49, Alpha=1, Translation="0,0,0", Rotation="0.000,5.441,-0.000", Scale="1,1,1") {
    rotationMode=BMM_Additive
    rotationSpace=BCS_BoneSpace
    translationMode=BMM_Additive
    translationSpace=BCS_BoneSpace
    alphaBoolBlend {blendOption=Linear}
    alphaScaleBiasClamp {bInterpResult=true, interpSpeedDecreasing=1, interpSpeedIncreasing=1}
    boneToModify {boneName=dyn_hair_back_01_l}
    rotation {yaw=5.441}
   }
   n51 = "Transform (Modify) Bone - Bone: dyn_hair_back_01_m"(ComponentPose=@n50, Alpha=1, Translation="0,0,0", Rotation="0.323,-20.046,0.062", Scale="1,1,1") {
    rotationMode=BMM_Additive
    rotationSpace=BCS_BoneSpace
    translationMode=BMM_Additive
    translationSpace=BCS_BoneSpace
    alphaBoolBlend {blendOption=Linear}
    alphaScaleBiasClamp {bInterpResult=true, interpSpeedDecreasing=1, interpSpeedIncreasing=1}
    boneToModify {boneName=dyn_hair_back_01_m}
    rotation=<rot 0.323 -20.046 0.062>
   }
   n52 = "Transform (Modify) Bone - Bone: dyn_hair_back_01_r"(ComponentPose=@n51, Alpha=1, Translation="0,0,0", Rotation="-0.001,-1.922,-0.004", Scale="1,1,1") {
    rotationMode=BMM_Additive
    rotationSpace=BCS_BoneSpace
    translationMode=BMM_Additive
    translationSpace=BCS_BoneSpace
    alphaBoolBlend {blendOption=Linear}
    alphaScaleBiasClamp {bInterpResult=true, interpSpeedDecreasing=1, interpSpeedIncreasing=1}
    boneToModify {boneName=dyn_hair_back_01_r}
    rotation=<rot -0.001 -1.922 -0.004>
   }
   n53 = "Transform (Modify) Bone - Bone: dyn_hair_back_03_l"(ComponentPose=@n52, Alpha=1, Translation="0,0,0", Rotation="0.000,5.441,-0.000", Scale="1,1,1") {
    rotationMode=BMM_Additive
    rotationSpace=BCS_BoneSpace
    translationMode=BMM_Additive
    translationSpace=BCS_BoneSpace
    alphaBoolBlend {blendOption=Linear}
    alphaScaleBiasClamp {bInterpResult=true, interpSpeedDecreasing=1, interpSpeedIncreasing=1}
    boneToModify {boneName=dyn_hair_back_03_l}
    rotation {yaw=5.441}
   }
   n54 = "Transform (Modify) Bone - Bone: dyn_hair_back_03_m"(ComponentPose=@n53, Alpha=1, Translation="0,0,0", Rotation="0.006,46.892,0.102", Scale="1,1,1") {
    rotationMode=BMM_Additive
    rotationSpace=BCS_BoneSpace
    translationMode=BMM_Additive
    translationSpace=BCS_BoneSpace
    alphaBoolBlend {blendOption=Linear}
    alphaScaleBiasClamp {bInterpResult=true, interpSpeedDecreasing=1, interpSpeedIncreasing=1}
    boneToModify {boneName=dyn_hair_back_03_m}
    rotation=<rot 0.006 46.892 0.102>
   }
   n55 = "Transform (Modify) Bone - Bone: dyn_hair_back_02_r"!(ComponentPose=@n54, Alpha:float=1, Translation="0,0,0", Rotation="0.000,5.439,-0.001", Scale="1,1,1") -> (Pose) {
    rotationMode=BMM_Additive
    rotationSpace=BCS_BoneSpace
    translationMode=BMM_Additive
    translationSpace=BCS_BoneSpace
    alphaBoolBlend {blendOption=Linear}
    alphaScaleBiasClamp {bInterpResult=true, interpSpeedDecreasing=1, interpSpeedIncreasing=1}
    boneToModify {boneName=dyn_hair_back_02_r}
    rotation=<rot 0 5.439 -0.001>
   }
   n56 = "Transform (Modify) Bone - Bone: dyn_hair_back_03_r"(ComponentPose=@n55, Alpha=1, Translation="0,0,0", Rotation="0.000,7.059,0.000", Scale="1,1,1") {
    rotationMode=BMM_Additive
    rotationSpace=BCS_BoneSpace
    translationMode=BMM_Additive
    translationSpace=BCS_BoneSpace
    alphaBoolBlend {blendOption=Linear}
    alphaScaleBiasClamp {bInterpResult=true, interpSpeedDecreasing=1, interpSpeedIncreasing=1}
    boneToModify {boneName=dyn_hair_back_03_r}
    rotation=<rot 0 7.059 0>
   }
   n57 = "Transform (Modify) Bone - Bone: dyn_hair_back_01_l"(ComponentPose=@n56, Alpha=1, Translation="0,0,0", Rotation="-0.024,-23.722,0.042", Scale="1,1,1") {
    rotationMode=BMM_Additive
    rotationSpace=BCS_BoneSpace
    translationMode=BMM_Additive
    translationSpace=BCS_BoneSpace
    alphaBoolBlend {blendOption=Linear}
    alphaScaleBiasClamp {bInterpResult=true, interpSpeedDecreasing=1, interpSpeedIncreasing=1}
    boneToModify {boneName=dyn_hair_back_01_l}
    rotation=<rot -0.024 -23.722 0.042>
   }
   n58 = "Transform (Modify) Bone - Bone: dyn_hair_back_02_l"!(ComponentPose=@n57, Alpha:float=1, Translation="0,0,0", Rotation="-0.103,21.788,0.045", Scale="1,1,1") -> (Pose) {
    rotationMode=BMM_Additive
    rotationSpace=BCS_BoneSpace
    translationMode=BMM_Additive
    translationSpace=BCS_BoneSpace
    alphaBoolBlend {blendOption=Linear}
    alphaScaleBiasClamp {bInterpResult=true, interpSpeedDecreasing=1, interpSpeedIncreasing=1}
    boneToModify {boneName=dyn_hair_back_02_l}
    rotation=<rot -0.103 21.788 0.045>
   }
   n59 = "Transform (Modify) Bone - Bone: dyn_hair_back_01_m"(ComponentPose=@n58, Alpha=1, Translation="0,0,0", Rotation="2.870,-19.005,-2.445", Scale="1,1,1") {
    rotationMode=BMM_Additive
    rotationSpace=BCS_BoneSpace
    translationMode=BMM_Additive
    translationSpace=BCS_BoneSpace
    alphaBoolBlend {blendOption=Linear}
    alphaScaleBiasClamp {bInterpResult=true, interpSpeedDecreasing=1, interpSpeedIncreasing=1}
    boneToModify {boneName=dyn_hair_back_01_m}
    rotation=<rot 2.87 -19.005 -2.445>
   }
   n60 = "Transform (Modify) Bone - Bone: dyn_hair_back_02_m"!(ComponentPose=@n59, Alpha:float=1, Translation="0,0,0", Rotation="-0.106,37.919,0.142", Scale="1,1,1") -> (Pose) {
    rotationMode=BMM_Additive
    rotationSpace=BCS_BoneSpace
    translationMode=BMM_Additive
    translationSpace=BCS_BoneSpace
    alphaBoolBlend {blendOption=Linear}
    alphaScaleBiasClamp {bInterpResult=true, interpSpeedDecreasing=1, interpSpeedIncreasing=1}
    boneToModify {boneName=dyn_hair_back_02_m}
    rotation=<rot -0.106 37.919 0.142>
   }
   n61 = "Transform (Modify) Bone - Bone: dyn_hair_back_01_r"(ComponentPose=@n60, Alpha=1, Translation="0,0,0", Rotation="-0.001,-27.607,-0.004", Scale="1,1,1") {
    rotationMode=BMM_Additive
    rotationSpace=BCS_BoneSpace
    translationMode=BMM_Additive
    translationSpace=BCS_BoneSpace
    alphaBoolBlend {blendOption=Linear}
    alphaScaleBiasClamp {bInterpResult=true, interpSpeedDecreasing=1, interpSpeedIncreasing=1}
    boneToModify {boneName=dyn_hair_back_01_r}
    rotation=<rot -0.001 -27.607 -0.004>
   }
   n62 = "Transform (Modify) Bone - Bone: dyn_hair_back_03_l"(ComponentPose=@n61, Alpha=1, Translation="0,0,0", Rotation="0.002,35.230,-0.004", Scale="1,1,1") {
    rotationMode=BMM_Additive
    rotationSpace=BCS_BoneSpace
    translationMode=BMM_Additive
    translationSpace=BCS_BoneSpace
    alphaBoolBlend {blendOption=Linear}
    alphaScaleBiasClamp {bInterpResult=true, interpSpeedDecreasing=1, interpSpeedIncreasing=1}
    boneToModify {boneName=dyn_hair_back_03_l}
    rotation=<rot 0.002 35.23 -0.004>
   }
   n63 = "Transform (Modify) Bone - Bone: dyn_hair_back_03_m"(ComponentPose=@n62, Alpha=1, Translation="0,0,0", Rotation="5.525,15.009,0.084", Scale="1,1,1") {
    rotationMode=BMM_Additive
    rotationSpace=BCS_BoneSpace
    translationMode=BMM_Additive
    translationSpace=BCS_BoneSpace
    alphaBoolBlend {blendOption=Linear}
    alphaScaleBiasClamp {bInterpResult=true, interpSpeedDecreasing=1, interpSpeedIncreasing=1}
    boneToModify {boneName=dyn_hair_back_03_m}
    rotation=<rot 5.525 15.009 0.084>
   }
   n64 = "Transform (Modify) Bone - Bone: dyn_hair_back_04_m"!(ComponentPose=@n63, Alpha:float=1, Translation="0,0,0", Rotation="9.107,-25.617,0.051", Scale="1,1,1") -> (Pose) {
    rotationMode=BMM_Additive
    rotationSpace=BCS_BoneSpace
    translationMode=BMM_Additive
    translationSpace=BCS_BoneSpace
    alphaBoolBlend {blendOption=Linear}
    alphaScaleBiasClamp {bInterpResult=true, interpSpeedDecreasing=1, interpSpeedIncreasing=1}
    boneToModify {boneName=dyn_hair_back_04_m}
    rotation=<rot 9.107 -25.617 0.051>
   }
   n65 = "Transform (Modify) Bone - Bone: dyn_hair_back_02_r"(ComponentPose=@n64, Alpha=1, Translation="0,0,0", Rotation="0.000,34.758,-0.001", Scale="1,1,1") {
    rotationMode=BMM_Additive
    rotationSpace=BCS_BoneSpace
    translationMode=BMM_Additive
    translationSpace=BCS_BoneSpace
    alphaBoolBlend {blendOption=Linear}
    alphaScaleBiasClamp {bInterpResult=true, interpSpeedDecreasing=1, interpSpeedIncreasing=1}
    boneToModify {boneName=dyn_hair_back_02_r}
    rotation=<rot 0 34.758 -0.001>
   }
   n66 = "Transform (Modify) Bone - Bone: dyn_hair_back_03_r"(ComponentPose=@n65, Alpha=1, Translation="0,0,0", Rotation="0.000,33.724,0.000", Scale="1,1,1") {
    rotationMode=BMM_Additive
    rotationSpace=BCS_BoneSpace
    translationMode=BMM_Additive
    translationSpace=BCS_BoneSpace
    alphaBoolBlend {blendOption=Linear}
    alphaScaleBiasClamp {bInterpResult=true, interpSpeedDecreasing=1, interpSpeedIncreasing=1}
    boneToModify {boneName=dyn_hair_back_03_r}
    rotation {yaw=33.724}
   }
   n67 = "Transform (Modify) Bone - Bone: dyn_hair_back_04_r"!(ComponentPose=@n66, Alpha:float=1, Translation="0,0,0", Rotation="0.191,-19.237,0.041", Scale="1,1,1") -> (Pose) {
    rotationMode=BMM_Additive
    rotationSpace=BCS_BoneSpace
    translationMode=BMM_Additive
    translationSpace=BCS_BoneSpace
    alphaBoolBlend {blendOption=Linear}
    alphaScaleBiasClamp {bInterpResult=true, interpSpeedDecreasing=1, interpSpeedIncreasing=1}
    boneToModify {boneName=dyn_hair_back_04_r}
    rotation=<rot 0.191 -19.237 0.041>
   }
   n68 = Blend Poses by bool(
    BlendPose_0=Component To Local(ComponentPose=@n67)
    BlendPose_1=Blend Poses by bool(
     BlendPose_0=Component To Local(ComponentPose=@n39)
     BlendPose_1=Component To Local(ComponentPose=@n41)
     bActiveValue=Get isWearingBackbling()
    ) {blendType=Linear}
    bActiveValue=Get Is Wearing Cape() -> (Is Wearing Cape:bool)
   ) {blendType=Linear}
   Control Rig(Source=Component To Local(ComponentPose=Local To Component(LocalPose=@n68))) {
    alphaBoolBlend {blendOption=Linear}
    controlRigAssetReference {blueprintRigClass=&/F_MED_MutedMaroon_FaceAcc_Post_CtrlRig.F_MED_MutedMaroon_FaceAcc_Post_CtrlRig_C}
   }
   Local To Component()
   n69 = "Transform (Modify) Bone - Bone: dyn_hair_back_02_m"!(
    ComponentPose=Local To Component(LocalPose=Use cached pose 'Cache_BackpackState'())
    bAlphaBoolEnabled:bool=True
    Translation="0,0,0"
    Rotation="0,-10,0"
    Scale="1,1,1"
   ) -> (Pose) {
    alphaInputType=Bool
    rotationMode=BMM_Additive
    rotationSpace=BCS_BoneSpace
    translationMode=BMM_Additive
    translationSpace=BCS_BoneSpace
    alphaBoolBlend {blendInTime=0.5, blendOption=Linear, blendOutTime=0.8}
    boneToModify {boneName=dyn_hair_back_02_m}
    rotation {yaw=-10}
   }
   n70 = "Transform (Modify) Bone - Bone: dyn_hair_back_03_m"!(ComponentPose=@n69, bAlphaBoolEnabled:bool=True, Translation="0,0,0", Rotation="0,-15,0", Scale="1,1,1") -> (Pose) {
    alphaInputType=Bool
    rotationMode=BMM_Additive
    rotationSpace=BCS_BoneSpace
    translationMode=BMM_Additive
    translationSpace=BCS_BoneSpace
    alphaBoolBlend {blendInTime=0.5, blendOption=Linear, blendOutTime=0.8}
    boneToModify {boneName=dyn_hair_back_03_m}
    rotation {yaw=-15}
   }
   n71 = "Transform (Modify) Bone - Bone: dyn_hair_back_04_m"!(ComponentPose=@n70, bAlphaBoolEnabled:bool=True, Translation="0,0,0", Rotation="0,-20,0", Scale="1,1,1") -> (Pose) {
    alphaInputType=Bool
    rotationMode=BMM_Additive
    rotationSpace=BCS_BoneSpace
    translationMode=BMM_Additive
    translationSpace=BCS_BoneSpace
    alphaBoolBlend {blendInTime=0.5, blendOption=Linear, blendOutTime=0.8}
    boneToModify {boneName=dyn_hair_back_04_m}
    rotation {yaw=-20}
   }
   n72 = "Transform (Modify) Bone - Bone: dyn_hair_back_05_m"!(ComponentPose=@n71, bAlphaBoolEnabled:bool=True, Translation="0,0,0", Rotation="0,-25,0", Scale="1,1,1") -> (Pose) {
    alphaInputType=Bool
    rotationMode=BMM_Additive
    rotationSpace=BCS_BoneSpace
    translationMode=BMM_Additive
    translationSpace=BCS_BoneSpace
    alphaBoolBlend {blendInTime=0.5, blendOption=Linear, blendOutTime=0.8}
    boneToModify {boneName=dyn_hair_back_05_m}
    rotation {yaw=-25}
   }
   n73 = "Transform (Modify) Bone - Bone: dyn_hair_back_02_l"!(ComponentPose=@n72, bAlphaBoolEnabled:bool=True, Translation="0,0,0", Rotation="0,-10,0", Scale="1,1,1") -> (Pose) {
    alphaInputType=Bool
    rotationMode=BMM_Additive
    rotationSpace=BCS_BoneSpace
    translationMode=BMM_Additive
    translationSpace=BCS_BoneSpace
    alphaBoolBlend {blendInTime=0.5, blendOption=Linear, blendOutTime=0.8}
    boneToModify {boneName=dyn_hair_back_02_l}
    rotation {yaw=-10}
   }
   n74 = "Transform (Modify) Bone - Bone: dyn_hair_back_03_l"!(ComponentPose=@n73, bAlphaBoolEnabled:bool=True, Translation="0,0,0", Rotation="0,-15,0", Scale="1,1,1") -> (Pose) {
    alphaInputType=Bool
    rotationMode=BMM_Additive
    rotationSpace=BCS_BoneSpace
    translationMode=BMM_Additive
    translationSpace=BCS_BoneSpace
    alphaBoolBlend {blendInTime=0.5, blendOption=Linear, blendOutTime=0.8}
    boneToModify {boneName=dyn_hair_back_03_l}
    rotation {yaw=-15}
   }
   n75 = "Transform (Modify) Bone - Bone: dyn_hair_back_04_l"(ComponentPose=@n74, bAlphaBoolEnabled=True, Translation="0,0,0", Rotation="0,-20,0", Scale="1,1,1") {
    alphaInputType=Bool
    rotationMode=BMM_Additive
    rotationSpace=BCS_BoneSpace
    translationMode=BMM_Additive
    translationSpace=BCS_BoneSpace
    alphaBoolBlend {blendInTime=0.5, blendOption=Linear, blendOutTime=0.8}
    boneToModify {boneName=dyn_hair_back_04_l}
    rotation {yaw=-20}
   }
   n76 = "Transform (Modify) Bone - Bone: dyn_hair_back_05_l"!(ComponentPose=@n75, bAlphaBoolEnabled:bool=True, Translation="0,0,0", Rotation="0,-25,0", Scale="1,1,1") -> (Pose) {
    alphaInputType=Bool
    rotationMode=BMM_Additive
    rotationSpace=BCS_BoneSpace
    translationMode=BMM_Additive
    translationSpace=BCS_BoneSpace
    alphaBoolBlend {blendInTime=0.5, blendOption=Linear, blendOutTime=0.8}
    boneToModify {boneName=dyn_hair_back_05_l}
    rotation {yaw=-25}
   }
   n77 = "Transform (Modify) Bone - Bone: dyn_hair_back_02_r"!(ComponentPose=@n76, bAlphaBoolEnabled:bool=True, Translation="0,0,0", Rotation="0,-10,0", Scale="1,1,1") -> (Pose) {
    alphaInputType=Bool
    rotationMode=BMM_Additive
    rotationSpace=BCS_BoneSpace
    translationMode=BMM_Additive
    translationSpace=BCS_BoneSpace
    alphaBoolBlend {blendInTime=0.5, blendOption=Linear, blendOutTime=0.8}
    boneToModify {boneName=dyn_hair_back_02_r}
    rotation {yaw=-10}
   }
   n78 = "Transform (Modify) Bone - Bone: dyn_hair_back_03_r"!(ComponentPose=@n77, bAlphaBoolEnabled:bool=True, Translation="0,0,0", Rotation="0,-15,0", Scale="1,1,1") -> (Pose) {
    alphaInputType=Bool
    rotationMode=BMM_Additive
    rotationSpace=BCS_BoneSpace
    translationMode=BMM_Additive
    translationSpace=BCS_BoneSpace
    alphaBoolBlend {blendInTime=0.5, blendOption=Linear, blendOutTime=0.8}
    boneToModify {boneName=dyn_hair_back_03_r}
    rotation {yaw=-15}
   }
   n79 = "Transform (Modify) Bone - Bone: dyn_hair_back_04_r"!(ComponentPose=@n78, bAlphaBoolEnabled:bool=True, Translation="0,0,0", Rotation="0,-20,0", Scale="1,1,1") -> (Pose) {
    alphaInputType=Bool
    rotationMode=BMM_Additive
    rotationSpace=BCS_BoneSpace
    translationMode=BMM_Additive
    translationSpace=BCS_BoneSpace
    alphaBoolBlend {blendInTime=0.5, blendOption=Linear, blendOutTime=0.8}
    boneToModify {boneName=dyn_hair_back_04_r}
    rotation {yaw=-20}
   }
   n80 = "Transform (Modify) Bone - Bone: dyn_hair_back_05_r"!(ComponentPose=@n79, bAlphaBoolEnabled:bool=True, Translation="0,0,0", Rotation="0,-25,0", Scale="1,1,1") -> (Pose) {
    alphaInputType=Bool
    rotationMode=BMM_Additive
    rotationSpace=BCS_BoneSpace
    translationMode=BMM_Additive
    translationSpace=BCS_BoneSpace
    alphaBoolBlend {blendInTime=0.5, blendOption=Linear, blendOutTime=0.8}
    boneToModify {boneName=dyn_hair_back_05_r}
    rotation {yaw=-25}
   }
   n81 = "Constraint - dyn_hair_back_01_l (PPP)"!(ComponentPose=@n80, Alpha=1, ConstraintWeights_0=0.5, ConstraintWeights_1:float=0.4, ConstraintWeights_2:float=0.1) -> (Pose) {
    alphaBoolBlend {blendOption=Linear}
    boneToModify {boneName=dyn_hair_back_01_l}
    constraintSetup [{targetBone {boneName=head}}, {targetBone {boneName=neck_02}}, {targetBone {boneName=clavicle_l}}]
    constraintWeights [1,0,1]
   }
   n82 = "Constraint - dyn_hair_back_02_l (PP)"(ComponentPose=@n81, Alpha=1, ConstraintWeights_0=0.3, ConstraintWeights_1=0.7) {
    alphaBoolBlend {blendOption=Linear}
    boneToModify {boneName=dyn_hair_back_02_l}
    constraintSetup [{targetBone {boneName=head}}, {targetBone {boneName=neck_02}}]
    constraintWeights [1,0]
   }
   n83 = "Constraint - dyn_hair_back_03_l (PPP)"(ComponentPose=@n82, Alpha=1, ConstraintWeights_0=0.825, ConstraintWeights_1=0.175, ConstraintWeights_2=0) {
    alphaBoolBlend {blendOption=Linear}
    boneToModify {boneName=dyn_hair_back_03_l}
    constraintSetup [{targetBone {boneName=spine_05}}, {targetBone {boneName=clavicle_l}}, {targetBone {boneName=neck_02}}]
    constraintWeights [1,1,1]
   }
   n84 = "Constraint - dyn_hair_back_04_l (PPPP)"(ComponentPose=@n83, Alpha=1, ConstraintWeights_0=0.6653, ConstraintWeights_1=0.3309, ConstraintWeights_2=0, ConstraintWeights_3=0.0038) {
    alphaBoolBlend {blendOption=Linear}
    boneToModify {boneName=dyn_hair_back_04_l}
    constraintSetup [
     {targetBone {boneName=clavicle_l}}
     {targetBone {boneName=spine_05}}
     {targetBone {boneName=upperarm_twist_01_l}}
     {targetBone {boneName=clavicle_r}}
    ]
    constraintWeights [1,1,1,1]
   }
   n85 = "Constraint - dyn_hair_back_05_l (PPPP)"(ComponentPose=@n84, Alpha=1, ConstraintWeights_0=0.759, ConstraintWeights_1=0.1995, ConstraintWeights_2=0, ConstraintWeights_3=0.0415) {
    alphaBoolBlend {blendOption=Linear}
    boneToModify {boneName=dyn_hair_back_05_l}
    constraintSetup [
     {targetBone {boneName=clavicle_l}}
     {targetBone {boneName=spine_05}}
     {targetBone {boneName=upperarm_twist_01_l}}
     {targetBone {boneName=spine_04}}
    ]
    constraintWeights [1,1,1,1]
   }
   n86 = "Constraint - dyn_hair_back_06_l (PPPPPP)"(
    ComponentPose=@n85
    Alpha=1
    ConstraintWeights_0=0.5094
    ConstraintWeights_1=0.2398
    ConstraintWeights_2=0.239
    ConstraintWeights_3=0
    ConstraintWeights_4=0.0105
    ConstraintWeights_5=0.0013
   ) {
    alphaBoolBlend {blendOption=Linear}
    boneToModify {boneName=dyn_hair_back_06_l}
    constraintSetup [
     {targetBone {boneName=clavicle_l}}
     {targetBone {boneName=spine_04}}
     {targetBone {boneName=spine_05}}
     {targetBone {boneName=upperarm_twist_01_l}}
     {targetBone {boneName=spine_03}}
     {targetBone {boneName=spine_02}}
    ]
    constraintWeights [1,1,1,1,1,1]
   }
   n87 = "Constraint - dyn_hair_back_02_m (PP)"(ComponentPose=@n86, Alpha=1, ConstraintWeights_0=0.3, ConstraintWeights_1=0.7) {
    alphaBoolBlend {blendOption=Linear}
    boneToModify {boneName=dyn_hair_back_02_m}
    constraintSetup [{targetBone {boneName=head}}, {targetBone {boneName=neck_02}}]
    constraintWeights [1,0]
   }
   n88 = "Constraint - dyn_hair_back_03_m (PPPP)"(ComponentPose=@n87, Alpha=1, ConstraintWeights_0=0.3, ConstraintWeights_1=0.3, ConstraintWeights_2=0, ConstraintWeights_3=0.4) {
    alphaBoolBlend {blendOption=Linear}
    boneToModify {boneName=dyn_hair_back_03_m}
    constraintSetup [{targetBone {boneName=neck_02}}, {targetBone {boneName=neck_01}}, {targetBone {boneName=head}}, {targetBone {boneName=spine_05}}]
    constraintWeights [1,1,1,1]
   }
   n89 = "Constraint - dyn_hair_back_04_m (PPP)"(ComponentPose=@n88, Alpha=1, ConstraintWeights_0=0.9548, ConstraintWeights_1=0.0226, ConstraintWeights_2=0.0226) {
    alphaBoolBlend {blendOption=Linear}
    boneToModify {boneName=dyn_hair_back_04_m}
    constraintSetup [{targetBone {boneName=spine_05}}, {targetBone {boneName=clavicle_r}}, {targetBone {boneName=clavicle_l}}]
    constraintWeights [1,1,1]
   }
   n90 = "Constraint - dyn_hair_back_05_m (PPPP)"(ComponentPose=@n89, Alpha=1, ConstraintWeights_0=0.6851, ConstraintWeights_1=0.1717, ConstraintWeights_2=0.1328, ConstraintWeights_3=0.0104) {
    alphaBoolBlend {blendOption=Linear}
    boneToModify {boneName=dyn_hair_back_05_m}
    constraintSetup [{targetBone {boneName=spine_05}}, {targetBone {boneName=clavicle_l}}, {targetBone {boneName=clavicle_r}}, {targetBone {boneName=spine_04}}]
    constraintWeights [1,1,1,1]
   }
   n91 = "Constraint - dyn_hair_back_06_m (PPPPP)"(
    ComponentPose=@n90
    Alpha=1
    ConstraintWeights_0=0.4118
    ConstraintWeights_1=0.3788
    ConstraintWeights_2=0.1139
    ConstraintWeights_3=0.0799
    ConstraintWeights_4=0.0156
   ) {
    alphaBoolBlend {blendOption=Linear}
    boneToModify {boneName=dyn_hair_back_06_m}
    constraintSetup [
     {targetBone {boneName=spine_04}}
     {targetBone {boneName=spine_05}}
     {targetBone {boneName=clavicle_l}}
     {targetBone {boneName=clavicle_r}}
     {targetBone {boneName=spine_03}}
    ]
    constraintWeights [1,1,1,1,1]
   }
   n92 = "Constraint - dyn_hair_back_01_r (PPP)"!(ComponentPose=@n91, Alpha=1, ConstraintWeights_0=0.4, ConstraintWeights_1:float=0.4, ConstraintWeights_2:float=0.2) -> (Pose) {
    alphaBoolBlend {blendOption=Linear}
    boneToModify {boneName=dyn_hair_back_01_r}
    constraintSetup [{targetBone {boneName=head}}, {targetBone {boneName=neck_02}}, {targetBone {boneName=clavicle_r}}]
    constraintWeights [1,0,1]
   }
   n93 = "Constraint - dyn_hair_back_02_r (PPP)"!(ComponentPose=@n92, Alpha=1, ConstraintWeights_0=0.1, ConstraintWeights_1:float=0.1, ConstraintWeights_2:float=0.8) -> (Pose) {
    alphaBoolBlend {blendOption=Linear}
    boneToModify {boneName=dyn_hair_back_02_r}
    constraintSetup [{targetBone {boneName=head}}, {targetBone {boneName=neck_02}}, {targetBone {boneName=clavicle_r}}]
    constraintWeights [1,0,1]
   }
   n94 = "Constraint - dyn_hair_back_03_r (PPPP)"!(
    ComponentPose=@n93
    Alpha=1
    ConstraintWeights_0=0.2
    ConstraintWeights_1:float=0.2
    ConstraintWeights_2:float=0.2
    ConstraintWeights_3:float=0.4
   ) -> (Pose) {
    alphaBoolBlend {blendOption=Linear}
    boneToModify {boneName=dyn_hair_back_03_r}
    constraintSetup [{targetBone {boneName=spine_05}}, {targetBone {boneName=neck_02}}, {targetBone {boneName=neck_01}}, {targetBone {boneName=clavicle_r}}]
    constraintWeights [1,1,1,1]
   }
   n95 = "Constraint - dyn_hair_back_04_r (PPPP)"(ComponentPose=@n94, Alpha=1, ConstraintWeights_0=0.6653, ConstraintWeights_1=0.3309, ConstraintWeights_2=0, ConstraintWeights_3=0.0038) {
    alphaBoolBlend {blendOption=Linear}
    boneToModify {boneName=dyn_hair_back_04_r}
    constraintSetup [
     {targetBone {boneName=clavicle_r}}
     {targetBone {boneName=spine_05}}
     {targetBone {boneName=upperarm_twist_01_r}}
     {targetBone {boneName=clavicle_l}}
    ]
    constraintWeights [1,1,1,1]
   }
   n96 = "Constraint - dyn_hair_back_05_r (PPPPP)"(
    ComponentPose=@n95
    Alpha=1
    ConstraintWeights_0=0.6906
    ConstraintWeights_1=0.2575
    ConstraintWeights_2=0.0507
    ConstraintWeights_3=0
    ConstraintWeights_4=0.0012
   ) {
    alphaBoolBlend {blendOption=Linear}
    boneToModify {boneName=dyn_hair_back_05_r}
    constraintSetup [
     {targetBone {boneName=clavicle_r}}
     {targetBone {boneName=spine_05}}
     {targetBone {boneName=spine_04}}
     {targetBone {boneName=upperarm_twist_01_r}}
     {targetBone {boneName=clavicle_l}}
    ]
    constraintWeights [1,1,1,1,1]
   }
   n97 = "Constraint - dyn_hair_back_06_r (PPPPP)"(
    ComponentPose=@n96
    Alpha=1
    ConstraintWeights_0=0.5099
    ConstraintWeights_1=0.2403
    ConstraintWeights_2=0.2394
    ConstraintWeights_3=0
    ConstraintWeights_4=0.0105
   ) {
    alphaBoolBlend {blendOption=Linear}
    boneToModify {boneName=dyn_hair_back_06_r}
    constraintSetup [
     {targetBone {boneName=clavicle_r}}
     {targetBone {boneName=spine_04}}
     {targetBone {boneName=spine_05}}
     {targetBone {boneName=upperarm_twist_01_r}}
     {targetBone {boneName=spine_03}}
    ]
    constraintWeights [1,1,1,1,1]
   }
   n98 = Component To Local(ComponentPose=@n97)
   n99 = RigidBody!(
    ComponentPose=Local To Component(LocalPose=Use cached pose 'Cache_StartPose'())
    bAlphaBoolEnabled=NOT Boolean(A=Get bIsInFrontEnd())
    OverrideWorldGravity=Get GravityOverride()
    ExternalForce
    ComponentLinearAccScale="0,0,0"
    ComponentLinearVelScale="0,0,0"
    SimSpaceSettings:Sim Space Settings Structure=(WorldAlpha=1,VelocityScaleZ=1,DampingAlpha=0,MaxLinearVelocity=200,MaxAngularVelocity=4,MaxLinearAcceleration=800,MaxAngularAcceleration=16,ExternalLinearDragV=(X=0,Y=0,Z=0),ExternalLinearVelocity=(X=0,Y=0,Z=0),ExternalAngularVelocity=(X=0,Y=0,Z=0))
   ) -> (Pose) {
    alphaInputType=Bool
    bClampLinearTranslationLimitToRefPose=true
    bOverrideWorldGravity=true
    overridePhysicsAsset=&/F_MED_MutedMaroon_FaceAcc_Physics_alt2
    simulationSpace=BaseBoneSpace
    alphaBoolBlend {blendOption=Linear}
    baseBoneRef {boneName=neck_01}
    simSpaceSettings {dampingAlpha=0, maxAngularAcceleration=16, maxAngularVelocity=4, maxLinearAcceleration=800, maxLinearVelocity=200, worldAlpha=1}
   }
   n100 = RigidBody!(
    ComponentPose=@n99
    bAlphaBoolEnabled=Get bIsInFrontEnd()
    OverrideWorldGravity=Get GravityOverride()
    ExternalForce
    ComponentLinearAccScale="0,0,0"
    ComponentLinearVelScale="0,0,0"
   ) -> (Pose) {
    alphaInputType=Bool
    bClampLinearTranslationLimitToRefPose=true
    bOverrideWorldGravity=true
    overridePhysicsAsset=&F_MED_MutedMaroon_FaceAcc_FrontEnd_Physics
    simulationSpace=BaseBoneSpace
    alphaBoolBlend {blendOption=Linear}
    baseBoneRef {boneName=neck_01}
    simSpaceSettings {worldAlpha=0.05}
   }
   n101 = "Transform (Modify) Bone - Bone: dyn_hair_back_02_l"!(ComponentPose=@n100, Alpha:float=1, Translation="-0.073,-1.247,-0.015", Rotation="0.002,-2.282,-0.000", Scale="1,1,1") -> (Pose) {
    rotationMode=BMM_Additive
    rotationSpace=BCS_BoneSpace
    translationMode=BMM_Additive
    translationSpace=BCS_BoneSpace
    alphaBoolBlend {blendOption=Linear}
    alphaScaleBiasClamp {bInterpResult=true, interpSpeedDecreasing=1, interpSpeedIncreasing=1}
    boneToModify {boneName=dyn_hair_back_02_l}
    rotation=<rot 0.002 -2.282 0>
    translation=<-0.073 -1.247 -0.015>
   }
   n102 = "Transform (Modify) Bone - Bone: dyn_hair_back_02_r"!(ComponentPose=@n101, Alpha:float=1, Translation="0,0,0", Rotation="0.000,-3.528,0.000", Scale="1,1,1") -> (Pose) {
    rotationMode=BMM_Additive
    rotationSpace=BCS_BoneSpace
    translationMode=BMM_Additive
    translationSpace=BCS_BoneSpace
    alphaBoolBlend {blendOption=Linear}
    alphaScaleBiasClamp {bInterpResult=true, interpSpeedDecreasing=1, interpSpeedIncreasing=1}
    boneToModify {boneName=dyn_hair_back_02_r}
    rotation=<rot 0 -3.528 0>
   }
   n103 = "Transform (Modify) Bone - Bone: dyn_hair_back_01_r"(ComponentPose=@n102, Alpha=1, Translation="0,0,0", Rotation="4.177,-25.243,-1.981", Scale="1,1,1") {
    rotationMode=BMM_Additive
    rotationSpace=BCS_BoneSpace
    translationMode=BMM_Additive
    translationSpace=BCS_BoneSpace
    alphaBoolBlend {blendOption=Linear}
    alphaScaleBiasClamp {bInterpResult=true, interpSpeedDecreasing=1, interpSpeedIncreasing=1}
    boneToModify {boneName=dyn_hair_back_01_r}
    rotation=<rot 4.177 -25.243 -1.981>
   }
   n104 = "Transform (Modify) Bone - Bone: dyn_hair_back_01_m"(ComponentPose=@n103, Alpha=1, Translation="0,0,0", Rotation="6.694,-21.603,2.249", Scale="1,1,1") {
    rotationMode=BMM_Additive
    rotationSpace=BCS_BoneSpace
    translationMode=BMM_Additive
    translationSpace=BCS_BoneSpace
    alphaBoolBlend {blendOption=Linear}
    alphaScaleBiasClamp {bInterpResult=true, interpSpeedDecreasing=1, interpSpeedIncreasing=1}
    boneToModify {boneName=dyn_hair_back_01_m}
    rotation=<rot 6.694 -21.603 2.249>
   }
   n105 = "Transform (Modify) Bone - Bone: dyn_hair_back_01_l"(ComponentPose=@n104, Alpha=1, Translation="0,0,0", Rotation="-0.014,21.674,0.067", Scale="1,1,1") {
    rotationMode=BMM_Additive
    rotationSpace=BCS_BoneSpace
    translationMode=BMM_Additive
    translationSpace=BCS_BoneSpace
    alphaBoolBlend {blendOption=Linear}
    alphaScaleBiasClamp {bInterpResult=true, interpSpeedDecreasing=1, interpSpeedIncreasing=1}
    boneToModify {boneName=dyn_hair_back_01_l}
    rotation=<rot -0.014 21.674 0.067>
   }
   n106 = "Transform (Modify) Bone - Bone: dyn_hair_back_02_r"(ComponentPose=@n105, Alpha=1, Translation="0,0,0", Rotation="1.399,100.527,-1.003", Scale="1,1,1") {
    rotationMode=BMM_Additive
    rotationSpace=BCS_BoneSpace
    translationMode=BMM_Additive
    translationSpace=BCS_BoneSpace
    alphaBoolBlend {blendOption=Linear}
    alphaScaleBiasClamp {bInterpResult=true, interpSpeedDecreasing=1, interpSpeedIncreasing=1}
    boneToModify {boneName=dyn_hair_back_02_r}
    rotation=<rot 1.399 100.527 -1.003>
   }
   n107 = "Transform (Modify) Bone - Bone: dyn_hair_back_02_m"!(ComponentPose=@n106, Alpha:float=1, Translation="0,0,0", Rotation="-0.022,52.676,0.029", Scale="1,1,1") -> (Pose) {
    rotationMode=BMM_Additive
    rotationSpace=BCS_BoneSpace
    translationMode=BMM_Additive
    translationSpace=BCS_BoneSpace
    alphaBoolBlend {blendOption=Linear}
    alphaScaleBiasClamp {bInterpResult=true, interpSpeedDecreasing=1, interpSpeedIncreasing=1}
    boneToModify {boneName=dyn_hair_back_02_m}
    rotation=<rot -0.022 52.676 0.029>
   }
   n108 = "Transform (Modify) Bone - Bone: dyn_hair_back_02_l"(ComponentPose=@n107, Alpha=1, Translation="0,0,0", Rotation="0.029,-42.757,0.034", Scale="1,1,1") {
    rotationMode=BMM_Additive
    rotationSpace=BCS_BoneSpace
    translationMode=BMM_Additive
    translationSpace=BCS_BoneSpace
    alphaBoolBlend {blendOption=Linear}
    alphaScaleBiasClamp {bInterpResult=true, interpSpeedDecreasing=1, interpSpeedIncreasing=1}
    boneToModify {boneName=dyn_hair_back_02_l}
    rotation=<rot 0.029 -42.757 0.034>
   }
   n109 = "Transform (Modify) Bone - Bone: dyn_hair_back_03_r"!(ComponentPose=@n108, Alpha:float=1, Translation="0,0,0", Rotation="-1.382,-75.211,0.688", Scale="1,1,1") -> (Pose) {
    rotationMode=BMM_Additive
    rotationSpace=BCS_BoneSpace
    translationMode=BMM_Additive
    translationSpace=BCS_BoneSpace
    alphaBoolBlend {blendOption=Linear}
    alphaScaleBiasClamp {bInterpResult=true, interpSpeedDecreasing=1, interpSpeedIncreasing=1}
    boneToModify {boneName=dyn_hair_back_03_r}
    rotation=<rot -1.382 -75.211 0.688>
   }
   n110 = "Transform (Modify) Bone - Bone: dyn_hair_back_03_m"!(ComponentPose=@n109, Alpha:float=1, Translation="0,0,0", Rotation="0.022,-34.424,-0.130", Scale="1,1,1") -> (Pose) {
    rotationMode=BMM_Additive
    rotationSpace=BCS_BoneSpace
    translationMode=BMM_Additive
    translationSpace=BCS_BoneSpace
    alphaBoolBlend {blendOption=Linear}
    alphaScaleBiasClamp {bInterpResult=true, interpSpeedDecreasing=1, interpSpeedIncreasing=1}
    boneToModify {boneName=dyn_hair_back_03_m}
    rotation=<rot 0.022 -34.424 -0.13>
   }
   n111 = "Transform (Modify) Bone - Bone: dyn_hair_back_03_l"!(ComponentPose=@n110, Alpha:float=1, Translation="0,0,0", Rotation="0.008,43.796,0.007", Scale="1,1,1") -> (Pose) {
    rotationMode=BMM_Additive
    rotationSpace=BCS_BoneSpace
    translationMode=BMM_Additive
    translationSpace=BCS_BoneSpace
    alphaBoolBlend {blendOption=Linear}
    alphaScaleBiasClamp {bInterpResult=true, interpSpeedDecreasing=1, interpSpeedIncreasing=1}
    boneToModify {boneName=dyn_hair_back_03_l}
    rotation=<rot 0.008 43.7961 0.0069>
   }
   n112 = Component To Local(ComponentPose=@n111)
   n113 = Get isWearingBackbling()
   Output Pose(Result=Blend Poses by bool(
     BlendPose_0=Use cached pose 'Cache_BackpackState'()
     BlendPose_1=Blend Poses by bool(BlendPose_0=@n98, BlendPose_1=@n112, bActiveValue=@n113) {blendType=Linear}
     bActiveValue=Get Is Wearing Cape()
    ) {blendType=Linear})
   n114 = Control Rig!(Source=Use cached pose 'Cache_StartPose'(), Alpha, parachuting_alpha:double=Get isParachuting_alpha() -> (isParachuting_alpha:double)) -> (Pose) {
    alphaBoolBlend {blendOption=Linear}
    controlRigAssetReference {blueprintRigClass=&/F_MED_MutedMaroon_FaceAcc_CtrlRig.F_MED_MutedMaroon_FaceAcc_CtrlRig_C}
   }
   n115 = "Transform (Modify) Bone - Bone: dyn_hair_back_03_m"!(ComponentPose=Local To Component(LocalPose=@n114), bAlphaBoolEnabled:bool=True, Translation="0,0,0", Rotation="0,0,0", Scale="1,0.5,1") -> (Pose) {
    alphaInputType=Bool
    rotationSpace=BCS_BoneSpace
    scaleMode=BMM_Additive
    scaleSpace=BCS_BoneSpace
    translationSpace=BCS_BoneSpace
    alphaBoolBlend {blendInTime=0.5, blendOption=Linear, blendOutTime=0.8}
    boneToModify {boneName=dyn_hair_back_03_m}
    scale {y=0.5}
   }
   Save cached pose 'Cache_BackpackState'(Pose=Component To Local(ComponentPose=@n115)) {cacheName=Cache_BackpackState}
   n116 = Layered blend per bone(
    BasePose=Copy Pose From Mesh() {bCopyCurves=true, bUseAttachedParent=true}
    BlendPoses_0=Copy Pose From Mesh(SourceMeshComponent=Get Body_Mesh())
   ) {layerSetup [{branchFilters [
       {boneName=clavicle_l}
       {boneName=clavicle_r}
       {boneName=upperarm_l}
       {boneName=upperarm_r}
       {boneName=upperarm_twist_01_l}
       {boneName=upperarm_twist_01_r}
       {boneName=deform_upperarm_bk_base_l}
       {boneName=deform_upperarm_bk_l}
       {boneName=deform_upperarm_fr_base_l}
       {boneName=deform_upperarm_fr_l}
       {boneName=deform_upperarm_up_base_l}
       {boneName=deform_upperarm_up_l}
       {boneName=deform_upperarm_bk_base_r}
       {boneName=deform_upperarm_bk_r}
       {boneName=deform_upperarm_fr_base_r}
       {boneName=deform_upperarm_fr_r}
       {boneName=deform_upperarm_up_base_r}
       {boneName=deform_upperarm_up_r}
      ]}]}
   n117 = "Transform (Modify) Bone - Bone: base_hair_back_r"!(ComponentPose=Local To Component(LocalPose=@n116), Alpha=1, Rotation="0,10,15") -> (Pose) {
    rotationMode=BMM_Additive
    rotationSpace=BCS_BoneSpace
    alphaBoolBlend {blendOption=Linear}
    boneToModify {boneName=base_hair_back_r}
    rotation {roll=15, yaw=10}
   }
   n118 = "Transform (Modify) Bone - Bone: base_hair_back_l"!(ComponentPose=@n117, Alpha=1, Rotation="0,0,-8") -> (Pose) {
    rotationMode=BMM_Additive
    rotationSpace=BCS_BoneSpace
    alphaBoolBlend {blendOption=Linear}
    boneToModify {boneName=base_hair_back_l}
    rotation {roll=-8}
   }
   n119 = Anim Dynamics - Bone: dyn_headwing_r_01!(ComponentPose=@n118, Alpha=1, GravityOverride:Vector="0,0,-980", ExternalForce="0,0,0") -> (Pose) {
    angularDampingOverride=1
    angularSpringConstant=15000
    bAngularSpring=true
    bOverrideAngularDamping=true
    bUseGravityOverride=true
    gravityScale=0.35
    simulationSpace=RootRelative
    alphaBoolBlend {blendOption=Linear}
    boundBone {boneName=dyn_headwing_r_01}
    componentLinearAccScale=<1 1 1>
    componentLinearVelScale=<1 1 1>
    gravityOverride {z=-980}
    physicsBodyDefinitions [{
      boundBone {boneName=dyn_headwing_r_01}
      boxExtents=<20 30 30>
      constraintSetup {angularLimitsMax=<0 0 5>, angularLimitsMin=<0 0 -5>, angularTarget=<1 0 0>}
      localJointOffset=<2 0 0>
     }]
   }
   n120 = Anim Dynamics - Bone: dyn_headwing_r_02!(ComponentPose=@n119, Alpha=1, GravityOverride:Vector="0,0,-980", ExternalForce="0,0,0") -> (Pose) {
    angularDampingOverride=1
    angularSpringConstant=15000
    bAngularSpring=true
    bOverrideAngularDamping=true
    bUseGravityOverride=true
    gravityScale=0.35
    simulationSpace=RootRelative
    alphaBoolBlend {blendOption=Linear}
    boundBone {boneName=dyn_headwing_r_02}
    componentLinearAccScale=<1 1 1>
    componentLinearVelScale=<1 1 1>
    gravityOverride {z=-980}
    physicsBodyDefinitions [{
      boundBone {boneName=dyn_headwing_r_02}
      boxExtents=<20 30 30>
      constraintSetup {angularLimitsMax=<0 0 5>, angularLimitsMin=<0 0 -5>, angularTarget=<1 0 0>}
      localJointOffset=<2 0 0>
     }]
   }
   n121 = Anim Dynamics - Bone: dyn_headwing_l_01!(ComponentPose=@n120, Alpha=1, GravityOverride:Vector="0,0,-980", ExternalForce="0,0,0") -> (Pose) {
    angularDampingOverride=1
    angularSpringConstant=15000
    bAngularSpring=true
    bOverrideAngularDamping=true
    bUseGravityOverride=true
    gravityScale=0.35
    simulationSpace=RootRelative
    alphaBoolBlend {blendOption=Linear}
    boundBone {boneName=dyn_headwing_l_01}
    componentLinearAccScale=<1 1 1>
    componentLinearVelScale=<1 1 1>
    gravityOverride {z=-980}
    physicsBodyDefinitions [{
      boundBone {boneName=dyn_headwing_l_01}
      boxExtents=<20 30 30>
      constraintSetup {angularLimitsMax=<0 0 5>, angularLimitsMin=<0 0 -5>, angularTarget=<1 0 0>}
      localJointOffset=<2 0 0>
     }]
   }
   n122 = Anim Dynamics - Bone: dyn_headwing_l_02!(ComponentPose=@n121, Alpha=1, GravityOverride:Vector="0,0,-980", ExternalForce="0,0,0") -> (Pose) {
    angularDampingOverride=1
    angularSpringConstant=15000
    bAngularSpring=true
    bOverrideAngularDamping=true
    bUseGravityOverride=true
    gravityScale=0.35
    simulationSpace=RootRelative
    alphaBoolBlend {blendOption=Linear}
    boundBone {boneName=dyn_headwing_l_02}
    componentLinearAccScale=<1 1 1>
    componentLinearVelScale=<1 1 1>
    gravityOverride {z=-980}
    physicsBodyDefinitions [{
      boundBone {boneName=dyn_headwing_l_02}
      boxExtents=<20 30 30>
      constraintSetup {angularLimitsMax=<0 0 5>, angularLimitsMin=<0 0 -5>, angularTarget=<1 0 0>}
      localJointOffset=<2 0 0>
     }]
   }
   n123 = Anim Dynamics - Bone: dyn_Top_Hair_1!(ComponentPose=@n122, Alpha=1, GravityOverride:Vector="0,0,-980", ExternalForce="0,0,0") -> (Pose) {
    angularDampingOverride=1
    angularSpringConstant=10000
    bAngularSpring=true
    bOverrideAngularDamping=true
    bUseGravityOverride=true
    gravityScale=0.35
    simulationSpace=RootRelative
    alphaBoolBlend {blendOption=Linear}
    boundBone {boneName=dyn_Top_Hair_1}
    componentAppliedLinearAccClamp=<5000 5000 5000>
    componentLinearAccScale=<0.5 0.5 0.5>
    componentLinearVelScale=<0.5 0.5 0.5>
    gravityOverride {z=-980}
    physicsBodyDefinitions [{
      boundBone {boneName=dyn_Top_Hair_1}
      boxExtents=<20 30 30>
      constraintSetup {angularLimitsMax=<0 0 2>, angularLimitsMin=<0 0 -2>, angularTarget=<1 0 0>}
      localJointOffset=<2 0 0>
     }]
   }
   n124 = Anim Dynamics - Bone: dyn_Top_Hair_2!(ComponentPose=@n123, Alpha=1, GravityOverride:Vector="0,0,-980", ExternalForce="0,0,0") -> (Pose) {
    angularDampingOverride=1
    angularSpringConstant=10000
    bAngularSpring=true
    bOverrideAngularDamping=true
    bUseGravityOverride=true
    gravityScale=0.35
    simulationSpace=RootRelative
    alphaBoolBlend {blendOption=Linear}
    boundBone {boneName=dyn_Top_Hair_2}
    componentAppliedLinearAccClamp=<5000 5000 5000>
    componentLinearAccScale=<0.5 0.5 0.5>
    componentLinearVelScale=<0.5 0.5 0.5>
    gravityOverride {z=-980}
    physicsBodyDefinitions [{
      boundBone {boneName=dyn_Top_Hair_2}
      boxExtents=<20 30 30>
      constraintSetup {angularLimitsMax=<0 0 2>, angularLimitsMin=<0 0 -2>, angularTarget=<1 0 0>}
      localJointOffset=<2 0 0>
     }]
   }
   n125 = Anim Dynamics - Bone: dyn_Top_Hair_3!(ComponentPose=@n124, Alpha=1, GravityOverride:Vector="0,0,-980", ExternalForce="0,0,0") -> (Pose) {
    angularDampingOverride=1
    angularSpringConstant=10000
    bAngularSpring=true
    bOverrideAngularDamping=true
    bUseGravityOverride=true
    gravityScale=0.35
    simulationSpace=RootRelative
    alphaBoolBlend {blendOption=Linear}
    boundBone {boneName=dyn_Top_Hair_3}
    componentAppliedLinearAccClamp=<5000 5000 5000>
    componentLinearAccScale=<0.5 0.5 0.5>
    componentLinearVelScale=<0.5 0.5 0.5>
    gravityOverride {z=-980}
    physicsBodyDefinitions [{
      boundBone {boneName=dyn_Top_Hair_3}
      boxExtents=<20 30 30>
      constraintSetup {angularLimitsMax=<0 0 2>, angularLimitsMin=<0 0 -2>, angularTarget=<1 0 0>}
      localJointOffset=<2 0 0>
     }]
   }
   n126 = Anim Dynamics - Bone: dyn_FrntHair_A_B_1!(ComponentPose=@n125, Alpha=1, GravityOverride:Vector="0,0,-980", ExternalForce="0,0,0") -> (Pose) {
    angularDampingOverride=1
    angularSpringConstant=15000
    bAngularSpring=true
    bOverrideAngularDamping=true
    bUseGravityOverride=true
    gravityScale=0.35
    simulationSpace=RootRelative
    alphaBoolBlend {blendOption=Linear}
    boundBone {boneName=dyn_FrntHair_A_B_1}
    componentLinearAccScale=<1 1 1>
    componentLinearVelScale=<1 1 1>
    gravityOverride {z=-980}
    physicsBodyDefinitions [{
      boundBone {boneName=dyn_FrntHair_A_B_1}
      boxExtents=<20 30 30>
      constraintSetup {angularLimitsMax=<0 10 10>, angularLimitsMin=<0 -10 -10>, angularTarget=<1 0 0>}
      localJointOffset=<4 0 0>
     }]
   }
   n127 = Anim Dynamics - Bone: dyn_FrntHair_A_B_2!(ComponentPose=@n126, Alpha=1, GravityOverride:Vector="0,0,-980", ExternalForce="0,0,0") -> (Pose) {
    angularDampingOverride=1
    angularSpringConstant=15000
    bAngularSpring=true
    bOverrideAngularDamping=true
    bUseGravityOverride=true
    gravityScale=0.35
    simulationSpace=RootRelative
    alphaBoolBlend {blendOption=Linear}
    boundBone {boneName=dyn_FrntHair_A_B_2}
    componentLinearAccScale=<1 1 1>
    componentLinearVelScale=<1 1 1>
    gravityOverride {z=-980}
    physicsBodyDefinitions [{
      boundBone {boneName=dyn_FrntHair_A_B_2}
      boxExtents=<20 30 30>
      constraintSetup {angularLimitsMax=<0 10 10>, angularLimitsMin=<0 -10 -10>, angularTarget=<1 0 0>}
      localJointOffset=<4 0 0>
     }]
   }
   n128 = Anim Dynamics - Bone: dyn_Back_Hair_C_A!(ComponentPose=@n127, Alpha=1, GravityOverride:Vector="0,0,-980", ExternalForce="0,0,0") -> (Pose) {
    angularDampingOverride=1
    angularSpringConstant=5000
    bAngularSpring=true
    bOverrideAngularDamping=true
    bUseGravityOverride=true
    gravityScale=0.35
    simulationSpace=RootRelative
    alphaBoolBlend {blendOption=Linear}
    boundBone {boneName=dyn_Back_Hair_C_A}
    componentAppliedLinearAccClamp=<5000 5000 5000>
    componentLinearAccScale=<0.5 0.5 0.5>
    componentLinearVelScale=<0.5 0.5 0.5>
    gravityOverride {z=-980}
    physicsBodyDefinitions [{
      boundBone {boneName=dyn_Back_Hair_C_A}
      boxExtents=<15 15 15>
      constraintSetup {angularLimitsMax=<0 5 0>, angularLimitsMin=<0 -5 0>, angularTarget=<1 0 0>}
      localJointOffset=<1 0 0>
     }]
   }
   n129 = Anim Dynamics - Bone: dyn_Back_Hair_A_A!(ComponentPose=@n128, Alpha=1, GravityOverride:Vector="0,0,-980", ExternalForce="0,0,0") -> (Pose) {
    angularDampingOverride=1
    angularSpringConstant=5000
    bAngularSpring=true
    bOverrideAngularDamping=true
    bUseGravityOverride=true
    gravityScale=0.35
    simulationSpace=RootRelative
    alphaBoolBlend {blendOption=Linear}
    boundBone {boneName=dyn_Back_Hair_A_A}
    componentAppliedLinearAccClamp=<5000 5000 5000>
    componentLinearAccScale=<0.5 0.5 0.5>
    componentLinearVelScale=<0.5 0.5 0.5>
    gravityOverride {z=-980}
    physicsBodyDefinitions [{
      boundBone {boneName=dyn_Back_Hair_A_A}
      boxExtents=<15 15 15>
      constraintSetup {angularLimitsMax=<0 10 0>, angularLimitsMin=<0 -10 0>, angularTarget=<1 0 0>}
      localJointOffset=<1 0 0>
     }]
   }
   n130 = Anim Dynamics - Bone: dyn_FrntHair_A_3!(ComponentPose=@n129, Alpha=1, GravityOverride:Vector="0,0,-980", ExternalForce="0,0,0") -> (Pose) {
    angularDampingOverride=1
    angularSpringConstant=7000
    bAngularSpring=true
    bOverrideAngularDamping=true
    bUseGravityOverride=true
    gravityScale=0.35
    simulationSpace=RootRelative
    alphaBoolBlend {blendOption=Linear}
    boundBone {boneName=dyn_FrntHair_A_3}
    gravityOverride {z=-980}
    physicsBodyDefinitions [{
      boundBone {boneName=dyn_FrntHair_A_3}
      boxExtents=<15 15 15>
      constraintSetup {angularLimitsMax=<0 0 2>, angularLimitsMin=<0 0 -2>, angularTarget=<1 0 0>}
      localJointOffset=<1 0 0>
     }]
   }
   n131 = Anim Dynamics - Bone: dyn_FrntHair_A_4!(ComponentPose=@n130, Alpha=1, GravityOverride:Vector="0,0,-980", ExternalForce="0,0,0") -> (Pose) {
    angularDampingOverride=1
    angularSpringConstant=7000
    bAngularSpring=true
    bOverrideAngularDamping=true
    bUseGravityOverride=true
    gravityScale=0.35
    simulationSpace=RootRelative
    alphaBoolBlend {blendOption=Linear}
    boundBone {boneName=dyn_FrntHair_A_4}
    gravityOverride {z=-980}
    physicsBodyDefinitions [{
      boundBone {boneName=dyn_FrntHair_A_4}
      boxExtents=<15 15 15>
      constraintSetup {angularLimitsMax=<0 0 2>, angularLimitsMin=<0 0 -2>, angularTarget=<1 0 0>}
      localJointOffset=<1 0 0>
     }]
   }
   n132 = Anim Dynamics - Bone: dyn_FrntHair_B_2!(ComponentPose=@n131, Alpha=1, GravityOverride:Vector="0,0,-980", ExternalForce="0,0,0") -> (Pose) {
    angularDampingOverride=1
    angularSpringConstant=10000
    bAngularSpring=true
    bOverrideAngularDamping=true
    bUseGravityOverride=true
    bUseSphericalLimits=true
    gravityScale=0.35
    simulationSpace=RootRelative
    alphaBoolBlend {blendOption=Linear}
    boundBone {boneName=dyn_FrntHair_B_2}
    componentAppliedLinearAccClamp=<5000 5000 5000>
    componentLinearAccScale=<0.5 0.5 0.5>
    componentLinearVelScale=<0.5 0.5 0.5>
    gravityOverride {z=-980}
    physicsBodyDefinitions [{
      boundBone {boneName=dyn_FrntHair_B_2}
      boxExtents=<15 15 15>
      constraintSetup {angularLimitsMax=<0 0 2>, angularLimitsMin=<0 0 -2>, angularTarget=<1 0 0>}
      localJointOffset=<1 0 0>
     }]
    sphericalLimits [{limitRadius=10, drivingBone {boneName=head}, sphereLocalOffset=<4.8 1.3 7>}]
   }
   n133 = Anim Dynamics - Bone: dyn_FrntHair_B_3!(ComponentPose=@n132, Alpha=1, GravityOverride:Vector="0,0,-980", ExternalForce="0,0,0") -> (Pose) {
    angularDampingOverride=1
    angularSpringConstant=10000
    bAngularSpring=true
    bOverrideAngularDamping=true
    bUseGravityOverride=true
    bUseSphericalLimits=true
    gravityScale=0.35
    simulationSpace=RootRelative
    alphaBoolBlend {blendOption=Linear}
    boundBone {boneName=dyn_FrntHair_B_3}
    componentAppliedLinearAccClamp=<5000 5000 5000>
    componentLinearAccScale=<0.5 0.5 0.5>
    componentLinearVelScale=<0.5 0.5 0.5>
    gravityOverride {z=-980}
    physicsBodyDefinitions [{
      boundBone {boneName=dyn_FrntHair_B_3}
      boxExtents=<15 15 15>
      constraintSetup {angularLimitsMax=<0 0 2>, angularLimitsMin=<0 0 -2>, angularTarget=<1 0 0>}
      localJointOffset=<1 0 0>
     }]
    sphericalLimits [{limitRadius=10, drivingBone {boneName=head}, sphereLocalOffset=<4.8 1.3 0.7>}]
   }
   n134 = Anim Dynamics - Bone: dyn_FrntHair_C_1!(ComponentPose=@n133, Alpha=1, GravityOverride:Vector="0,0,-980", ExternalForce="0,0,0") -> (Pose) {
    angularDampingOverride=1
    angularSpringConstant=10000
    bAngularSpring=true
    bOverrideAngularDamping=true
    bUseGravityOverride=true
    gravityScale=0.35
    simulationSpace=RootRelative
    alphaBoolBlend {blendOption=Linear}
    boundBone {boneName=dyn_FrntHair_C_1}
    componentAppliedLinearAccClamp=<5000 5000 5000>
    componentLinearAccScale=<0.5 0.5 0.5>
    componentLinearVelScale=<0.5 0.5 0.5>
    gravityOverride {z=-980}
    physicsBodyDefinitions [{
      boundBone {boneName=dyn_FrntHair_C_1}
      boxExtents=<15 15 15>
      constraintSetup {angularLimitsMax=<0 0 1>, angularLimitsMin=<0 0 -1>, angularTarget=<1 0 0>}
      localJointOffset=<1 0 0>
     }]
   }
   n135 = Anim Dynamics - Bone: dyn_FrntHair_C_2!(ComponentPose=@n134, Alpha=1, GravityOverride:Vector="0,0,-980", ExternalForce="0,0,0") -> (Pose) {
    angularDampingOverride=1
    angularSpringConstant=10000
    bAngularSpring=true
    bOverrideAngularDamping=true
    bUseGravityOverride=true
    bUseSphericalLimits=true
    gravityScale=0.35
    simulationSpace=RootRelative
    alphaBoolBlend {blendOption=Linear}
    boundBone {boneName=dyn_FrntHair_C_2}
    componentAppliedLinearAccClamp=<5000 5000 5000>
    componentLinearAccScale=<0.5 0.5 0.5>
    componentLinearVelScale=<0.5 0.5 0.5>
    gravityOverride {z=-980}
    physicsBodyDefinitions [{
      boundBone {boneName=dyn_FrntHair_C_2}
      boxExtents=<15 15 15>
      constraintSetup {angularLimitsMax=<0 0 1>, angularLimitsMin=<0 0 -1>, angularTarget=<1 0 0>}
      localJointOffset=<1 0 0>
     }]
    sphericalLimits [{limitRadius=10.5, drivingBone {boneName=head}, sphereLocalOffset=<4.9173 0.996 1.789>}]
   }
   n136 = Anim Dynamics - Bone: dyn_earring_l!(ComponentPose=@n135, Alpha=1, GravityOverride:Vector="0,0,-980", ExternalForce="0,0,0") -> (Pose) {
    angularDampingOverride=1
    angularSpringConstant=5000
    bAngularSpring=true
    bOverrideAngularDamping=true
    bUseGravityOverride=true
    gravityScale=0.35
    simulationSpace=RootRelative
    alphaBoolBlend {blendOption=Linear}
    boundBone {boneName=dyn_earring_l}
    gravityOverride {z=-980}
    physicsBodyDefinitions [{
      boundBone {boneName=dyn_earring_l}
      boxExtents=<50 50 50>
      constraintSetup {angularLimitsMax=<30 30 30>, angularLimitsMin=<-30 -30 -30>, angularTarget=<1 0 0>}
      localJointOffset=<5 0 0>
     }]
   }
   n137 = Anim Dynamics - Bone: dyn_earring_r!(ComponentPose=@n136, Alpha=1, GravityOverride:Vector="0,0,-980", ExternalForce="0,0,0") -> (Pose) {
    angularDampingOverride=1
    angularSpringConstant=5000
    bAngularSpring=true
    bOverrideAngularDamping=true
    bUseGravityOverride=true
    gravityScale=0.35
    simulationSpace=RootRelative
    alphaBoolBlend {blendOption=Linear}
    boundBone {boneName=dyn_earring_r}
    gravityOverride {z=-980}
    physicsBodyDefinitions [{
      boundBone {boneName=dyn_earring_r}
      boxExtents=<50 50 50>
      constraintSetup {angularLimitsMax=<30 30 30>, angularLimitsMin=<-30 -30 -30>, angularTarget=<1 0 0>}
      localJointOffset=<5 0 0>
     }]
   }
   Save cached pose 'Cache_StartPose'(Pose=Component To Local(ComponentPose=@n137)) {cacheName=Cache_StartPose}
  ]}
 EventGraph {nodes [
   Event BlueprintUpdateAnimation()
   > n1 = Sequence!(execute) -> (then_0, then_1, then_2:Exec, then_3:Exec, then_4:Exec, then_5:Exec, then_6:Exec, then_7:Exec, then_8:Exec, then_9:Exec)
   n2 = NOT Boolean(A=Get bIsMoving2D())
   n3 = Get bIsCrouching()
   Set Crouch not walking(execute=@n1.then_3, Crouch not walking:double="To Float (Boolean)"(InBool=AND Boolean(A=@n3, B=@n2))) -> (Output_Get:double)
   n4 = Get bIsMoving2D()
   n5 = Get bIsCrouching()
   Set CrouchWalking(execute=@n1.then_4, CrouchWalking:double="To Float (Boolean)"(InBool=AND Boolean(A=@n4, B=@n5))) -> (Output_Get:double)
   Set Back Pack SkyFloating(execute=@n1.then_5, Back Pack SkyFloating:bool=AND Boolean(A=Get isWearingBackbling(), B=Get bIsSkydiveFloating())) -> (Output_Get:bool)
   n6 = "Gravity Override State Function (Parameter Asset) (Nativized)"(
    execute=@n1.then_0
    CharacterPartAnimInstance=Self-Reference()
    GameTimeInSeconds=GetGameTimeInSeconds()
    Parameters=&/F_MED_MutedMaroon_FaceAcc_Parameters
   )
   > Set GravityOverride(GravityOverride=@n6)
   > Set RBAN Settings Out(RBAN Settings Out=@n6.OutRigidBodyAnimNodeParameters)
   n7 = CheckForBackbling()
   > Set isWearingBackbling(isWearingBackbling=AND Boolean!(A=@n7, B=NOT Boolean(A=@n7.bIsBackblingACape), C:bool=NOT Boolean(A=@n7.bIsBackblingLong)) -> (ReturnValue))
   > Set Is Wearing Cape(Is Wearing Cape=OR Boolean(A=@n7.bIsBackblingACape, B=@n7.bIsBackblingLong))
   n8 = Get HeadDown_Target_ADJ() -> (HeadDown_Target_ADJ:double)
   > Set AR_Down_ADJ(AR_Down_ADJ:bool=AND Boolean!(A="float > float"(A=@n8), B=Get bIsTargeting()) -> (ReturnValue)) -> (Output_Get:bool)
   n9 = Get HeadUpFloat() -> (HeadUpFloat:double)
   > Set AR_UP_ADJ(AR_UP_ADJ:bool=AND Boolean(A="float >= float"(A=@n9), B=Get bIsTargeting())) -> (Output_Get:bool)
   > Set HeadDown_Target_Float(HeadDown_Target_Float:double=SelectFloat(A=Get HeadDown_Target_ADJ(), B=0, bPickA=Get bIsTargeting())) -> (Output_Get:double)
   > Set IsmovingFloat(IsmovingFloat:double="To Float (Boolean)"(InBool=Get bIsMoving2D())) -> (Output_Get:double)
   > Set IsSkydive_Float(IsSkydive_Float:double="To Float (Boolean)"(InBool=Get bIsSkydiveFloating())) -> (Output_Get:double)
   > n10 = Branch(Condition=Get bIsSkyDiving())
   > Set IsSkydive_DYN(IsSkydive_DYN:double=0) -> (Output_Get:double)
   Set IsSkydive_DYN(execute=@n10.else, IsSkydive_DYN=1)
   n11 = Get Skel Mesh()
   Set Head Down(
    execute=@n1.then_2
    Head Down:double=GetAlphaFromReferenceTransformDelta(SkeletalMeshComponent=@n11, BoneName=spine_05, BaseName="head ", InputRange=(X=0,Y=-90), OutputRange=(X=0,Y=-1), TwistAxis="0,0,1")
   ) -> (Output_Get:double)
   > Set Head_Turn_L(Head_Turn_L:double=GetAlphaFromReferenceTransformDelta(SkeletalMeshComponent=@n11, BoneName=spine_05, BaseName=head, DeltaComponentType=EulerX, InputRange=(X=0,Y=-22), TwistAxis="1,0,0")) -> (Output_Get:double)
   > Set Head_Turn_R(Head_Turn_R:double=GetAlphaFromReferenceTransformDelta(SkeletalMeshComponent=@n11, BoneName=spine_05, BaseName=head, DeltaComponentType=EulerX, InputRange=(X=0,Y=22), TwistAxis="1,0,0")) -> (Output_Get:double)
   n12 = GetAlphaFromReferenceTransformDelta(SkeletalMeshComponent=@n11, BoneName=spine_05, BaseName=head, InputRange=(X=10,Y=70), TwistAxis="0,0,1")
   Set HeadUpFloat(execute=@n1.then_1, HeadUpFloat:double=@n12) -> (Output_Get:double)
   n13 = BreakTransform(InTransform=GetDeltaTransformFromRefPose(self=GetOwnerSkeletalMeshComponent(), BoneName=head, BaseName=spine_05))
   > Set HeadDown_Target_ADJ(HeadDown_Target_ADJ:double=MapRangeClamped(Value=BreakRotator(InRot=@n13.Rotation).Yaw, InRangeA=0, InRangeB=75)) -> (Output_Get:double)
   n14 = Get bIsTargeting()
   n15 = Get isWearingBackbling()
   > Set HeadDown_Target_ADJ_Backpack(HeadDown_Target_ADJ_Backpack:double=SelectFloat(A=Get HeadDown_Target_ADJ(), B=0, bPickA=AND Boolean(A=@n15, B=@n14))) -> (Output_Get:double)
   n16 = Get bIsTargeting()
   n17 = Get bIsCrouching()
   > Set HeadUpFloat_Target(HeadUpFloat_Target:double=SelectFloat(A=Get HeadUpFloat(), B=0, bPickA=AND Boolean(A=@n16, B=@n17))) -> (Output_Get:double)
   n18 = NOT Boolean(A=Get bIsCrouching())
   n19 = NOT Boolean(A=Get Is Wearing Cape())
   n20 = NOT Boolean(A=Get isWearingBackbling())
   n21 = Get bIsTargeting()
   > Set HeadUpFloat_Target_Stand(HeadUpFloat_Target_Stand:double=SelectFloat(A=Get HeadUpFloat(), B=0, bPickA=AND Boolean!(A=@n21, B=@n18, C:bool=@n20, D:bool=@n19) -> (ReturnValue))) -> (Output_Get:double)
   n22 = Get bIsTargeting()
   n23 = Get bIsCrouching()
   n24 = Get bIsMoving2D()
   > Set crouching walk headUP(crouching walk headUP:double=SelectFloat(A=Get HeadUpFloat(), B=0, bPickA=AND Boolean!(A=@n22, B=@n23, C=@n24) -> (ReturnValue))) -> (Output_Get:double)
   Event OnCharacterCustomizationLoaded()
   > n25 = Sequence!(execute) -> (then_0, then_1)
   Delay(execute=@n25.then_1, Duration=0.2)
   > ResetDynamics(InTeleportType=TeleportPhysics)
   > ResetDynamics(InTeleportType=ResetPhysics)
   n26 = Cast To PlayerPawn_Athena(execute=@n25.then_0, Object=TryGetPawnOwner())
   > Set Body_Mesh(Body_Mesh=GetSkeletalMeshForPartType(self=@n26))
   Event BlueprintInitializeAnimation()
   > Set Skel Mesh(Skel Mesh=GetOwnerSkeletalMeshComponent())
   Reroute Node(InputPin:float=@n12) -> (OutputPin:float)
   Get bIsTargeting()
  ]}
]
variables [
 AR_Down_ADJ {type=boolean}
 AR_UP_ADJ {type=boolean}
 Back Pack SkyFloating {type=boolean}
 BackPack Jumping {type=boolean}
 Body_Mesh {type=SkeletalMeshComponent}
 Crouch not walking {type=number}
 crouching walk headUP {type=number}
 CrouchWalking {type=number}
 CroushingNot {type=boolean}
 GravityOverride {type=Vector}
 Head Down {type=number}
 Head_R_Rot {type=number}
 Head_Turn_L {type=number}
 Head_Turn_R {type=number}
 HeadDown_Target_ADJ {type=number}
 HeadDown_Target_ADJ_Backpack {type=number}
 HeadDown_Target_Float {type=number}
 HeadDownSide_Target_ADJ_0 {type=number}
 HeadUpBool {type=boolean}
 HeadUpFloat {type=number}
 HeadUpFloat_Target {type=number}
 HeadUpFloat_Target_Stand {type=number}
 Is Wearing Cape {type=boolean}
 IsmovingFloat {type=number}
 isParachuting_alpha {type=number}
 IsSkydive_DYN {type=number}
 IsSkydive_Float {type=number}
 isWearingBackbling {type=boolean}
 RBAN Settings Out {type=FortRigidBodyAnimNodeParameters}
 Skel Mesh {type=SkeletalMeshComponent}
]
